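# Additive UI material for an anamorphic flare in the Ashton LTM HUD, with a subtle time-driven flicker.
# A shine texture's red channel is shaped by a Falloff power, then multiplied by a looping LinearSine that slightly modulates brightness at FlickerSpeed.
# ColorSwitch blends between the flat Color parameter and that grayscale flare, defaulting to the flare, so Color replaces the shape rather than tinting it.
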Material /Game/Athena/HUD/LTM/Ashton/Materials/M_UI_AnamorphicFlare
graph {
 properties {blendMode=BLEND_Additive, materialDomain=MD_UI, pixelDepthOffsetMode=PDOM_Legacy, usedShadingModels=Default Lit}
 nodes [
  n1 = TextureSample() {samplerType=SAMPLERTYPE_LinearColor, texture=/Game/UI/Foundation/Textures/Pattern/T-UI-Shine-Anim}
  n2 = ScalarParameter() {defaultValue=3, parameterName=Falloff}
  n3 = Time()
  n4 = ScalarParameter() {defaultValue=3, parameterName=FlickerSpeed}
  n5 = MaterialFunctionCall(Value=Fmod(A=Multiply(A=@n3, B=@n4), B=Constant() {r=1})) {materialFunction=&LinearSine}
  LinearInterpolate(
   A=VectorParameter() {parameterName=Color, defaultValue=<rgba 1 1 1 1>}.RGB
   B=Multiply(A=Power(Base=@n1.R, Exp=@n2), B=LinearInterpolate(Alpha=@n5.Linear Sine) {constA=0.8})
   Alpha=ScalarParameter() {defaultValue=1, parameterName=ColorSwitch}
  )
 ]
}
