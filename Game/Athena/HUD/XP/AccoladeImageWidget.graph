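# Widget Blueprint for an XP accolade medal that animates its appearance, socketing and upgrade with wings, spikes, glows and a pulse ring.
# MedalInit plays Intro, adding Upgrade and spike effects for MedalLevel above one.
# EventSocket plays Socket with sound below level two, otherwise Upgrade over the end-state of Socket_VisualState, skipping ahead silently if already socketed.
# TriggerParticles, reached via a sequence event, optionally starts a UI emitter, then broadcasts SocketSmash.
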WidgetBlueprint /Game/Athena/HUD/XP/AccoladeImageWidget
animations [
 Intro {
  binding_count=9
  length_seconds=1.885
  object=&self:Intro
  sequence {bindings [
    MedalIcon {
     id=#1
     display_name="Image_Left_Wing1 (9)"
     possessed_class=Image
     tracks [{
       type=MovieScene2DTransformTrack
       sections [{
         type=MovieScene2DTransformSection
         channels [
          Scale.X {type=MovieSceneScriptingFloatChannel, keys [[0,0],[3.369,1],[8.4697,1]]}
          Scale.Y {type=MovieSceneScriptingFloatChannel, keys [[0,0],[3.369,1],[8.4697,1]]}
         ]
        }]
      }]
    }
    Image_Right_Wing1 {
     id=#2
     possessed_class=Image
     tracks [{
       type=MovieScene2DTransformTrack
       sections [{
         type=MovieScene2DTransformSection
         channels [
          Scale.X {type=MovieSceneScriptingFloatChannel, keys [[4.0703,-0.0045],[9.1177,1.651],[11.1953,1]]}
          Scale.Y {type=MovieSceneScriptingFloatChannel, keys [[4.0703,-0.0045],[9.4673,1.5601],[11.1953,1]]}
         ]
        }]
      }]
    }
    Image_Left_Wing1 {
     id=#3
     possessed_class=Image
     tracks [{
       type=MovieScene2DTransformTrack
       sections [{
         type=MovieScene2DTransformSection
         channels [
          Scale.X {type=MovieSceneScriptingFloatChannel, keys [[4.0703,-0.0045],[9.1177,1.651],[11.1953,1]]}
          Scale.Y {type=MovieSceneScriptingFloatChannel, keys [[4.0703,-0.0045],[9.4673,1.5601],[11.1953,1]]}
         ]
        }]
      }]
    }
    Image_Right_Wing2 {
     id=#4
     possessed_class=Image
     tracks [{
       type=MovieScene2DTransformTrack
       sections [{
         type=MovieScene2DTransformSection
         channels [
          Scale.X {type=MovieSceneScriptingFloatChannel, keys [[1.9407,-0.0045],[6.6147,1.651],[9.065,1]]}
          Scale.Y {type=MovieSceneScriptingFloatChannel, keys [[1.9407,-0.0045],[7.337,1.5601],[9.065,1]]}
         ]
        }]
      }]
    }
    Image_Right_Wing3 {
     id=#5
     possessed_class=Image
     tracks [{
       type=MovieScene2DTransformTrack
       sections [{
         type=MovieScene2DTransformSection
         channels [
          Scale.X {type=MovieSceneScriptingFloatChannel, keys [[0,-0.0045],[5.184,1.651],[7.124,1]]}
          Scale.Y {type=MovieSceneScriptingFloatChannel, keys [[0,-0.0045],[5.396,1.5601],[7.124,1]]}
         ]
        }]
      }]
    }
    Image_Left_Wing3 {
     id=#6
     possessed_class=Image
     tracks [{
       type=MovieScene2DTransformTrack
       sections [{
         type=MovieScene2DTransformSection
         channels [
          Scale.X {type=MovieSceneScriptingFloatChannel, keys [[0,-0.0045],[5.184,1.651],[7.124,1]]}
          Scale.Y {type=MovieSceneScriptingFloatChannel, keys [[0,-0.0045],[5.396,1.5601],[7.124,1]]}
         ]
        }]
      }]
    }
    Image_PulseRingImage {
     id=#7
     possessed_class=Image
     tracks [
      {
       type=MovieSceneWidgetMaterialTrack
       sections [{
         type=MovieSceneParameterSection
         channels [
          HardRingRadius {type=MovieSceneScriptingFloatChannel, keys [[9.4577,0.0414],[21.7817,0.5457]]}
          HardRingDensity {type=MovieSceneScriptingFloatChannel, keys [[9.4457,44.6938],[20.2787,1.9212],[32.7387,0]]}
          NoiseAmount {type=MovieSceneScriptingFloatChannel, keys [[9.4457,0],[21.7817,1.2778],[37.6987,1.2778]]}
          HardRingSize {type=MovieSceneScriptingFloatChannel, keys [[9.4457,-0.0047],[17.289,0.0222],[21.7817,0.0714]]}
          HardRingEdge {type=MovieSceneScriptingFloatChannel, keys [[9.4457,0.1947],[21.7817,0.0496]]}
          NoiseRepeat {type=MovieSceneScriptingFloatChannel, keys [[15.0577,0.9471],[37.6987,1.4696]]}
         ]
        }]
      }
      {
       type=MovieSceneFloatTrack
       display_name=Render Opacity
       sections [{type=MovieSceneFloatSection, channels [None {type=MovieSceneScriptingFloatChannel, keys [[9.4457,0],[11.6263,1]]}]}]
      }
     ]
    }
    Image_Left_Wing2 {
     id=#8
     possessed_class=Image
     tracks [{
       type=MovieScene2DTransformTrack
       sections [{
         type=MovieScene2DTransformSection
         channels [
          Scale.X {type=MovieSceneScriptingFloatChannel, keys [[1.9407,-0.0045],[6.6147,1.651],[9.065,1]]}
          Scale.Y {type=MovieSceneScriptingFloatChannel, keys [[1.9407,-0.0045],[7.337,1.5601],[9.065,1]]}
         ]
        }]
      }]
    }
   ]}
 }
 Upgrade {
  binding_count=7
  length_seconds=2.2216
  object=&self:Upgrade
  sequence {bindings [
    MedalIcon {
     id=#1
     display_name="Image_Right_Wing2 (7)"
     possessed_class=Image
     tracks [{
       type=MovieSceneWidgetMaterialTrack
       sections [{type=MovieSceneParameterSection, channels [RevealAmount {type=MovieSceneScriptingFloatChannel, keys [[0,0],[20.2787,1]]}]}]
      }]
    }
    Image_spike_LeftTop {
     id=#9
     possessed_class=Image
     tracks [{
       type=MovieScene2DTransformTrack
       sections [{
         type=MovieScene2DTransformSection
         channels [
          Scale.X {type=MovieSceneScriptingFloatChannel, keys [[10.8543,0],[14.2563,0.32],[16.0267,0.55]]}
          Scale.Y {type=MovieSceneScriptingFloatChannel, keys [[10.8543,0],[14.2563,0.7728],[16.0267,0.55]]}
         ]
        }]
      }]
    }
    Image_spike_RightBottom {
     id=#10
     possessed_class=Image
     tracks [{
       type=MovieScene2DTransformTrack
       sections [{
         type=MovieScene2DTransformSection
         channels [
          Scale.X {type=MovieSceneScriptingFloatChannel, keys [[10.8543,0],[14.2563,0.32],[16.0267,0.55]]}
          Scale.Y {type=MovieSceneScriptingFloatChannel, keys [[10.8543,0],[14.2563,0.7728],[16.0267,0.55]]}
         ]
        }]
      }]
    }
    Image_spike_LeftBottom {
     id=#11
     possessed_class=Image
     tracks [{
       type=MovieScene2DTransformTrack
       sections [{
         type=MovieScene2DTransformSection
         channels [
          Scale.X {type=MovieSceneScriptingFloatChannel, keys [[10.8543,0],[14.2563,0.32],[16.0267,0.55]]}
          Scale.Y {type=MovieSceneScriptingFloatChannel, keys [[10.8543,0],[14.2563,0.7728],[16.0267,0.55]]}
         ]
        }]
      }]
    }
    Image_FXParticles {
     id=#12
     possessed_class=Image
     tracks [
      {
       type=MovieSceneWidgetMaterialTrack
       sections [{type=MovieSceneParameterSection, channels [FlareFade {type=MovieSceneScriptingFloatChannel, keys [[3.8367,0],[6.5167,1],[27.953,0]]}]}]
      }
      {
       type=MovieSceneFloatTrack
       display_name=Render Opacity
       sections [{type=MovieSceneFloatSection, channels [None {type=MovieSceneScriptingFloatChannel, keys [[3.8367,0],[5.491,1],[24.3223,1],[44.4313,0]]}]}]
      }
     ]
    }
    Image_spike_RightTop {
     id=#13
     possessed_class=Image
     tracks [{
       type=MovieScene2DTransformTrack
       sections [{
         type=MovieScene2DTransformSection
         channels [
          Scale.X {type=MovieSceneScriptingFloatChannel, keys [[10.8543,0],[14.2563,0.32],[16.0267,0.55]]}
          Scale.Y {type=MovieSceneScriptingFloatChannel, keys [[10.8543,0],[14.2563,0.7728],[16.0267,0.55]]}
         ]
        }]
      }]
    }
   ]}
 }
 Socket {
  binding_count=15
  length_seconds=1.8694
  object=&self:Socket
  sequence {
   bindings [
    Image_spike_LeftBottom {
     id=#14
     possessed_class=Image
     tracks [{
       type=MovieScene2DTransformTrack
       sections [{
         type=MovieScene2DTransformSection
         channels [
          Scale.X {type=MovieSceneScriptingFloatChannel, keys [[3.477,0.55],[5.2473,0.32],[8.6493,0]]}
          Scale.Y {type=MovieSceneScriptingFloatChannel, keys [[3.477,0.55],[5.2473,0.7728],[8.6493,0]]}
         ]
        }]
      }]
    }
    Image_Right_Wing1 {
     id=#15
     possessed_class=Image
     tracks [
      {
       type=MovieScene2DTransformTrack
       sections [{
         type=MovieScene2DTransformSection
         channels [
          Translation.X {type=MovieSceneScriptingFloatChannel}
          Translation.Y {type=MovieSceneScriptingFloatChannel}
          Angle {type=MovieSceneScriptingFloatChannel}
          Scale.X {type=MovieSceneScriptingFloatChannel, keys [[2.575,1],[8.9087,0]]}
          Scale.Y {type=MovieSceneScriptingFloatChannel, keys [[2.575,1],[8.9087,0]]}
          Shear.X {type=MovieSceneScriptingFloatChannel}
          Shear.Y {type=MovieSceneScriptingFloatChannel}
         ]
        }]
      }
      {
       type=MovieSceneFloatTrack
       display_name=Render Opacity
       sections [{type=MovieSceneFloatSection, channels [None {type=MovieSceneScriptingFloatChannel, keys [[5.3987,1],[8.9087,0]]}]}]
      }
     ]
    }
    Image_spike_RightTop {
     id=#16
     possessed_class=Image
     tracks [{
       type=MovieScene2DTransformTrack
       sections [{
         type=MovieScene2DTransformSection
         channels [
          Scale.X {type=MovieSceneScriptingFloatChannel, keys [[3.477,0.55],[5.2473,0.32],[8.6493,0]]}
          Scale.Y {type=MovieSceneScriptingFloatChannel, keys [[3.477,0.55],[5.2473,0.7728],[8.6493,0]]}
         ]
        }]
      }]
    }
    Image_AnimatedCorona {
     id=#17
     possessed_class=Image
     tracks [
      {
       type=MovieSceneFloatTrack
       display_name=Render Opacity
       sections [{type=MovieSceneFloatSection, channels [None {type=MovieSceneScriptingFloatChannel, keys [[12.2013,0],[13.6877,0.5],[25.4103,0]]}]}]
      }
      {
       type=MovieScene2DTransformTrack
       sections [{
         type=MovieScene2DTransformSection
         channels [
          Scale.X {type=MovieSceneScriptingFloatChannel, keys [[12.2013,0.8],[27.858,1]]}
          Scale.Y {type=MovieSceneScriptingFloatChannel, keys [[12.2013,0.8],[27.858,1]]}
         ]
        }]
      }
     ]
    }
    Overlay_Medal {
     id=#18
     possessed_class=Overlay
     tracks [{
       type=MovieScene2DTransformTrack
       sections [{
         type=MovieScene2DTransformSection
         channels [
          Scale.X {type=MovieSceneScriptingFloatChannel, keys [[6.1487,1],[9.4067,1.0957],[10.6617,0.8328],[11.6393,0.9491],[13.5083,0.8]]}
          Scale.Y {type=MovieSceneScriptingFloatChannel, keys [[6.1487,1],[9.4067,1.0957],[10.6617,0.8328],[11.6393,0.9491],[13.5083,0.8]]}
         ]
        }]
      }]
    }
    Image_Right_Wing3 {
     id=#19
     possessed_class=Image
     tracks [
      {
       type=MovieScene2DTransformTrack
       sections [{
         type=MovieScene2DTransformSection
         channels [
          Translation.X {type=MovieSceneScriptingFloatChannel}
          Translation.Y {type=MovieSceneScriptingFloatChannel}
          Angle {type=MovieSceneScriptingFloatChannel}
          Scale.X {type=MovieSceneScriptingFloatChannel, keys [[5.7117,1],[12.0453,0]]}
          Scale.Y {type=MovieSceneScriptingFloatChannel, keys [[5.7117,1],[12.0453,0]]}
          Shear.X {type=MovieSceneScriptingFloatChannel}
          Shear.Y {type=MovieSceneScriptingFloatChannel}
         ]
        }]
      }
      {
       type=MovieSceneFloatTrack
       display_name=Render Opacity
       sections [{type=MovieSceneFloatSection, channels [None {type=MovieSceneScriptingFloatChannel, keys [[8.5353,1],[12.0453,0]]}]}]
      }
     ]
    }
    Image_largeglow {
     id=#20
     possessed_class=Image
     tracks [{
       type=MovieSceneFloatTrack
       display_name=Render Opacity
       sections [{type=MovieSceneFloatSection, channels [None {type=MovieSceneScriptingFloatChannel, keys [[10.7497,0],[12.236,1],[37.3887,0]]}]}]
      }]
    }
    Image_spike_RightBottom {
     id=#21
     possessed_class=Image
     tracks [{
       type=MovieScene2DTransformTrack
       sections [{
         type=MovieScene2DTransformSection
         channels [
          Scale.X {type=MovieSceneScriptingFloatChannel, keys [[3.477,0.55],[5.2473,0.32],[8.6493,0]]}
          Scale.Y {type=MovieSceneScriptingFloatChannel, keys [[3.477,0.55],[5.2473,0.7728],[8.6493,0]]}
         ]
        }]
      }]
    }
    Image_spike_LeftTop {
     id=#22
     possessed_class=Image
     tracks [{
       type=MovieScene2DTransformTrack
       sections [{
         type=MovieScene2DTransformSection
         channels [
          Scale.X {type=MovieSceneScriptingFloatChannel, keys [[3.477,0.55],[5.2473,0.32],[8.6493,0]]}
          Scale.Y {type=MovieSceneScriptingFloatChannel, keys [[3.477,0.55],[5.2473,0.7728],[8.6493,0]]}
         ]
        }]
      }]
    }
    Image_Left_Wing3 {
     id=#23
     possessed_class=Image
     tracks [
      {
       type=MovieScene2DTransformTrack
       sections [{
         type=MovieScene2DTransformSection
         channels [
          Translation.X {type=MovieSceneScriptingFloatChannel}
          Translation.Y {type=MovieSceneScriptingFloatChannel}
          Angle {type=MovieSceneScriptingFloatChannel}
          Scale.X {type=MovieSceneScriptingFloatChannel, keys [[4.965,1],[11.2987,0]]}
          Scale.Y {type=MovieSceneScriptingFloatChannel, keys [[4.965,1],[11.2987,0]]}
          Shear.X {type=MovieSceneScriptingFloatChannel}
          Shear.Y {type=MovieSceneScriptingFloatChannel}
         ]
        }]
      }
      {
       type=MovieSceneFloatTrack
       display_name=Render Opacity
       sections [{type=MovieSceneFloatSection, channels [None {type=MovieSceneScriptingFloatChannel, keys [[7.7887,1],[11.2987,0]]}]}]
      }
     ]
    }
    Image_HaloCorona {
     id=#24
     possessed_class=Image
     tracks [
      {
       type=MovieSceneFloatTrack
       display_name=Render Opacity
       sections [{type=MovieSceneFloatSection, channels [None {type=MovieSceneScriptingFloatChannel, keys [[10.7497,0],[12.236,1],[31.4907,0]]}]}]
      }
      {
       type=MovieScene2DTransformTrack
       sections [{
         type=MovieScene2DTransformSection
         channels [
          Scale.X {type=MovieSceneScriptingFloatChannel, keys [[10.7497,0.7498],[25.819,1.0418],[31.4907,1.0512]]}
          Scale.Y {type=MovieSceneScriptingFloatChannel, keys [[10.7497,0.7498],[25.819,1.0418],[31.4907,1.0512]]}
         ]
        }]
      }
     ]
    }
    Image_Right_Wing2 {
     id=#25
     possessed_class=Image
     tracks [
      {
       type=MovieScene2DTransformTrack
       sections [{
         type=MovieScene2DTransformSection
         channels [
          Translation.X {type=MovieSceneScriptingFloatChannel}
          Translation.Y {type=MovieSceneScriptingFloatChannel}
          Angle {type=MovieSceneScriptingFloatChannel}
          Scale.X {type=MovieSceneScriptingFloatChannel, keys [[4.3673,1],[10.701,0]]}
          Scale.Y {type=MovieSceneScriptingFloatChannel, keys [[4.3673,1],[10.701,0]]}
          Shear.X {type=MovieSceneScriptingFloatChannel}
          Shear.Y {type=MovieSceneScriptingFloatChannel}
         ]
        }]
      }
      {
       type=MovieSceneFloatTrack
       display_name=Render Opacity
       sections [{type=MovieSceneFloatSection, channels [None {type=MovieSceneScriptingFloatChannel, keys [[7.191,1],[10.701,0]]}]}]
      }
     ]
    }
    Image_Left_Wing2 {
     id=#26
     possessed_class=Image
     tracks [
      {
       type=MovieScene2DTransformTrack
       sections [{
         type=MovieScene2DTransformSection
         channels [
          Translation.X {type=MovieSceneScriptingFloatChannel}
          Translation.Y {type=MovieSceneScriptingFloatChannel}
          Angle {type=MovieSceneScriptingFloatChannel}
          Scale.X {type=MovieSceneScriptingFloatChannel, keys [[3.994,1],[10.3277,0]]}
          Scale.Y {type=MovieSceneScriptingFloatChannel, keys [[3.994,1],[10.3277,0]]}
          Shear.X {type=MovieSceneScriptingFloatChannel}
          Shear.Y {type=MovieSceneScriptingFloatChannel}
         ]
        }]
      }
      {
       type=MovieSceneFloatTrack
       display_name=Render Opacity
       sections [{type=MovieSceneFloatSection, channels [None {type=MovieSceneScriptingFloatChannel, keys [[6.8177,1],[10.3277,0]]}]}]
      }
     ]
    }
    Image_Left_Wing1 {
     id=#27
     possessed_class=Image
     tracks [
      {
       type=MovieScene2DTransformTrack
       sections [{
         type=MovieScene2DTransformSection
         channels [
          Translation.X {type=MovieSceneScriptingFloatChannel}
          Translation.Y {type=MovieSceneScriptingFloatChannel}
          Angle {type=MovieSceneScriptingFloatChannel}
          Scale.X {type=MovieSceneScriptingFloatChannel, keys [[2.575,1],[8.9087,0]]}
          Scale.Y {type=MovieSceneScriptingFloatChannel, keys [[2.575,1],[8.9087,0]]}
          Shear.X {type=MovieSceneScriptingFloatChannel}
          Shear.Y {type=MovieSceneScriptingFloatChannel}
         ]
        }]
      }
      {
       type=MovieSceneFloatTrack
       display_name=Render Opacity
       sections [{type=MovieSceneFloatSection, channels [None {type=MovieSceneScriptingFloatChannel, keys [[5.3987,1],[8.9087,0]]}]}]
      }
     ]
    }
    Overlay_TopContainer {
     id=#28
     possessed_class=Overlay
     tracks [
      {
       type=MovieScene2DTransformTrack
       sections [{
         type=MovieScene2DTransformSection
         channels [
          Scale.X {type=MovieSceneScriptingFloatChannel, keys [[0,5],[2.7307,2.3118],[9.229,2.3118],[11.3997,1],[12.6533,1.0654],[17.6873,1]]}
          Scale.Y {type=MovieSceneScriptingFloatChannel, keys [[0,5],[2.7307,2.3118],[9.229,2.3118],[11.3997,1],[12.6533,1.0654],[17.6873,1]]}
         ]
        }]
      }
      {
       type=MovieSceneFloatTrack
       display_name=Render Opacity
       sections [{type=MovieSceneFloatSection, channels [None {type=MovieSceneScriptingFloatChannel, keys [[0,0],[2.7307,1]]}]}]
      }
     ]
    }
   ]
   tracks [{
     type=MovieSceneEventTrack
     sections [{
       type=MovieSceneEventTriggerSection
       channels [None {type=MovieSceneScriptingEventChannel, const ["<Struct 'MovieSceneEvent' (0x467aac3ee20) {}>", 1]}]
      }]
    }]
  }
 }
 Socket_VisualState {
  binding_count=15
  length_seconds=1.8694
  object=&self:Socket_VisualState
  sequence {
   bindings [
    Image_spike_LeftBottom {
     id=#14
     possessed_class=Image
     tracks [{
       type=MovieScene2DTransformTrack
       sections [{
         type=MovieScene2DTransformSection
         channels [
          Scale.X {type=MovieSceneScriptingFloatChannel, keys [[3.477,0.55],[5.2473,0.32],[8.6493,0]]}
          Scale.Y {type=MovieSceneScriptingFloatChannel, keys [[3.477,0.55],[5.2473,0.7728],[8.6493,0]]}
         ]
        }]
      }]
    }
    Image_Right_Wing1 {
     id=#15
     possessed_class=Image
     tracks [
      {
       type=MovieScene2DTransformTrack
       sections [{
         type=MovieScene2DTransformSection
         channels [
          Translation.X {type=MovieSceneScriptingFloatChannel}
          Translation.Y {type=MovieSceneScriptingFloatChannel}
          Angle {type=MovieSceneScriptingFloatChannel}
          Scale.X {type=MovieSceneScriptingFloatChannel, keys [[2.575,1],[8.9087,0]]}
          Scale.Y {type=MovieSceneScriptingFloatChannel, keys [[2.575,1],[8.9087,0]]}
          Shear.X {type=MovieSceneScriptingFloatChannel}
          Shear.Y {type=MovieSceneScriptingFloatChannel}
         ]
        }]
      }
      {
       type=MovieSceneFloatTrack
       display_name=Render Opacity
       sections [{type=MovieSceneFloatSection, channels [None {type=MovieSceneScriptingFloatChannel, keys [[5.3987,1],[8.9087,0]]}]}]
      }
     ]
    }
    Image_spike_RightTop {
     id=#16
     possessed_class=Image
     tracks [{
       type=MovieScene2DTransformTrack
       sections [{
         type=MovieScene2DTransformSection
         channels [
          Scale.X {type=MovieSceneScriptingFloatChannel, keys [[3.477,0.55],[5.2473,0.32],[8.6493,0]]}
          Scale.Y {type=MovieSceneScriptingFloatChannel, keys [[3.477,0.55],[5.2473,0.7728],[8.6493,0]]}
         ]
        }]
      }]
    }
    Image_AnimatedCorona {
     id=#17
     possessed_class=Image
     tracks [
      {
       type=MovieSceneFloatTrack
       display_name=Render Opacity
       sections [{type=MovieSceneFloatSection, channels [None {type=MovieSceneScriptingFloatChannel, keys [[12.2013,0],[13.6877,0.5],[25.4103,0]]}]}]
      }
      {
       type=MovieScene2DTransformTrack
       sections [{
         type=MovieScene2DTransformSection
         channels [
          Scale.X {type=MovieSceneScriptingFloatChannel, keys [[12.2013,0.8],[27.858,1]]}
          Scale.Y {type=MovieSceneScriptingFloatChannel, keys [[12.2013,0.8],[27.858,1]]}
         ]
        }]
      }
     ]
    }
    Overlay_Medal {
     id=#18
     possessed_class=Overlay
     tracks [{
       type=MovieScene2DTransformTrack
       sections [{
         type=MovieScene2DTransformSection
         channels [
          Scale.X {type=MovieSceneScriptingFloatChannel, keys [[6.1487,1],[9.4067,1.0957],[10.6617,0.8328],[11.6393,0.9491],[13.5083,0.8]]}
          Scale.Y {type=MovieSceneScriptingFloatChannel, keys [[6.1487,1],[9.4067,1.0957],[10.6617,0.8328],[11.6393,0.9491],[13.5083,0.8]]}
         ]
        }]
      }]
    }
    Image_Right_Wing3 {
     id=#19
     possessed_class=Image
     tracks [
      {
       type=MovieScene2DTransformTrack
       sections [{
         type=MovieScene2DTransformSection
         channels [
          Translation.X {type=MovieSceneScriptingFloatChannel}
          Translation.Y {type=MovieSceneScriptingFloatChannel}
          Angle {type=MovieSceneScriptingFloatChannel}
          Scale.X {type=MovieSceneScriptingFloatChannel, keys [[5.7117,1],[12.0453,0]]}
          Scale.Y {type=MovieSceneScriptingFloatChannel, keys [[5.7117,1],[12.0453,0]]}
          Shear.X {type=MovieSceneScriptingFloatChannel}
          Shear.Y {type=MovieSceneScriptingFloatChannel}
         ]
        }]
      }
      {
       type=MovieSceneFloatTrack
       display_name=Render Opacity
       sections [{type=MovieSceneFloatSection, channels [None {type=MovieSceneScriptingFloatChannel, keys [[8.5353,1],[12.0453,0]]}]}]
      }
     ]
    }
    Image_largeglow {
     id=#20
     possessed_class=Image
     tracks [{
       type=MovieSceneFloatTrack
       display_name=Render Opacity
       sections [{type=MovieSceneFloatSection, channels [None {type=MovieSceneScriptingFloatChannel, keys [[10.7497,0],[12.236,1],[37.3887,0]]}]}]
      }]
    }
    Image_spike_RightBottom {
     id=#21
     possessed_class=Image
     tracks [{
       type=MovieScene2DTransformTrack
       sections [{
         type=MovieScene2DTransformSection
         channels [
          Scale.X {type=MovieSceneScriptingFloatChannel, keys [[3.477,0.55],[5.2473,0.32],[8.6493,0]]}
          Scale.Y {type=MovieSceneScriptingFloatChannel, keys [[3.477,0.55],[5.2473,0.7728],[8.6493,0]]}
         ]
        }]
      }]
    }
    Image_spike_LeftTop {
     id=#22
     possessed_class=Image
     tracks [{
       type=MovieScene2DTransformTrack
       sections [{
         type=MovieScene2DTransformSection
         channels [
          Scale.X {type=MovieSceneScriptingFloatChannel, keys [[3.477,0.55],[5.2473,0.32],[8.6493,0]]}
          Scale.Y {type=MovieSceneScriptingFloatChannel, keys [[3.477,0.55],[5.2473,0.7728],[8.6493,0]]}
         ]
        }]
      }]
    }
    Image_Left_Wing3 {
     id=#23
     possessed_class=Image
     tracks [
      {
       type=MovieScene2DTransformTrack
       sections [{
         type=MovieScene2DTransformSection
         channels [
          Translation.X {type=MovieSceneScriptingFloatChannel}
          Translation.Y {type=MovieSceneScriptingFloatChannel}
          Angle {type=MovieSceneScriptingFloatChannel}
          Scale.X {type=MovieSceneScriptingFloatChannel, keys [[4.965,1],[11.2987,0]]}
          Scale.Y {type=MovieSceneScriptingFloatChannel, keys [[4.965,1],[11.2987,0]]}
          Shear.X {type=MovieSceneScriptingFloatChannel}
          Shear.Y {type=MovieSceneScriptingFloatChannel}
         ]
        }]
      }
      {
       type=MovieSceneFloatTrack
       display_name=Render Opacity
       sections [{type=MovieSceneFloatSection, channels [None {type=MovieSceneScriptingFloatChannel, keys [[7.7887,1],[11.2987,0]]}]}]
      }
     ]
    }
    Image_HaloCorona {
     id=#24
     possessed_class=Image
     tracks [
      {
       type=MovieSceneFloatTrack
       display_name=Render Opacity
       sections [{type=MovieSceneFloatSection, channels [None {type=MovieSceneScriptingFloatChannel, keys [[10.7497,0],[12.236,1],[31.4907,0]]}]}]
      }
      {
       type=MovieScene2DTransformTrack
       sections [{
         type=MovieScene2DTransformSection
         channels [
          Scale.X {type=MovieSceneScriptingFloatChannel, keys [[10.7497,0.7498],[25.819,1.0418],[31.4907,1.0512]]}
          Scale.Y {type=MovieSceneScriptingFloatChannel, keys [[10.7497,0.7498],[25.819,1.0418],[31.4907,1.0512]]}
         ]
        }]
      }
     ]
    }
    Image_Right_Wing2 {
     id=#25
     possessed_class=Image
     tracks [
      {
       type=MovieScene2DTransformTrack
       sections [{
         type=MovieScene2DTransformSection
         channels [
          Translation.X {type=MovieSceneScriptingFloatChannel}
          Translation.Y {type=MovieSceneScriptingFloatChannel}
          Angle {type=MovieSceneScriptingFloatChannel}
          Scale.X {type=MovieSceneScriptingFloatChannel, keys [[4.3673,1],[10.701,0]]}
          Scale.Y {type=MovieSceneScriptingFloatChannel, keys [[4.3673,1],[10.701,0]]}
          Shear.X {type=MovieSceneScriptingFloatChannel}
          Shear.Y {type=MovieSceneScriptingFloatChannel}
         ]
        }]
      }
      {
       type=MovieSceneFloatTrack
       display_name=Render Opacity
       sections [{type=MovieSceneFloatSection, channels [None {type=MovieSceneScriptingFloatChannel, keys [[7.191,1],[10.701,0]]}]}]
      }
     ]
    }
    Image_Left_Wing2 {
     id=#26
     possessed_class=Image
     tracks [
      {
       type=MovieScene2DTransformTrack
       sections [{
         type=MovieScene2DTransformSection
         channels [
          Translation.X {type=MovieSceneScriptingFloatChannel}
          Translation.Y {type=MovieSceneScriptingFloatChannel}
          Angle {type=MovieSceneScriptingFloatChannel}
          Scale.X {type=MovieSceneScriptingFloatChannel, keys [[3.994,1],[10.3277,0]]}
          Scale.Y {type=MovieSceneScriptingFloatChannel, keys [[3.994,1],[10.3277,0]]}
          Shear.X {type=MovieSceneScriptingFloatChannel}
          Shear.Y {type=MovieSceneScriptingFloatChannel}
         ]
        }]
      }
      {
       type=MovieSceneFloatTrack
       display_name=Render Opacity
       sections [{type=MovieSceneFloatSection, channels [None {type=MovieSceneScriptingFloatChannel, keys [[6.8177,1],[10.3277,0]]}]}]
      }
     ]
    }
    Image_Left_Wing1 {
     id=#27
     possessed_class=Image
     tracks [
      {
       type=MovieScene2DTransformTrack
       sections [{
         type=MovieScene2DTransformSection
         channels [
          Translation.X {type=MovieSceneScriptingFloatChannel}
          Translation.Y {type=MovieSceneScriptingFloatChannel}
          Angle {type=MovieSceneScriptingFloatChannel}
          Scale.X {type=MovieSceneScriptingFloatChannel, keys [[2.575,1],[8.9087,0]]}
          Scale.Y {type=MovieSceneScriptingFloatChannel, keys [[2.575,1],[8.9087,0]]}
          Shear.X {type=MovieSceneScriptingFloatChannel}
          Shear.Y {type=MovieSceneScriptingFloatChannel}
         ]
        }]
      }
      {
       type=MovieSceneFloatTrack
       display_name=Render Opacity
       sections [{type=MovieSceneFloatSection, channels [None {type=MovieSceneScriptingFloatChannel, keys [[5.3987,1],[8.9087,0]]}]}]
      }
     ]
    }
    Overlay_TopContainer {
     id=#28
     possessed_class=Overlay
     tracks [
      {
       type=MovieScene2DTransformTrack
       sections [{
         type=MovieScene2DTransformSection
         channels [
          Scale.X {type=MovieSceneScriptingFloatChannel, keys [[0,5],[2.7307,2.3118],[9.229,2.3118],[11.3997,1],[12.6533,1.0654],[17.6873,1]]}
          Scale.Y {type=MovieSceneScriptingFloatChannel, keys [[0,5],[2.7307,2.3118],[9.229,2.3118],[11.3997,1],[12.6533,1.0654],[17.6873,1]]}
         ]
        }]
      }
      {
       type=MovieSceneFloatTrack
       display_name=Render Opacity
       sections [{type=MovieSceneFloatSection, channels [None {type=MovieSceneScriptingFloatChannel, keys [[0,0],[2.7307,1]]}]}]
      }
     ]
    }
   ]
   tracks [{
     type=MovieSceneEventTrack
     sections [{
       type=MovieSceneEventTriggerSection
       channels [None {type=MovieSceneScriptingEventChannel, const ["<Struct 'MovieSceneEvent' (0x467aada1a40) {}>", 1]}]
      }]
    }]
  }
 }
]
graphs [
 EventGraph {nodes [
   MedalInit() {is=title K2Node_CustomEvent}
   n1 = Get MedalLevel() -> (MedalLevel:int=0)
   n2 = "integer > integer"(A=@n1, B=1) {is=title K2Node_CallFunction}
   > SetVisibility(
    self=[Get Image_FXParticles() -> (Image_FXParticles:Image Object Reference), Get Overlay_Spikes() -> (Overlay_Spikes:Overlay Object Reference)]
    InVisibility=Select(Option 0=Hidden, Option 1=HitTestInvisible, Index=@n2)
   )
   > SetRenderScale!(
    execute
    self:Widget Object Reference=Get Image_PulseRingImage() -> (Image_PulseRingImage:Image Object Reference)
    Scale:Vector 2D Structure=Select(Option 0:Vector 2D Structure=(X=0.5,Y=0.5), Option 1:Vector 2D Structure=(X=1,Y=1), Index="integer > integer"(A=@n1, B=2)) -> (ReturnValue:Vector 2D Structure)
   ) -> (then)
   > PlayAnimation(InAnimation=Get Intro())
   > Branch(Condition=@n2)
   > PlayAnimation(InAnimation=Get Upgrade() -> (Upgrade:Widget Animation Object Reference))
   TriggerParticles() {is=title K2Node_CustomEvent}
   > n3 = Branch(Condition=Get ShowUIParticles() -> (ShowUIParticles:bool))
   > n4 = StartEmitter(self:UIEmitter Object Reference=Get UIEmitter() -> (UIEmitter:UIEmitter Object Reference))
   Call Socket Smash!(execute:Exec=[@n4.then, @n3.else], self:Accolade Image Widget Object Reference) -> (then:Exec) {is=title K2Node_CallDelegate}
   n5 = EventSocket!() -> (OutputDelegate, then, bAlreadySocketed:bool=~) {is=title K2Node_CustomEvent}
   n6 = "integer < integer"!(A:int=Get MedalLevel() -> (MedalLevel=0), B:int=2) -> (ReturnValue:bool) {is=title K2Node_CallFunction}
   > PlayAnimation(
    InAnimation=Select(
     Option 0:Widget Animation Object Reference=Get Upgrade()
     Option 1:Widget Animation Object Reference=Get Socket() -> (Socket:Widget Animation Object Reference)
     Index=@n6
    ) -> (ReturnValue:Widget Animation Object Reference)
    StartAtTime=Select(Option 0:double=0, Option 1:double=1, Index=@n5.bAlreadySocketed) -> (ReturnValue:double)
   )
   > n7 = Branch(Condition=@n6)
   n8 = Branch(execute=[@n11.then, @n7.then], Condition=@n5.bAlreadySocketed)
   n9 = Get MedalLevel() -> (MedalLevel=0)
   PlaySound2D(
    execute=@n8.else
    Sound=Select(
     Option 0:Sound Base Object Reference=Get OnUpgradeSound() -> (OnUpgradeSound:Sound Base Object Reference)
     Option 1:Sound Base Object Reference=Get OnSocketSound() -> (OnSocketSound:Sound Base Object Reference)
     Index="integer < integer"(A=@n9, B=2)
    ) -> (ReturnValue:Sound Base Object Reference)
   )
   n10 = Get Socket_VisualState() -> (Socket_VisualState:Widget Animation Object Reference)
   n11 = PlayAnimation(execute=@n7.else, InAnimation=@n10, StartAtTime=GetEndTime(self=@n10))
   Make Small() {is=title K2Node_CustomEvent}
   > SetVisibility(
    self=[
     Get Image_PulseRingImage()
     Get Image_largeglow() -> (Image_largeglow:Image Object Reference)
     Get Image_AnimatedCorona() -> (Image_AnimatedCorona:Image Object Reference)
     Get Image_HaloCorona() -> (Image_HaloCorona:Image Object Reference)
     Get CommonBorder_WingsLeft() -> (CommonBorder_WingsLeft:Common Border Object Reference)
     Get CommonBorder_WingsRight() -> (CommonBorder_WingsRight:Common Border Object Reference)
     Get Overlay_Spikes()
    ]
    InVisibility=Collapsed
   )
   EventSetSocketed() {is=title K2Node_CustomEvent}
   > SetVisibility(self=[Get CommonBorder_WingsLeft(), Get CommonBorder_WingsRight(), Get Overlay_Spikes()], InVisibility=Hidden)
   > SetRenderScale(self=Get Overlay_Medal() -> (Overlay_Medal:Overlay Object Reference), Scale=(X=0.8,Y=0.8))
   n12 = Event PreConstruct()
   > Branch(Condition=AND Boolean(A=@n12.IsDesignTime, B=Get ShowSocketed() -> (ShowSocketed:bool)))
   > EventSetSocketed() {is=title K2Node_CallFunction}
  ]}
 SequenceEvent_0 {nodes [
   SequenceEvent_0() {is=title K2Node_FunctionEntry}
   > TriggerParticles() {is=title K2Node_CallFunction}
  ]}
 SocketSmash {nodes [SocketSmash() {is=title K2Node_FunctionEntry}]}
]
variables [
 MedalLevel {type=integer}
 OnSocketSound {type=SoundBase, default=/Game/Sounds/Fort_UI_Sounds/Gameplay/XP/Fort_XP_Medal_Socket_Cue}
 OnUpgradeSound {type=SoundBase, default=/Game/Sounds/Fort_UI_Sounds/Gameplay/XP/Fort_XP_Medal_Socket_Upgrade_Cue}
 ShowSocketed {type=boolean}
 ShowUIParticles {type=boolean}
 SocketSmash {}
]
widgets {
 info {inheritedWidgetCount=0, parentClass=&UserWidget, rootWidgetClass=&Overlay, widgetCount=25}
 widgets [
  Overlay_TopContainer {widgetClassPath=&Overlay, widget=Overlay_TopContainer, properties {renderTransformPivot=<0 0.5>}}
  Image_PulseRingImage {
   widgetClassPath=&Image
   bIsVariable=true
   parent=Overlay_TopContainer
   widget=Image_PulseRingImage
   properties {
    renderOpacity=0
    visibility=SelfHitTestInvisible
    brush {
     bIsDynamicallyLoaded=false
     drawAs=Image
     imageType=NoImage
     mirroring=NoMirror
     resourceName=None
     resourceObject=/Game/Athena/Items/Accolades/Materials/M_UI_PulseRing_Anime_Inst
     tiling=NoTile
     imageSize=<512 512>
     margin=<ltrb 0 0 0 0>
     outlineSettings {
      bUseBrushTransparency=false
      roundingType=HalfHeightRadius
      width=0
      color {colorUseRule=UseColor_Specified, specifiedColor=<rgba 0 0 0 0>}
      cornerRadii=<0 0 0 0>
     }
     tintColor {colorUseRule=UseColor_Specified, specifiedColor=<rgba 1 1 1 1>}
     uVRegion {bIsValid=false, max=<0 0>, min=<0 0>}
    }
   }
   slot {class=OverlaySlot, properties {horizontalAlignment=HAlign_Center, verticalAlignment=VAlign_Center, padding=<ltrb -256 -256 -256 -256>}}
  }
  CommonBorder_WingsLeft {
   widgetClassPath=&CommonBorder
   bIsVariable=true
   parent=Overlay_TopContainer
   widget=CommonBorder_WingsLeft
   properties {padding=<ltrb 0 0 0 0>}
   slot {class=OverlaySlot, properties {horizontalAlignment=HAlign_Center, verticalAlignment=VAlign_Center}}
  }
  Overlay_1 {
   widgetClassPath=&Overlay
   parent=CommonBorder_WingsLeft
   widget=Overlay_1
   properties {renderTransform {angle=0, scale=<0.55 0.55>, shear=<0 0>, translation=<-57 -8>}}
   slot {class=BorderSlot, properties {padding=<ltrb 0 0 0 0>}}
  }
  Image_Left_Wing1 {
   widgetClassPath=&Image
   bIsVariable=true
   parent=Overlay_1
   widget=Image_Left_Wing1
   properties {
    brush {
     bIsDynamicallyLoaded=false
     drawAs=Image
     imageType=NoImage
     mirroring=NoMirror
     resourceName=None
     resourceObject=/Game/Athena/Items/Accolades/Materials/M_UI_XPMedal_Base_Wing1
     tiling=NoTile
     imageSize=<68 43>
     margin=<ltrb 0 0 0 0>
     outlineSettings {
      bUseBrushTransparency=false
      roundingType=HalfHeightRadius
      width=0
      color {colorUseRule=UseColor_Specified, specifiedColor=<rgba 0 0 0 0>}
      cornerRadii=<0 0 0 0>
     }
     tintColor {colorUseRule=UseColor_Specified, specifiedColor=<rgba 1 1 1 1>}
     uVRegion {bIsValid=false, max=<0 0>, min=<0 0>}
    }
    renderTransform {angle=0, scale=<1 1>, shear=<0 0>, translation=<-1 0>}
    renderTransformPivot=<1 1>
   }
   slot {class=OverlaySlot, properties {horizontalAlignment=HAlign_Right}}
  }
  Image_Left_Wing2 {
   widgetClassPath=&Image
   bIsVariable=true
   parent=Overlay_1
   widget=Image_Left_Wing2
   properties {
    brush {
     bIsDynamicallyLoaded=false
     drawAs=Image
     imageType=NoImage
     mirroring=NoMirror
     resourceName=None
     resourceObject=/Game/Athena/Items/Accolades/Materials/M_UI_XPMedal_Base_Wing2
     tiling=NoTile
     imageSize=<64 36>
     margin=<ltrb 0 0 0 0>
     outlineSettings {
      bUseBrushTransparency=false
      roundingType=HalfHeightRadius
      width=0
      color {colorUseRule=UseColor_Specified, specifiedColor=<rgba 0 0 0 0>}
      cornerRadii=<0 0 0 0>
     }
     tintColor {colorUseRule=UseColor_Specified, specifiedColor=<rgba 1 1 1 1>}
     uVRegion {bIsValid=false, max=<0 0>, min=<0 0>}
    }
    renderTransform {angle=0, scale=<1 1>, shear=<0 0>, translation=<7 29>}
    renderTransformPivot=<0.9415 0.5>
   }
   slot {class=OverlaySlot, properties {horizontalAlignment=HAlign_Right}}
  }
  Image_Left_Wing3 {
   widgetClassPath=&Image
   bIsVariable=true
   parent=Overlay_1
   widget=Image_Left_Wing3
   properties {
    brush {
     bIsDynamicallyLoaded=false
     drawAs=Image
     imageType=NoImage
     mirroring=NoMirror
     resourceName=None
     resourceObject=/Game/Athena/Items/Accolades/Materials/M_UI_XPMedal_Base_Wing3
     tiling=NoTile
     imageSize=<64 32>
     margin=<ltrb 0 0 0 0>
     outlineSettings {
      bUseBrushTransparency=false
      roundingType=HalfHeightRadius
      width=0
      color {colorUseRule=UseColor_Specified, specifiedColor=<rgba 0 0 0 0>}
      cornerRadii=<0 0 0 0>
     }
     tintColor {colorUseRule=UseColor_Specified, specifiedColor=<rgba 1 1 1 1>}
     uVRegion {bIsValid=false, max=<0 0>, min=<0 0>}
    }
    renderTransform {angle=0, scale=<1 1>, shear=<0 0>, translation=<14 53>}
    renderTransformPivot=<0.8671 0>
   }
   slot {class=OverlaySlot, properties {horizontalAlignment=HAlign_Right}}
  }
  CommonBorder_WingsRight {
   widgetClassPath=&CommonBorder
   bIsVariable=true
   parent=Overlay_TopContainer
   widget=CommonBorder_WingsRight
   properties {padding=<ltrb 0 0 0 0>, renderTransform {angle=0, scale=<-1 1>, shear=<0 0>, translation=<0 0>}}
   slot {class=OverlaySlot, properties {horizontalAlignment=HAlign_Center, verticalAlignment=VAlign_Center}}
  }
  Overlay {
   widgetClassPath=&Overlay
   parent=CommonBorder_WingsRight
   widget=Overlay
   properties {renderTransform {angle=0, scale=<0.55 0.55>, shear=<0 0>, translation=<-57 -8>}}
   slot {class=BorderSlot, properties {padding=<ltrb 0 0 0 0>}}
  }
  Image_Right_Wing1 {
   widgetClassPath=&Image
   bIsVariable=true
   parent=Overlay
   widget=Image_Right_Wing1
   properties {
    brush {
     bIsDynamicallyLoaded=false
     drawAs=Image
     imageType=NoImage
     mirroring=NoMirror
     resourceName=None
     resourceObject=&M_UI_XPMedal_Base_Wing1
     tiling=NoTile
     imageSize=<68 43>
     margin=<ltrb 0 0 0 0>
     outlineSettings {
      bUseBrushTransparency=false
      roundingType=HalfHeightRadius
      width=0
      color {colorUseRule=UseColor_Specified, specifiedColor=<rgba 0 0 0 0>}
      cornerRadii=<0 0 0 0>
     }
     tintColor {colorUseRule=UseColor_Specified, specifiedColor=<rgba 1 1 1 1>}
     uVRegion {bIsValid=false, max=<0 0>, min=<0 0>}
    }
    renderTransform {angle=0, scale=<1 1>, shear=<0 0>, translation=<-1 0>}
    renderTransformPivot=<1 1>
   }
   slot {class=OverlaySlot, properties {horizontalAlignment=HAlign_Right}}
  }
  Image_Right_Wing2 {
   widgetClassPath=&Image
   bIsVariable=true
   parent=Overlay
   widget=Image_Right_Wing2
   properties {
    brush {
     bIsDynamicallyLoaded=false
     drawAs=Image
     imageType=NoImage
     mirroring=NoMirror
     resourceName=None
     resourceObject=&M_UI_XPMedal_Base_Wing2
     tiling=NoTile
     imageSize=<64 36>
     margin=<ltrb 0 0 0 0>
     outlineSettings {
      bUseBrushTransparency=false
      roundingType=HalfHeightRadius
      width=0
      color {colorUseRule=UseColor_Specified, specifiedColor=<rgba 0 0 0 0>}
      cornerRadii=<0 0 0 0>
     }
     tintColor {colorUseRule=UseColor_Specified, specifiedColor=<rgba 1 1 1 1>}
     uVRegion {bIsValid=false, max=<0 0>, min=<0 0>}
    }
    renderTransform {angle=0, scale=<1 1>, shear=<0 0>, translation=<7 29>}
    renderTransformPivot=<0.9415 0.5>
   }
   slot {class=OverlaySlot, properties {horizontalAlignment=HAlign_Right}}
  }
  Image_Right_Wing3 {
   widgetClassPath=&Image
   bIsVariable=true
   parent=Overlay
   widget=Image_Right_Wing3
   properties {
    brush {
     bIsDynamicallyLoaded=false
     drawAs=Image
     imageType=NoImage
     mirroring=NoMirror
     resourceName=None
     resourceObject=&M_UI_XPMedal_Base_Wing3
     tiling=NoTile
     imageSize=<64 32>
     margin=<ltrb 0 0 0 0>
     outlineSettings {
      bUseBrushTransparency=false
      roundingType=HalfHeightRadius
      width=0
      color {colorUseRule=UseColor_Specified, specifiedColor=<rgba 0 0 0 0>}
      cornerRadii=<0 0 0 0>
     }
     tintColor {colorUseRule=UseColor_Specified, specifiedColor=<rgba 1 1 1 1>}
     uVRegion {bIsValid=false, max=<0 0>, min=<0 0>}
    }
    renderTransform {angle=0, scale=<1 1>, shear=<0 0>, translation=<14 53>}
    renderTransformPivot=<0.8671 0>
   }
   slot {class=OverlaySlot, properties {horizontalAlignment=HAlign_Right}}
  }
  Overlay_Medal {
   widgetClassPath=&Overlay
   bIsVariable=true
   parent=Overlay_TopContainer
   widget=Overlay_Medal
   slot {class=OverlaySlot, properties {horizontalAlignment=HAlign_Center, verticalAlignment=VAlign_Center}}
  }
  Overlay_Spikes {
   widgetClassPath=&Overlay
   bIsVariable=true
   parent=Overlay_Medal
   widget=Overlay_Spikes
   slot {class=OverlaySlot, properties {horizontalAlignment=HAlign_Fill, verticalAlignment=VAlign_Fill}}
  }
  Image_spike_LeftBottom {
   widgetClassPath=&Image
   bIsVariable=true
   parent=Overlay_Spikes
   widget=Image_spike_LeftBottom
   properties {
    brush {
     bIsDynamicallyLoaded=false
     drawAs=Image
     imageType=NoImage
     mirroring=NoMirror
     resourceName=None
     resourceObject=/Game/Athena/Items/Accolades/Materials/Textures/T_UI_Medal_Spike_LeftBottom
     tiling=NoTile
     imageSize=<32 32>
     margin=<ltrb 0 0 0 0>
     outlineSettings {
      bUseBrushTransparency=false
      roundingType=HalfHeightRadius
      width=0
      color {colorUseRule=UseColor_Specified, specifiedColor=<rgba 0 0 0 0>}
      cornerRadii=<0 0 0 0>
     }
     tintColor {colorUseRule=UseColor_Specified, specifiedColor=<rgba 1 1 1 1>}
     uVRegion {bIsValid=false, max=<0 0>, min=<0 0>}
    }
    renderTransform {angle=-154, scale=<0.55 0.55>, shear=<0 0>, translation=<-20 22>}
    renderTransformPivot=<0.5 1>
   }
   slot {class=OverlaySlot, properties {horizontalAlignment=HAlign_Center, verticalAlignment=VAlign_Center}}
  }
  Image_spike_LeftTop {
   widgetClassPath=&Image
   bIsVariable=true
   parent=Overlay_Spikes
   widget=Image_spike_LeftTop
   properties {
    brush {
     bIsDynamicallyLoaded=false
     drawAs=Image
     imageType=NoImage
     mirroring=NoMirror
     resourceName=None
     resourceObject=/Game/Athena/Items/Accolades/Materials/Textures/T_UI_Medal_Spike_LeftTop
     tiling=NoTile
     imageSize=<32 32>
     margin=<ltrb 0 0 0 0>
     outlineSettings {
      bUseBrushTransparency=false
      roundingType=HalfHeightRadius
      width=0
      color {colorUseRule=UseColor_Specified, specifiedColor=<rgba 0 0 0 0>}
      cornerRadii=<0 0 0 0>
     }
     tintColor {colorUseRule=UseColor_Specified, specifiedColor=<rgba 1 1 1 1>}
     uVRegion {bIsValid=false, max=<0 0>, min=<0 0>}
    }
    renderTransform {angle=-34, scale=<0.55 0.55>, shear=<0 0>, translation=<-22 -49>}
    renderTransformPivot=<0.5 1>
   }
   slot {class=OverlaySlot, properties {horizontalAlignment=HAlign_Center, verticalAlignment=VAlign_Center}}
  }
  Image_spike_RightBottom {
   widgetClassPath=&Image
   bIsVariable=true
   parent=Overlay_Spikes
   widget=Image_spike_RightBottom
   properties {
    brush {
     bIsDynamicallyLoaded=false
     drawAs=Image
     imageType=NoImage
     mirroring=NoMirror
     resourceName=None
     resourceObject=/Game/Athena/Items/Accolades/Materials/Textures/T_UI_Medal_Spike_RightBottom
     tiling=NoTile
     imageSize=<32 32>
     margin=<ltrb 0 0 0 0>
     outlineSettings {
      bUseBrushTransparency=false
      roundingType=HalfHeightRadius
      width=0
      color {colorUseRule=UseColor_Specified, specifiedColor=<rgba 0 0 0 0>}
      cornerRadii=<0 0 0 0>
     }
     tintColor {colorUseRule=UseColor_Specified, specifiedColor=<rgba 1 1 1 1>}
     uVRegion {bIsValid=false, max=<0 0>, min=<0 0>}
    }
    renderTransform {angle=154, scale=<0.55 0.55>, shear=<0 0>, translation=<20 22>}
    renderTransformPivot=<0.5 1>
   }
   slot {class=OverlaySlot, properties {horizontalAlignment=HAlign_Center, verticalAlignment=VAlign_Center}}
  }
  Image_spike_RightTop {
   widgetClassPath=&Image
   bIsVariable=true
   parent=Overlay_Spikes
   widget=Image_spike_RightTop
   properties {
    brush {
     bIsDynamicallyLoaded=false
     drawAs=Image
     imageType=NoImage
     mirroring=NoMirror
     resourceName=None
     resourceObject=/Game/Athena/Items/Accolades/Materials/Textures/T_UI_Medal_Spike_RightTop
     tiling=NoTile
     imageSize=<32 32>
     margin=<ltrb 0 0 0 0>
     outlineSettings {
      bUseBrushTransparency=false
      roundingType=HalfHeightRadius
      width=0
      color {colorUseRule=UseColor_Specified, specifiedColor=<rgba 0 0 0 0>}
      cornerRadii=<0 0 0 0>
     }
     tintColor {colorUseRule=UseColor_Specified, specifiedColor=<rgba 1 1 1 1>}
     uVRegion {bIsValid=false, max=<0 0>, min=<0 0>}
    }
    renderTransform {angle=35, scale=<0.55 0.55>, shear=<0 0>, translation=<22 -49>}
    renderTransformPivot=<0.5 1>
   }
   slot {class=OverlaySlot, properties {horizontalAlignment=HAlign_Center, verticalAlignment=VAlign_Center}}
  }
  ScaleBoxMedal {
   widgetClassPath=&ScaleBox
   parent=Overlay_Medal
   widget=ScaleBoxMedal
   properties {stretch=UserSpecified, userSpecifiedScale=0.85}
   slot {class=OverlaySlot, properties {horizontalAlignment=HAlign_Center, verticalAlignment=VAlign_Center}}
  }
  LazyImage_MedalIcon {
   widgetClassPath=&FortLazyImage
   bIsVariable=true
   parent=ScaleBoxMedal
   widget=LazyImage_MedalIcon
   properties {brush {
     bIsDynamicallyLoaded=false
     drawAs=Image
     imageType=NoImage
     mirroring=NoMirror
     resourceName=None
     resourceObject=/Game/Athena/Items/Accolades/Materials/M_UI_XPMedal_Base_Inst
     tiling=NoTile
     imageSize=<128 128>
     margin=<ltrb 0 0 0 0>
     outlineSettings {
      bUseBrushTransparency=false
      roundingType=HalfHeightRadius
      width=0
      color {colorUseRule=UseColor_Specified, specifiedColor=<rgba 0 0 0 0>}
      cornerRadii=<0 0 0 0>
     }
     tintColor {colorUseRule=UseColor_Specified, specifiedColor=<rgba 1 1 1 1>}
     uVRegion {bIsValid=false, max=<0 0>, min=<0 0>}
    }}
   slot {class=ScaleBoxSlot, properties {horizontalAlignment=HAlign_Left, verticalAlignment=VAlign_Top}}
  }
  Image_FXParticles {
   widgetClassPath=&Image
   bIsVariable=true
   parent=Overlay_TopContainer
   widget=Image_FXParticles
   properties {
    renderOpacity=0
    brush {
     bIsDynamicallyLoaded=false
     drawAs=Image
     imageType=NoImage
     mirroring=NoMirror
     resourceName=None
     resourceObject=/Game/UI/Foundation/Onboarding/Materials/M_UI_UpgradeSquareParticle_Inst
     tiling=NoTile
     imageSize=<128 128>
     margin=<ltrb 0 0 0 0>
     outlineSettings {
      bUseBrushTransparency=false
      roundingType=HalfHeightRadius
      width=0
      color {colorUseRule=UseColor_Specified, specifiedColor=<rgba 0 0 0 0>}
      cornerRadii=<0 0 0 0>
     }
     tintColor {colorUseRule=UseColor_Specified, specifiedColor=<rgba 1 1 1 1>}
     uVRegion {bIsValid=false, max=<0 0>, min=<0 0>}
    }
    renderTransform {angle=-90, scale=<1 1>, shear=<0 0>, translation=<0 0>}
   }
   slot {class=OverlaySlot, properties {horizontalAlignment=HAlign_Center, verticalAlignment=VAlign_Center}}
  }
  Image_largeglow {
   widgetClassPath=&Image
   bIsVariable=true
   parent=Overlay_TopContainer
   widget=Image_largeglow
   properties {
    renderOpacity=0
    brush {
     bIsDynamicallyLoaded=false
     drawAs=Image
     imageType=NoImage
     mirroring=NoMirror
     resourceName=None
     resourceObject=/Game/Athena/HUD/LTM/Wax/Materials/M_UI_TextGlow_Add
     tiling=NoTile
     imageSize=<32 32>
     margin=<ltrb 0 0 0 0>
     outlineSettings {
      bUseBrushTransparency=false
      roundingType=HalfHeightRadius
      width=0
      color {colorUseRule=UseColor_Specified, specifiedColor=<rgba 0 0 0 0>}
      cornerRadii=<0 0 0 0>
     }
     tintColor {colorUseRule=UseColor_Specified, specifiedColor=<rgba 1 1 1 1>}
     uVRegion {bIsValid=false, max=<0 0>, min=<0 0>}
    }
    colorAndOpacity=<rgba 0.2135 0.0983 0 1>
   }
   slot {class=OverlaySlot, properties {horizontalAlignment=HAlign_Fill, verticalAlignment=VAlign_Fill, padding=<ltrb -10 -10 -10 -10>}}
  }
  UIEmitter {
   widgetClassPath=/Game/UI/Foundation/UIEmitter/UIEmitter
   bIsVariable=true
   parent=Overlay_TopContainer
   widget=UIEmitter
   properties {
    maxElements=25
    runOnce=true
    startAutomatically=false
    distanceRange=<40 140>
    emitterLifeRange=<0 2>
    materials [/Game/Athena/HUD/LTM/Ashton/Materials/M_Sharp_Cross_Flare_Inst]
    particleLifeRange=<0.2 5>
    sizeRange=<20 150>
   }
   slot {class=OverlaySlot, properties {horizontalAlignment=HAlign_Center, verticalAlignment=VAlign_Center}}
  }
  Image_AnimatedCorona {
   widgetClassPath=&Image
   bIsVariable=true
   parent=Overlay_TopContainer
   widget=Image_AnimatedCorona
   properties {
    renderOpacity=0
    brush {
     bIsDynamicallyLoaded=false
     drawAs=Image
     imageType=NoImage
     mirroring=NoMirror
     resourceName=None
     resourceObject=/Game/Athena/HUD/LTM/Ashton/Materials/M_Flare_Halo_Corona
     tiling=NoTile
     imageSize=<32 32>
     margin=<ltrb 0 0 0 0>
     outlineSettings {
      bUseBrushTransparency=false
      roundingType=HalfHeightRadius
      width=0
      color {colorUseRule=UseColor_Specified, specifiedColor=<rgba 0 0 0 0>}
      cornerRadii=<0 0 0 0>
     }
     tintColor {colorUseRule=UseColor_Specified, specifiedColor=<rgba 1 1 1 1>}
     uVRegion {bIsValid=false, max=<0 0>, min=<0 0>}
    }
   }
   slot {class=OverlaySlot, properties {horizontalAlignment=HAlign_Fill, verticalAlignment=VAlign_Fill, padding=<ltrb -30 -30 -30 -30>}}
  }
  Image_HaloCorona {
   widgetClassPath=&Image
   bIsVariable=true
   parent=Overlay_TopContainer
   widget=Image_HaloCorona
   properties {
    renderOpacity=0
    brush {
     bIsDynamicallyLoaded=false
     drawAs=Image
     imageType=NoImage
     mirroring=NoMirror
     resourceName=None
     resourceObject=/Game/Athena/Items/Accolades/Materials/T_UI_HaloCorona
     tiling=NoTile
     imageSize=<202 202>
     margin=<ltrb 0 0 0 0>
     outlineSettings {
      bUseBrushTransparency=false
      roundingType=HalfHeightRadius
      width=0
      color {colorUseRule=UseColor_Specified, specifiedColor=<rgba 0 0 0 0>}
      cornerRadii=<0 0 0 0>
     }
     tintColor {colorUseRule=UseColor_Specified, specifiedColor=<rgba 1 1 1 1>}
     uVRegion {bIsValid=false, max=<0 0>, min=<0 0>}
    }
    colorAndOpacity=<rgba 1 1 1 0.5>
   }
   slot {class=OverlaySlot, properties {horizontalAlignment=HAlign_Center, verticalAlignment=VAlign_Center, padding=<ltrb -100 -100 -100 -100>}}
  }
 ]
}
