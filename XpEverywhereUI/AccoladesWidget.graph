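# HUD popup widget that announces an earned accolade or medal with icon, title, description and XP amount or simulated reward-size label.
# OpenWidget (item definition) and AccoladeTriggered (raw data) async-load art, then fill visuals, play sound and an intro, using the upgrade variant for medal levels above one.
# A close timer plays Outro, shortened when more XP events queue.
# Layout shifts for playlist tags, touch input and the post-game XP screen.
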WidgetBlueprint /XpEverywhereUI/AccoladesWidget
animations [
 Intro {
  binding_count=10
  length_seconds=3.0982
  object=&self:Intro
  sequence {bindings [
    SizeBoxTriangles {
     id=#1
     possessed_class=SizeBox
     tracks [
      {
       type=MovieSceneFloatTrack
       display_name=Render Opacity
       sections [{type=MovieSceneFloatSection, channels [None {type=MovieSceneScriptingFloatChannel, keys [[25.4643,1],[40.4497,1],[61.9627,0]]}]}]
      }
      {
       type=MovieSceneFloatTrack
       display_name=Width Override
       sections [{type=MovieSceneFloatSection, channels [None {type=MovieSceneScriptingFloatChannel, keys [[25.4643,121.8743],[61.9627,425.1152]]}]}]
      }
      {
       type=MovieSceneFloatTrack
       display_name=Height Override
       sections [{type=MovieSceneFloatSection, channels [None {type=MovieSceneScriptingFloatChannel, keys [[25.4643,13.8272],[34.6563,49.1371]]}]}]
      }
     ]
    }
    Text_Title {
     id=#2
     possessed_class=CommonTextBlock
     tracks [
      {
       type=MovieScene2DTransformTrack
       sections [{
         type=MovieScene2DTransformSection
         channels [
          Scale.X {type=MovieSceneScriptingFloatChannel, keys [[18.795,10],[23.049,0.8094],[23.8003,1],[25.4643,1]]}
          Scale.Y {type=MovieSceneScriptingFloatChannel, keys [[18.795,10],[23.049,0.8094],[23.8003,1],[25.4643,1]]}
         ]
        }]
      }
      {
       type=MovieSceneFloatTrack
       display_name=Render Opacity
       sections [{type=MovieSceneFloatSection, channels [None {type=MovieSceneScriptingFloatChannel, keys [[18.795,0],[22.2427,1]]}]}]
      }
     ]
    }
    Overlay-AccoladeBG {
     id=#3
     possessed_class=Image
     tracks [{
       type=MovieSceneFloatTrack
       display_name=Render Opacity
       sections [{type=MovieSceneFloatSection, channels [None {type=MovieSceneScriptingFloatChannel, keys [[25.4643,0],[32.1337,1]]}]}]
      }]
    }
    Text_Source {
     id=#4
     possessed_class=CommonTextBlock
     tracks [
      {
       type=MovieSceneFloatTrack
       display_name=Render Opacity
       sections [{type=MovieSceneFloatSection, channels [None {type=MovieSceneScriptingFloatChannel, keys [[40.4497,0],[43.8973,1]]}]}]
      }
      {
       type=MovieScene2DTransformTrack
       sections [{
         type=MovieScene2DTransformSection
         channels [
          Scale.X {type=MovieSceneScriptingFloatChannel, keys [[40.4497,8],[44.7037,0.8094],[45.7963,0.9959],[47.119,1]]}
          Scale.Y {type=MovieSceneScriptingFloatChannel, keys [[40.4497,8],[44.7037,0.8094],[45.7963,0.9959],[47.119,1]]}
         ]
        }]
      }
     ]
    }
    AccoladeImageWidget {
     id=#5
     possessed_class=AccoladeImageWidget_C
     tracks [{
       type=MovieScene2DTransformTrack
       sections [{
         type=MovieScene2DTransformSection
         channels [
          Translation.X {type=MovieSceneScriptingFloatChannel}
          Translation.Y {type=MovieSceneScriptingFloatChannel}
          Angle {type=MovieSceneScriptingFloatChannel}
          Scale.X {type=MovieSceneScriptingFloatChannel, keys [[0,8],[6,1.6669],[11,1.5244],[14,0.6025],[15.4243,0.9985],[17.9957,0.9973]]}
          Scale.Y {type=MovieSceneScriptingFloatChannel, keys [[0,8],[6,1.6669],[11,1.5244],[14,0.6025],[15.4243,0.9985],[17.9957,0.9973]]}
          Shear.X {type=MovieSceneScriptingFloatChannel}
          Shear.Y {type=MovieSceneScriptingFloatChannel}
         ]
        }]
      }]
    }
    Image_Glow {
     id=#6
     possessed_class=Image
     tracks [
      {
       type=MovieScene2DTransformTrack
       sections [{
         type=MovieScene2DTransformSection
         channels [
          Scale.X {type=MovieSceneScriptingFloatChannel, keys [[25.854,1],[29.7603,1],[47.119,1.2]]}
          Scale.Y {type=MovieSceneScriptingFloatChannel, keys [[25.854,1],[29.7603,1],[47.119,1.1]]}
         ]
        }]
      }
      {
       type=MovieSceneFloatTrack
       display_name=Render Opacity
       sections [{type=MovieSceneFloatSection, channels [None {type=MovieSceneScriptingFloatChannel, keys [[25.854,1],[61.9627,0]]}]}]
      }
     ]
    }
    Overlay-AccoladeIcon {
     id=#7
     possessed_class=Overlay
     tracks [{
       type=MovieScene2DTransformTrack
       sections [{
         type=MovieScene2DTransformSection
         channels [
          Scale.X {type=MovieSceneScriptingFloatChannel, keys [[0,5],[3.8273,0.8643],[13.6013,1]]}
          Scale.Y {type=MovieSceneScriptingFloatChannel, keys [[0,5],[3.8273,0.8643],[13.6013,1]]}
         ]
        }]
      }]
    }
    HorizontalBoxXPAmount {
     id=#8
     possessed_class=HorizontalBox
     tracks [
      {
       type=MovieSceneFloatTrack
       display_name=Render Opacity
       sections [{type=MovieSceneFloatSection, channels [None {type=MovieSceneScriptingFloatChannel, keys [[40.4497,0],[43.8973,1]]}]}]
      }
      {
       type=MovieScene2DTransformTrack
       sections [{
         type=MovieScene2DTransformSection
         channels [
          Scale.X {type=MovieSceneScriptingFloatChannel, keys [[40.4497,8],[44.7037,0.8094],[45.7963,0.9959],[47.119,1]]}
          Scale.Y {type=MovieSceneScriptingFloatChannel, keys [[40.4497,8],[44.7037,0.8094],[45.7963,0.9959],[47.119,1]]}
         ]
        }]
      }
     ]
    }
    VerticalBox_Container {
     id=#9
     display_name="Text_Title (4)"
     possessed_class=VerticalBox
     tracks [{
       type=MovieSceneFloatTrack
       display_name=Render Opacity
       sections [{type=MovieSceneFloatSection, channels [None {type=MovieSceneScriptingFloatChannel, keys [[0,0],[6,1]]}]}]
      }]
    }
    VerticalBox_Container {id=#10, possessed_class=VerticalBox, tracks []}
   ]}
 }
 Outro {
  binding_count=1
  length_seconds=0.5016
  object=&self:Outro
  sequence {bindings [AccoladeImageWidget {
     id=#11
     display_name=Overlay_TopContainer
     possessed_class=AccoladeImageWidget_C
     tracks [{
       type=MovieSceneFloatTrack
       display_name=Render Opacity
       sections [{type=MovieSceneFloatSection, channels [None {type=MovieSceneScriptingFloatChannel, keys [[0,1],[10.0313,0]]}]}]
      }]
    }]}
 }
 Intro_upgrade {
  binding_count=7
  length_seconds=2.7734
  object=&self:Intro_upgrade
  sequence {bindings [
    Text_Title {
     id=#2
     possessed_class=CommonTextBlock
     tracks [
      {
       type=MovieScene2DTransformTrack
       sections [{
         type=MovieScene2DTransformSection
         channels [
          Scale.X {type=MovieSceneScriptingFloatChannel, keys [[26.9877,10],[31.2417,0.8094],[31.993,1],[33.657,1]]}
          Scale.Y {type=MovieSceneScriptingFloatChannel, keys [[26.9877,10],[31.2417,0.8094],[31.993,1],[33.657,1]]}
         ]
        }]
      }
      {
       type=MovieSceneFloatTrack
       display_name=Render Opacity
       sections [{type=MovieSceneFloatSection, channels [None {type=MovieSceneScriptingFloatChannel, keys [[26.9877,0],[30.4353,1]]}]}]
      }
     ]
    }
    AccoladeImageWidget {
     id=#5
     possessed_class=AccoladeImageWidget_C
     tracks [{
       type=MovieScene2DTransformTrack
       sections [{
         type=MovieScene2DTransformSection
         channels [
          Translation.X {type=MovieSceneScriptingFloatChannel}
          Translation.Y {type=MovieSceneScriptingFloatChannel}
          Angle {type=MovieSceneScriptingFloatChannel}
          Scale.X {type=MovieSceneScriptingFloatChannel, keys [[0,8],[6,1.8279],[20,1.9568],[22.9047,0.6915],[24.3333,0.9985],[26.9047,0.9973]]}
          Scale.Y {type=MovieSceneScriptingFloatChannel, keys [[0,8],[6,1.8279],[20,1.9568],[22.9047,0.6915],[24.3333,0.9985],[26.9047,0.9973]]}
          Shear.X {type=MovieSceneScriptingFloatChannel}
          Shear.Y {type=MovieSceneScriptingFloatChannel}
         ]
        }]
      }]
    }
    Text_Source {
     id=#12
     possessed_class=CommonTextBlock
     tracks [
      {
       type=MovieSceneFloatTrack
       display_name=Render Opacity
       sections [{type=MovieSceneFloatSection, channels [None {type=MovieSceneScriptingFloatChannel, keys [[48.7987,0],[52.2463,1]]}]}]
      }
      {
       type=MovieScene2DTransformTrack
       sections [{
         type=MovieScene2DTransformSection
         channels [
          Scale.X {type=MovieSceneScriptingFloatChannel, keys [[48.7987,8],[53.0527,0.8094],[54.1453,0.9959],[55.468,1]]}
          Scale.Y {type=MovieSceneScriptingFloatChannel, keys [[48.7987,8],[53.0527,0.8094],[54.1453,0.9959],[55.468,1]]}
         ]
        }]
      }
     ]
    }
    Image_TitleBG {
     id=#13
     possessed_class=Image
     tracks [{
       type=MovieSceneFloatTrack
       display_name=Render Opacity
       sections [{type=MovieSceneFloatSection, channels [None {type=MovieSceneScriptingFloatChannel, keys [[32.2613,0],[40.4153,1]]}]}]
      }]
    }
    HorizontalBoxXPAmount {
     id=#8
     possessed_class=HorizontalBox
     tracks [
      {
       type=MovieSceneFloatTrack
       display_name=Render Opacity
       sections [{type=MovieSceneFloatSection, channels [None {type=MovieSceneScriptingFloatChannel, keys [[48.7987,0],[52.2463,1]]}]}]
      }
      {
       type=MovieScene2DTransformTrack
       sections [{
         type=MovieScene2DTransformSection
         channels [
          Scale.X {type=MovieSceneScriptingFloatChannel, keys [[48.7987,8],[53.0527,0.8094],[54.1453,0.9959],[55.468,1]]}
          Scale.Y {type=MovieSceneScriptingFloatChannel, keys [[48.7987,8],[53.0527,0.8094],[54.1453,0.9959],[55.468,1]]}
         ]
        }]
      }
     ]
    }
    Overlay-AccoladeIcon {
     id=#14
     possessed_class=Overlay
     tracks [{
       type=MovieScene2DTransformTrack
       sections [{
         type=MovieScene2DTransformSection
         channels [
          Scale.X {type=MovieSceneScriptingFloatChannel, keys [[0,5],[3.8273,0.8643],[13.6013,1]]}
          Scale.Y {type=MovieSceneScriptingFloatChannel, keys [[0,5],[3.8273,0.8643],[13.6013,1]]}
         ]
        }]
      }]
    }
    VerticalBox_Container {
     id=#9
     display_name="Text_Title (4)"
     possessed_class=VerticalBox
     tracks [{
       type=MovieSceneFloatTrack
       display_name=Render Opacity
       sections [{type=MovieSceneFloatSection, channels [None {type=MovieSceneScriptingFloatChannel, keys [[0,0],[6,1]]}]}]
      }]
    }
   ]}
 }
 XPScreenLayout {
  binding_count=1
  length_seconds=0.3953
  object=&self:XPScreenLayout
  sequence {bindings [Overlay_TopContainer {
     id=#15
     possessed_class=Overlay
     tracks [{
       type=MovieScene2DTransformTrack
       sections [{
         type=MovieScene2DTransformSection
         channels [
          Translation.Y {type=MovieSceneScriptingFloatChannel, keys [[0,0],[7.906,140]]}
          Scale.X {type=MovieSceneScriptingFloatChannel, keys [[0,1],[7.906,0.85]]}
          Scale.Y {type=MovieSceneScriptingFloatChannel, keys [[0,1],[7.906,0.85]]}
         ]
        }]
      }]
    }]}
 }
]
graphs [
 EventGraph {nodes [
   n1 = EventXPLayout!() -> (OutputDelegate, then, Placement:Name) {is=title K2Node_CustomEvent}
   n2 = Get isXPLayout() -> (isXPLayout:bool)
   > n3 = Branch(Condition=AND Boolean(A="Equal (Name)"!(A:Name=@n1.Placement, B:Name=XPScreen) -> (ReturnValue:bool) {is=title K2Node_CallFunction}, B=NOT Boolean(A=@n2)))
   n4 = Get XPScreenLayout() -> (XPScreenLayout:Widget Animation Object Reference)
   > PlayAnimation(InAnimation=@n4)
   > Set isXPLayout(isXPLayout:bool=true) -> (Output_Get:bool)
   Branch(execute=@n3.else, Condition=AND Boolean(A="Equal (Name)"(A=@n1.Placement, B=RecoverFromPostGame), B=Get isXPLayout()))
   > PlayAnimation(InAnimation=@n4, PlayMode=Reverse)
   > Set isXPLayout(isXPLayout=false)
   n5 = WaitForClose() {is=title K2Node_CustomEvent}
   n6 = "DoClose!"() {is=title K2Node_CustomEvent}
   n7 = HasMoreXPEvents!(self:Xp Everywhere Accolades Widget Object Reference) -> (ReturnValue:bool)
   n8 = Set Timer by Event(execute=@n5.then, Delegate=@n6, Time=Select(Option 0:double=2.5, Option 1:double=0.1, Index=@n7) -> (ReturnValue:double), bLooping=false)
   > Set WaitForCloseTimer(WaitForCloseTimer:Timer Handle Structure=@n8) -> (Output_Get:Timer Handle Structure)
   n9 = FinishLoadingAccoladeTexture!() -> (
    OutputDelegate
    then
    LargePreviewImageOverride:Texture 2D Object Reference
    NameOverride:Text
    DescriptionOverride:Text
    AccoladeType:EFortAccoladeType Enum=Acknowledgement
    AccoladeSound:Sound Cue Object Reference
   ) {is=title K2Node_CustomEvent}
   > n10 = GetDynamicMaterial(self=Get Image_Icon() -> (Image_Icon:Image Object Reference))
   > SetTextureParameterValue(self=@n10, ParameterName=Texture, Value=@n9.LargePreviewImageOverride)
   > SetScalarParameterValue(
    self=@n10
    ParameterName=Tier
    Value="To Float (Byte)"!(InByte:Byte=Get AccoladeTier() -> (AccoladeTier:EFortAccoladeTierType Enum=None)) -> (ReturnValue:double)
   )
   > "SetText (Text)"(self=Get Text_Title() -> (Text_Title:Common Text Object Reference), InText=@n9.NameOverride)
   > "SetText (Text)"(self=Get Text_Source() -> (Text_Source:Common Text Object Reference), InText=@n9.DescriptionOverride)
   > SetActiveWidgetIndex(
    self=Get WidgetSwitcher_IconType() -> (WidgetSwitcher_IconType:Widget Switcher Object Reference)
    Index=Select!(Acknowledgement:int=1, Accolade:int=1, Medal:int=0, Index:EFortAccoladeType Enum=@n9.AccoladeType) -> (ReturnValue:int)
   )
   > SetVisibility(
    self=Get Overlay-AccoladeBG() -> (Overlay-AccoladeBG:Overlay Object Reference)
    InVisibility=Select(
     Acknowledgement:ESlateVisibility Enum=HitTestInvisible
     Accolade:ESlateVisibility Enum=HitTestInvisible
     Medal:ESlateVisibility Enum=Hidden
     Index=@n9.AccoladeType
    ) -> (ReturnValue:ESlateVisibility Enum)
   )
   > SetVisibility(InVisibility=HitTestInvisible)
   n11 = Get AccoladeImageWidget() -> (AccoladeImageWidget:Accolade Image Widget Object Reference)
   > n12 = Set MedalLevel!(execute, MedalLevel:int=0, self:Accolade Image Widget Object Reference=@n11) -> (then, Output_Get:int=0)
   > MedalInit(self:Accolade Image Widget Object Reference=@n11)
   > PlayAnimation(
    InAnimation=Select!(
     Option 0:Widget Animation Object Reference=Get Intro()
     Option 1:Widget Animation Object Reference=Get Intro_upgrade() -> (Intro_upgrade:Widget Animation Object Reference)
     Index:bool="integer > integer"(A=@n12, B=1) {is=title K2Node_CallFunction}
    ) -> (ReturnValue:Widget Animation Object Reference)
    PlaybackSpeed=1.5
   )
   > PlaySound2D(Sound=@n9.AccoladeSound)
   n13 = SetTriggeredAccoladeVisuals!() -> (
    OutputDelegate
    then
    LargePreviewImageOverride:Texture 2D Soft Object Reference
    NameOverride:Text
    DescriptionOverride:Text
    AccoladeType:EFortAccoladeType Enum=Acknowledgement
    AccoladeSound:Sound Cue Soft Object Reference
   ) {is=title K2Node_CustomEvent}
   > n14 = Async Load Asset(Asset=@n13.LargePreviewImageOverride)
   n15 = Async Load Asset(execute=@n14.Completed, Asset=@n13.AccoladeSound)
   n16 = Get LazyImage_MedalIcon!(self:Accolade Image Widget Object Reference=Get AccoladeImageWidget()) -> (LazyImage_MedalIcon:Fort Lazy Image Object Reference)
   SetMaterialTextureParamName!(execute=@n15.Completed, self:Common Lazy Image Object Reference=@n16, TextureParamName:Name=Texture) -> (then)
   > SetBrushFromLazyDisplayAsset!(
    execute
    self:Common Lazy Image Object Reference=@n16
    LazyObject:"Soft Object Reference (by ref)"=@n13.LargePreviewImageOverride
    bMatchTextureSize:bool=false
   ) -> (then)
   n17 = Get LazyImage_MedalIcon(self=Get AccoladeImageWidget())
   > SetMaterialTextureParamName(self=@n17, TextureParamName=TextureTop)
   > SetBrushFromLazyDisplayAsset(self=@n17, LazyObject=@n13.LargePreviewImageOverride, bMatchTextureSize=false)
   > FinishLoadingAccoladeTexture!(
    execute
    self
    LargePreviewImageOverride:Texture 2D Object Reference=Cast To Texture2D!(Object=@n14) -> (AsTexture 2D, bSuccess:bool=~)
    NameOverride:Text=@n13.NameOverride
    DescriptionOverride:Text=@n13.DescriptionOverride
    AccoladeType:EFortAccoladeType Enum=@n13.AccoladeType
    AccoladeSound:Sound Cue Object Reference=Cast To SoundCue!(Object=@n15) -> (AsSound Cue:Sound Cue Object Reference, bSuccess:bool=~) {is=title K2Node_DynamicCast}
   ) -> (then) {is=title K2Node_CallFunction}
   HideAndCloseWidget() {is=title K2Node_CustomEvent}
   > SetRenderTranslation(
    self=Get Overlay_TopOffset() -> (Overlay_TopOffset:Overlay Object Reference)
    Translation=Get LatestPlaylistOffset() -> (LatestPlaylistOffset:Vector 2D Structure)
   )
   > SetVisibility(InVisibility=Collapsed)
   > CloseWidget(self:Xp Everywhere Accolades Widget Object Reference)
   n18 = SetAccoladeVisuals!() -> (
    OutputDelegate
    then
    ItemDef:Fort Accolade Item Definition Object Reference
    LargePreviewImageOverride:Texture 2D Soft Object Reference
    NameOverride:Text
    DescriptionOverride:Text
   ) {is=title K2Node_CustomEvent}
   > n19 = Sequence()
   n20 = Select(
    Option 0:Texture 2D Soft Object Reference=@n18.LargePreviewImageOverride
    Option 1:Texture 2D Soft Object Reference=GetLargePreviewImage!(self:Fort Item Definition Object Reference=@n18.ItemDef) -> (ReturnValue:Texture 2D Soft Object Reference)
    Index=IsAssetNull!(Asset:Soft Object Reference=@n18.LargePreviewImageOverride) -> (ReturnValue:bool)
   ) -> (ReturnValue:Texture 2D Soft Object Reference)
   > n21 = Async Load Asset(Asset=@n20)
   Wait for Both Executions!(Required A:Exec=@n21.Completed, Required B:Exec=@n26.Completed, Reset:Exec, bResetOnOutput:bool=true) -> (Then:Exec) {is=title K2Node_MacroInstance}
   n22 = Get LazyImage_MedalIcon(self=Get AccoladeImageWidget())
   > SetMaterialTextureParamName(self=@n22, TextureParamName=Texture)
   n23 = GetPreviousAccoladeLargePreviewImage!(self:Fort Accolade Item Definition Object Reference=@n18.ItemDef) -> (ReturnValue:Texture 2D Soft Object Reference)
   n24 = Select(Option 0=@n23, Option 1=@n20, Index=IsAssetNull(Asset=@n23))
   > SetBrushFromLazyDisplayAsset(self=@n22, LazyObject=@n24, bMatchTextureSize=false)
   n25 = Get LazyImage_MedalIcon(self=Get AccoladeImageWidget())
   > SetMaterialTextureParamName(self=@n25, TextureParamName=TextureTop)
   > SetBrushFromLazyDisplayAsset(self=@n25, LazyObject=@n20, bMatchTextureSize=false)
   > FinishedLoadingTextures!(
    execute
    self
    LargePreviewImageOverride:Texture 2D Soft Object Reference=@n18.LargePreviewImageOverride
    NameOverride:Text=@n18.NameOverride
    DescriptionOverride:Text=@n18.DescriptionOverride
   ) -> (then)
   n26 = Async Load Asset(execute=@n19.then_1, Asset=@n24)
   n27 = FinishedLoadingTextures!() -> (OutputDelegate, then, LargePreviewImageOverride:Texture 2D Soft Object Reference, NameOverride:Text, DescriptionOverride:Text) {is=title K2Node_CustomEvent}
   > n28 = GetDynamicMaterial(self=Get Image_Icon())
   n29 = Get ActiveAccolade() -> (ActiveAccolade:Fort Accolade Item Definition Object Reference)
   n30 = GetLargePreviewImage(self=@n29)
   n31 = IsAssetNull(Asset=@n27.LargePreviewImageOverride)
   > SetTextureParameterValue(
    self=@n28
    ParameterName=Texture
    Value=Resolve Soft Reference!(Input:Texture 2D Soft Object Reference=Select(Option 0=@n27.LargePreviewImageOverride, Option 1=@n30, Index=@n31)) -> (Output:Texture 2D Object Reference)
   )
   > "SetText (Text)"(
    self=Get Text_Title()
    InText=Select(
     Option 0:Text=@n27.NameOverride
     Option 1:Text=Get ItemName!(self:Item Definition Base Object Reference=@n29) -> (ItemName:Text)
     Index=TextIsEmpty(InText=@n27.NameOverride)
    ) -> (ReturnValue:Text)
   )
   > "SetText (Text)"(
    self=Get Text_Source()
    InText=Select(
     Option 0=@n27.DescriptionOverride
     Option 1=GetItemDescription!(self:Item Definition Base Object Reference=@n29) -> (ReturnValue:Text)
     Index=TextIsEmpty(InText=@n27.DescriptionOverride)
    )
   )
   n32 = GetAccoladeType!(self:Fort Accolade Item Definition Object Reference=@n29) -> (ReturnValue:EFortAccoladeType Enum=Acknowledgement)
   > SetActiveWidgetIndex(
    self=Get WidgetSwitcher_IconType()
    Index=Select!(Acknowledgement:int=1, Accolade:int=1, Medal:int=0, Index:EFortAccoladeType Enum=@n32) -> (ReturnValue:int)
   )
   > SetVisibility(
    self=Get Overlay-AccoladeBG()
    InVisibility=Select(
     Acknowledgement:ESlateVisibility Enum=HitTestInvisible
     Accolade:ESlateVisibility Enum=HitTestInvisible
     Medal:ESlateVisibility Enum=Hidden
     Index=@n32
    ) -> (ReturnValue:ESlateVisibility Enum)
   )
   > SetVisibility(InVisibility=HitTestInvisible)
   n33 = Get AccoladeImageWidget()
   > n34 = Set MedalLevel(MedalLevel=Truncate(A=GetAccoladeLevel!(self:Fort Accolade Item Definition Object Reference=@n29) -> (ReturnValue:float)), self=@n33) -> (Output_Get=0)
   > MedalInit(self=@n33)
   > PlayAnimation(
    InAnimation=Select!(
     Option 0:Widget Animation Object Reference=Get Intro()
     Option 1:Widget Animation Object Reference=Get Intro_upgrade()
     Index:bool="integer > integer"(A=@n34, B=1)
    ) -> (ReturnValue:Widget Animation Object Reference)
    PlaybackSpeed=1.5
   )
   > PlaySound2D(Sound=GetAwardedSound!(self:Fort Accolade Item Definition Object Reference=@n29) -> (ReturnValue:Sound Cue Object Reference))
   Event Construct()
   > SetMobileLayout()
   > BindToAnimationFinished(Animation=Get Outro() -> (Outro:Widget Animation Object Reference), Delegate=Create Event())
   n35 = Create Event()
   > BindToAnimationFinished(Animation=Get Intro(), Delegate=@n35)
   > BindToAnimationFinished(Animation=Get Intro_upgrade(), Delegate=@n35)
   > Bind Event to On Post Game Widget Placement Change(
    self:Athena HUDContext Object Reference=Get AthenaHUDContext!() -> (ReturnValue:Athena HUDContext Object Reference) {is=title K2Node_GetBlueprintContextEx}
    Delegate=Create Event()
   ) {is=title K2Node_AddDelegate}
   > n36 = PlaylistDataReadyAsync!(execute:Exec) -> (
    then:Exec
    OnReady:Exec
    GameState:Fort Game State Athena Object Reference
    Playlist:Fort Playlist Object Reference
    PlaylistContextTags:"Gameplay Tag Container Structure (by ref)"
   ) {is=title K2Node_AsyncAction}
   n37 = Handle Offset Translation on Gameplay Tags!(
    Exec:Exec=@n36.OnReady
    PlaylistGameplayTags:Array of Gameplay Tag Structures=BreakGameplayTagContainer!(GameplayTagContainer:"Gameplay Tag Container Structure (by ref)"=@n36.PlaylistContextTags) -> (GameplayTags:Array of Gameplay Tag Structures)
    PlaylistOffsetMap:Map of Gameplay Tag Structures to Vector 2D Structures=Get PlaylistOffsetMap() -> (PlaylistOffsetMap:Map of Gameplay Tag Structures to Vector 2D Structures)
   ) -> (OnOffsetFound:Exec, OnCompleted:Exec, TranslationOffset:Vector 2D Structure, FoundTag:Gameplay Tag Structure) {is=title K2Node_MacroInstance}
   > SetRenderTranslation(self=Get Overlay_TopContainer() -> (Overlay_TopContainer:Overlay Object Reference), Translation=@n37)
   > Set LatestPlaylistOffset(LatestPlaylistOffset:Vector 2D Structure=@n37) -> (Output_Get:Vector 2D Structure)
   Event OnStompedByOtherWidget() {is=title K2Node_Event}
   > StopAllAnimations(self:User Widget Object Reference)
   > SetVisibility(InVisibility=Collapsed)
   > Clear and Invalidate Timer by Handle(Handle=Get WaitForCloseTimer() -> (WaitForCloseTimer:Timer Handle Structure))
   n38 = Event OpenWidget!() -> (
    OutputDelegate
    then
    AccoladeDef:Fort Accolade Item Definition Object Reference
    XpValue:int=0
    DisplayName:"Text (by ref)"
    AccoladeLargePreviewImageOverride:"Texture 2D Soft Object Reference (by ref)"
    SimulatedName:"Text (by ref)"
    SimulatedText:"Text (by ref)"
    SimulatedXPSize:EFortSimulatedXPSize Enum=None
   ) {is=title K2Node_Event}
   > n39 = Set ActiveAccolade(ActiveAccolade:Fort Accolade Item Definition Object Reference=@n38.AccoladeDef) -> (Output_Get:Fort Accolade Item Definition Object Reference)
   > Parent: OpenWidget!(
    execute
    AccoladeDef:Fort Accolade Item Definition Object Reference=@n39
    XpValue:int=@n38.XpValue
    DisplayName:"Text (by ref)"=@n38.DisplayName
    AccoladeLargePreviewImageOverride:"Texture 2D Soft Object Reference (by ref)"=@n38.AccoladeLargePreviewImageOverride
    SimulatedName:"Text (by ref)"=@n38.SimulatedName
    SimulatedText:"Text (by ref)"=@n38.SimulatedText
    SimulatedXPSize:EFortSimulatedXPSize Enum=None
   ) -> (then) {is=title K2Node_CallParentFunction}
   > SetCurrentValue!(
    execute
    self:Common Numeric Text Block Object Reference=Get Text_Amount() -> (Text_Amount:Common Numeric Text Block Object Reference)
    NewValue:float="To Float (Integer)"(InInt=@n38.XpValue)
   ) -> (then)
   > SetAccoladeVisuals!(
    execute
    self
    ItemDef:Fort Accolade Item Definition Object Reference=@n39
    LargePreviewImageOverride:Texture 2D Soft Object Reference=@n38.AccoladeLargePreviewImageOverride
    NameOverride:Text=@n38.SimulatedName
    DescriptionOverride:Text=@n38.SimulatedText
   ) -> (then) {is=title K2Node_CallFunction}
   n40 = "Not Equal (Enum)"!(A:EFortSimulatedXPSize Enum=@n38.SimulatedXPSize, B:EFortSimulatedXPSize Enum=None) -> (ReturnValue:bool)
   > SetActiveWidgetIndex(self=Get XPAmountSwitcher() -> (XPAmountSwitcher:Widget Switcher Object Reference), Index=SelectInt(A=1, B=0, bPickA=@n40))
   > Branch(Condition=@n40)
   > "SetText (Text)"(
    self=Get Text_RewardSize() -> (Text_RewardSize:Common Text Object Reference)
    InText=Format Text!(
     Format="LOCTEXT(\"<{0}>\")"
     0:Text=Select!(
      None:Text
      VerySmall:Text="LOCTEXT(\"VERY SMALL XP AWARD\")"
      Small:Text="LOCTEXT(\"SMALL XP AWARD\")"
      Medium:Text="LOCTEXT(\"MEDIUM XP AWARD\")"
      Large:Text="LOCTEXT(\"LARGE XP AWARD\")"
      VeryLarge:Text="LOCTEXT(\"VERY LARGE XP AWARD\")"
      Index:EFortSimulatedXPSize Enum=@n38.SimulatedXPSize
     ) -> (ReturnValue:Text)
    ) -> (Result)
   )
   n41 = Event AccoladeTriggered!() -> (
    OutputDelegate
    then
    XpValue:int=0
    SimulatedName:"Text (by ref)"
    SimulatedText:"Text (by ref)"
    PreviewImage:"Texture 2D Soft Object Reference (by ref)"
    SimulatedXPSize:EFortSimulatedXPSize Enum=None
    AccoladeType:EFortAccoladeType Enum=Acknowledgement
    SoundCue:"Sound Cue Soft Object Reference (by ref)"
    AccoladeTier:EFortAccoladeTierType Enum=None
   ) {is=title K2Node_Event}
   > Set AccoladeTier(AccoladeTier:EFortAccoladeTierType Enum=@n41.AccoladeTier) -> (Output_Get:EFortAccoladeTierType Enum=None)
   > SetCurrentValue(self=Get Text_Amount(), NewValue="To Float (Integer)"(InInt=@n41.XpValue))
   > SetTriggeredAccoladeVisuals!(
    execute
    self
    LargePreviewImageOverride:Texture 2D Soft Object Reference=@n41.PreviewImage
    NameOverride:Text=@n41.SimulatedName
    DescriptionOverride:Text=@n41.SimulatedText
    AccoladeType:EFortAccoladeType Enum=@n41.AccoladeType
    AccoladeSound:Sound Cue Soft Object Reference=@n41.SoundCue
   ) -> (then) {is=title K2Node_CallFunction}
   n42 = "Not Equal (Enum)"(A=@n41.SimulatedXPSize, B=None)
   > SetActiveWidgetIndex(self=Get XPAmountSwitcher(), Index=SelectInt(A=1, B=0, bPickA=@n42))
   > Branch(Condition=@n42)
   > "SetText (Text)"(
    self=Get Text_RewardSize()
    InText=Format Text(
     Format="LOCTEXT(\"<{0}>\")"
     0=Select(
      VerySmall="LOCTEXT(\"VERY SMALL XP AWARD\")"
      Small="LOCTEXT(\"SMALL XP AWARD\")"
      Medium="LOCTEXT(\"MEDIUM XP AWARD\")"
      Large="LOCTEXT(\"LARGE XP AWARD\")"
      VeryLarge="LOCTEXT(\"VERY LARGE XP AWARD\")"
      Index=@n41.SimulatedXPSize
     )
    )
   )
   Event OnEventAdded() {is=title K2Node_Event}
   > Branch(Condition=Is Timer Active by Handle!(Handle:Timer Handle Structure=Get WaitForCloseTimer()) -> (ReturnValue:bool))
   > Clear and Invalidate Timer by Handle(Handle=Get WaitForCloseTimer())
   > "DoClose!"() {is=title K2Node_CallFunction}
   PlayAnimation(
    execute=@n6.then
    InAnimation=Get Outro()
    PlaybackSpeed=Select!(Option 0:double=1, Option 1:double=4, Index:bool=@n7) -> (ReturnValue:double)
   )
  ]}
 SetMobileLayout {nodes [
   SetMobileLayout() {is=title K2Node_FunctionEntry}
   > Branch(Condition=IsUsingTouch(PlayerController=GetOwningPlayer()))
   > SetPadding!(
    execute
    self:Overlay Slot Object Reference=SlotAsOverlaySlot(Widget=Get Overlay_TopContainer())
    InPadding_Left:float=0
    InPadding_Top:float=132
    InPadding_Right:float=0
    InPadding_Bottom:float=0
   ) -> (then)
  ]}
]
variables [
 AccoladeTier {type=EFortAccoladeTierType, default=None}
 ActiveAccolade {type=FortAccoladeItemDefinition, default=null}
 isXPLayout {type=boolean}
 LatestPlaylistOffset {type=Vector2D, default=<0 0>}
 PlaylistOffsetMap {default {
   "(TagName=\"Athena.Playlist.Argon\")"=<0 120>
   "(TagName=\"Athena.Playlist.CobaltRespawn\")"=<0 70>
   "(TagName=\"Athena.Playlist.Infiltration\")"=<0 120>
   "(TagName=\"Athena.Playlist.Payback\")"=<0 180>
   "(TagName=\"Athena.Playlist.Phosphorus\")"=<0 40>
   "(TagName=\"Athena.Playlist.UraniumP\")"=<0 120>
  }}
 WaitForCloseTimer {type=TimerHandle}
]
widgets {
 info {inheritedWidgetCount=1, parentClass=/Script/XpEverywhereUI.XpEverywhereAccoladesWidget, rootWidgetClass=&SizeBox, widgetCount=36}
 widgets [
  SizeBox_0 {
   widgetClassPath=&SizeBox
   widget=SizeBox_0
   properties {bOverride_MinDesiredHeight=true, bOverride_MinDesiredWidth=true, minDesiredHeight=125, minDesiredWidth=600}
  }
  Overlay_TopOffset {
   widgetClassPath=&Overlay
   bIsVariable=true
   parent=SizeBox_0
   widget=Overlay_TopOffset
   slot {class=SizeBoxSlot, properties {horizontalAlignment=HAlign_Center, verticalAlignment=VAlign_Top}}
  }
  Overlay_TopContainer {widgetClassPath=&Overlay, bIsVariable=true, parent=Overlay_TopOffset, widget=Overlay_TopContainer, slot {class=OverlaySlot}}
  Image_shadow {
   widgetClassPath=&Image
   bIsVariable=true
   parent=Overlay_TopContainer
   widget=Image_shadow
   properties {
    brush {
     bIsDynamicallyLoaded=false
     drawAs=Image
     imageType=NoImage
     mirroring=NoMirror
     resourceName=None
     resourceObject=/Game/Athena/HUD/LTM/Wax/Materials/M_UI_TextGlow_Inst
     tiling=NoTile
     imageSize=<279.4116 128.8524>
     margin=<ltrb 0 0 0 0>
     outlineSettings {
      bUseBrushTransparency=false
      roundingType=HalfHeightRadius
      width=0
      color {colorUseRule=UseColor_Specified, specifiedColor=<rgba 0 0 0 0>}
      cornerRadii=<0 0 0 0>
     }
     tintColor {colorUseRule=UseColor_Specified, specifiedColor=<rgba 1 1 1 1>}
     uVRegion {bIsValid=false, max=<0 0>, min=<0 0>}
    }
    colorAndOpacity=<rgba 0.0156 0.0073 0.0032 0.2>
    renderTransform {angle=0, scale=<1 1>, shear=<0 0>, translation=<0 -5>}
   }
   slot {class=OverlaySlot, properties {horizontalAlignment=HAlign_Fill, verticalAlignment=VAlign_Fill, padding=<ltrb -125 -125 -125 -125>}}
  }
  HorizontalBox_2 {
   widgetClassPath=&HorizontalBox
   parent=Overlay_TopContainer
   widget=HorizontalBox_2
   slot {class=OverlaySlot, properties {horizontalAlignment=HAlign_Center, verticalAlignment=VAlign_Fill}}
  }
  WidgetSwitcher_IconType {
   widgetClassPath=&WidgetSwitcher
   bIsVariable=true
   parent=HorizontalBox_2
   widget=WidgetSwitcher_IconType
   properties {activeWidgetIndex=1}
   slot {class=HorizontalBoxSlot}
  }
  ScaleBox_0 {
   widgetClassPath=&ScaleBox
   parent=WidgetSwitcher_IconType
   widget=ScaleBox_0
   properties {stretch=UserSpecified, userSpecifiedScale=1.5}
   slot {class=WidgetSwitcherSlot, properties {horizontalAlignment=HAlign_Center, verticalAlignment=VAlign_Center}}
  }
  AccoladeImageWidget {
   widgetClassPath=/Game/Athena/HUD/XP/AccoladeImageWidget
   bIsVariable=true
   parent=ScaleBox_0
   widget=AccoladeImageWidget
   slot {class=ScaleBoxSlot}
  }
  Overlay-AccoladeIcon {
   widgetClassPath=&Overlay
   parent=WidgetSwitcher_IconType
   widget=Overlay-AccoladeIcon
   slot {class=WidgetSwitcherSlot, properties {horizontalAlignment=HAlign_Center, verticalAlignment=VAlign_Top, padding=<ltrb 0 0 0 12>}}
  }
  CommonVisualAttachment_0 {
   widgetClassPath=&CommonVisualAttachment
   parent=Overlay-AccoladeIcon
   widget=CommonVisualAttachment_0
   properties {contentAnchor=<0.5 0.5>}
   slot {class=OverlaySlot, properties {horizontalAlignment=HAlign_Center, verticalAlignment=VAlign_Center}}
  }
  Image_IconGlow {
   widgetClassPath=&Image
   bIsVariable=true
   parent=CommonVisualAttachment_0
   widget=Image_IconGlow
   properties {
    brush {
     bIsDynamicallyLoaded=false
     drawAs=Image
     imageType=NoImage
     mirroring=NoMirror
     resourceName=None
     resourceObject=/Game/Athena/HUD/LTM/Wax/Materials/M_UI_TextGlow
     tiling=NoTile
     imageSize=<200 200>
     margin=<ltrb 0 0 0 0>
     outlineSettings {
      bUseBrushTransparency=false
      roundingType=HalfHeightRadius
      width=0
      color {colorUseRule=UseColor_Specified, specifiedColor=<rgba 0 0 0 0>}
      cornerRadii=<0 0 0 0>
     }
     tintColor {colorUseRule=UseColor_Specified, specifiedColor=<rgba 1 1 1 1>}
     uVRegion {bIsValid=false, max=<0 0>, min=<0 0>}
    }
    colorAndOpacity=<rgba 1 0.1996 0 0.27>
   }
   slot {class=SizeBoxSlot}
  }
  Image_Icon {
   widgetClassPath=&Image
   bIsVariable=true
   parent=Overlay-AccoladeIcon
   widget=Image_Icon
   properties {brush {
     bIsDynamicallyLoaded=false
     drawAs=Image
     imageType=FullColor
     mirroring=NoMirror
     resourceName=None
     resourceObject=/Game/Athena/Items/Accolades/Materials/MI_UI_AccoladeSplashIcon_Tiered
     tiling=NoTile
     imageSize=<88 88>
     margin=<ltrb 0 0 0 0>
     outlineSettings {
      bUseBrushTransparency=false
      roundingType=HalfHeightRadius
      width=0
      color {colorUseRule=UseColor_Specified, specifiedColor=<rgba 0 0 0 0>}
      cornerRadii=<0 0 0 0>
     }
     tintColor {colorUseRule=UseColor_Specified, specifiedColor=<rgba 1 1 1 1>}
     uVRegion {bIsValid=false, max=<0 0>, min=<0 0>}
    }}
   slot {class=OverlaySlot, properties {horizontalAlignment=HAlign_Center, verticalAlignment=VAlign_Center, padding=<ltrb 0 25 0 0>}}
  }
  VerticalBox_Container {
   widgetClassPath=&VerticalBox
   parent=HorizontalBox_2
   widget=VerticalBox_Container
   slot {class=HorizontalBoxSlot, properties {padding=<ltrb 20 0 0 0>}}
  }
  Text_Description {
   widgetClassPath=&TextBlock
   bIsVariable=true
   parent=VerticalBox_Container
   widget=Text_Description
   properties {text="Great Job!", visibility=Collapsed}
   slot {class=VerticalBoxSlot, properties {horizontalAlignment=HAlign_Center}}
  }
  Overlay_Title {
   widgetClassPath=&Overlay
   parent=VerticalBox_Container
   widget=Overlay_Title
   slot {class=VerticalBoxSlot, properties {horizontalAlignment=HAlign_Left, padding=<ltrb 0 -8 0 0>}}
  }
  Overlay-AccoladeBG {
   widgetClassPath=&Overlay
   bIsVariable=true
   parent=Overlay_Title
   widget=Overlay-AccoladeBG
   slot {class=OverlaySlot, properties {horizontalAlignment=HAlign_Fill, verticalAlignment=VAlign_Center, padding=<ltrb -125 0 -125 0>}}
  }
  SizeBoxTriangles {
   widgetClassPath=&SizeBox
   parent=Overlay-AccoladeBG
   widget=SizeBoxTriangles
   properties {bOverride_HeightOverride=true, bOverride_WidthOverride=true, heightOverride=49.1371, widthOverride=121.8743}
   slot {class=OverlaySlot, properties {horizontalAlignment=HAlign_Center, verticalAlignment=VAlign_Center, padding=<ltrb 0 -50 0 -50>}}
  }
  HorizontalBox_0 {widgetClassPath=&HorizontalBox, parent=SizeBoxTriangles, widget=HorizontalBox_0, slot {class=SizeBoxSlot}}
  ImageTriangleL {
   widgetClassPath=&Image
   parent=HorizontalBox_0
   widget=ImageTriangleL
   properties {
    brush {
     bIsDynamicallyLoaded=false
     drawAs=Image
     imageType=NoImage
     mirroring=NoMirror
     resourceName=None
     resourceObject=/Game/UI/HUD/Materials/M_FX_TriangleSpiral
     tiling=NoTile
     imageSize=<125 85>
     margin=<ltrb 0 0 0 0>
     outlineSettings {
      bUseBrushTransparency=false
      roundingType=HalfHeightRadius
      width=0
      color {colorUseRule=UseColor_Specified, specifiedColor=<rgba 0 0 0 0>}
      cornerRadii=<0 0 0 0>
     }
     tintColor {colorUseRule=UseColor_Specified, specifiedColor=<rgba 0.9427 1 0 1>}
     uVRegion {bIsValid=false, max=<0 0>, min=<0 0>}
    }
    renderTransform {angle=0, scale=<-1 1>, shear=<0 0>, translation=<0 0>}
    renderTransformPivot=<0.5 1>
   }
   slot {class=HorizontalBoxSlot, properties {verticalAlignment=VAlign_Center, padding=<ltrb -200 0 0 0>, size {sizeRule=Fill, value=0.5}}}
  }
  ImageTriangleR {
   widgetClassPath=&Image
   parent=HorizontalBox_0
   widget=ImageTriangleR
   properties {
    brush {
     bIsDynamicallyLoaded=false
     drawAs=Image
     imageType=NoImage
     mirroring=NoMirror
     resourceName=None
     resourceObject=&M_FX_TriangleSpiral
     tiling=NoTile
     imageSize=<125 85>
     margin=<ltrb 0 0 0 0>
     outlineSettings {
      bUseBrushTransparency=false
      roundingType=HalfHeightRadius
      width=0
      color {colorUseRule=UseColor_Specified, specifiedColor=<rgba 0 0 0 0>}
      cornerRadii=<0 0 0 0>
     }
     tintColor {colorUseRule=UseColor_Specified, specifiedColor=<rgba 0.9427 1 0 1>}
     uVRegion {bIsValid=false, max=<0 0>, min=<0 0>}
    }
    renderTransformPivot=<0.5 1>
   }
   slot {class=HorizontalBoxSlot, properties {verticalAlignment=VAlign_Center, padding=<ltrb 0 0 -200 0>, size {sizeRule=Fill, value=0.5}}}
  }
  Image_TitleBG {
   widgetClassPath=&Image
   bIsVariable=true
   parent=Overlay-AccoladeBG
   widget=Image_TitleBG
   properties {
    brush {
     bIsDynamicallyLoaded=false
     drawAs=Image
     imageType=NoImage
     mirroring=NoMirror
     resourceName=None
     resourceObject=/Game/Athena/HUD/LTM/Ashton/Materials/M_UI_AnamorphicFlare
     tiling=NoTile
     imageSize=<399.3419 32>
     margin=<ltrb 0 0 0 0>
     outlineSettings {
      bUseBrushTransparency=false
      roundingType=HalfHeightRadius
      width=0
      color {colorUseRule=UseColor_Specified, specifiedColor=<rgba 0 0 0 0>}
      cornerRadii=<0 0 0 0>
     }
     tintColor {colorUseRule=UseColor_Specified, specifiedColor=<rgba 1 1 1 1>}
     uVRegion {bIsValid=false, max=<0 0>, min=<0 0>}
    }
    colorAndOpacity=<rgba 1 0.3529 0 0.39>
    renderTransform {angle=0, scale=<1 1>, shear=<0 0>, translation=<2 -2>}
   }
   slot {class=OverlaySlot, properties {horizontalAlignment=HAlign_Fill, padding=<ltrb 30 0 -30 0>}}
  }
  Image_Glow {
   widgetClassPath=&Image
   bIsVariable=true
   parent=Overlay-AccoladeBG
   widget=Image_Glow
   properties {brush {
     bIsDynamicallyLoaded=false
     drawAs=Image
     imageType=NoImage
     mirroring=NoMirror
     resourceName=None
     resourceObject=/Game/Athena/HUD/LTM/Wax/Materials/M_UI_TextGlow_Add
     tiling=NoTile
     imageSize=<32 32>
     margin=<ltrb 0 0 0 0>
     outlineSettings {
      bUseBrushTransparency=false
      roundingType=HalfHeightRadius
      width=0
      color {colorUseRule=UseColor_Specified, specifiedColor=<rgba 0 0 0 0>}
      cornerRadii=<0 0 0 0>
     }
     tintColor {colorUseRule=UseColor_Specified, specifiedColor=<rgba 1 0.1187 0 1>}
     uVRegion {bIsValid=false, max=<0 0>, min=<0 0>}
    }}
   slot {class=OverlaySlot, properties {horizontalAlignment=HAlign_Fill, verticalAlignment=VAlign_Fill, padding=<ltrb 0 -10 0 -10>}}
  }
  Text_Title {
   widgetClassPath=&CommonTextBlock
   bIsVariable=true
   parent=Overlay_Title
   widget=Text_Title
   properties {
    mobileFontSizeMultiplier=1.05
    style=/Game/Athena/HUD/TextStyles/TextStyle-HUD-XPDescriptor.TextStyle-HUD-XPDescriptor_C
    text=Accolade Title
    textTransformPolicy=ToUpper
    font {
     bForceMonospaced=false
     bMaterialIsStencil=false
     fontMaterial=null
     fontObject=&Burbank
     letterSpacing=0
     monospacedWidth=1
     size=20
     skewAmount=0
     typefaceFontName=Burbank-PlayerPower
     outlineSettings {
      bApplyOutlineToDropShadows=false
      bMiteredCorners=false
      bSeparateFillAlpha=true
      outlineMaterial=&M-FontShadow
      outlineSize=0
      outlineColor=<rgba 0.1094 0.0129 0 0.5>
     }
    }
    renderTransformPivot=<0.5 -0.5>
    shadowColorAndOpacity=<rgba 0.0052 0.0052 0.0056 0.4>
    shadowOffset=<2 2>
   }
   slot {class=OverlaySlot}
  }
  ScaleBox_1 {
   widgetClassPath=&ScaleBox
   parent=VerticalBox_Container
   widget=ScaleBox_1
   properties {stretch=UserSpecified, userSpecifiedScale=0.9}
   slot {class=VerticalBoxSlot, properties {horizontalAlignment=HAlign_Left, padding=<ltrb 0 -8 0 0>}}
  }
  Text_Source {
   widgetClassPath=&CommonTextBlock
   bIsVariable=true
   parent=ScaleBox_1
   widget=Text_Source
   properties {
    mobileFontSizeMultiplier=1.05
    style=&TextStyle-HUD-XPDescriptor_C
    text=Accolade Source
    font {
     bForceMonospaced=false
     bMaterialIsStencil=false
     fontMaterial=null
     fontObject=&Burbank
     letterSpacing=0
     monospacedWidth=1
     size=20
     skewAmount=0
     typefaceFontName=Burbank-PlayerPower
     outlineSettings {
      bApplyOutlineToDropShadows=false
      bMiteredCorners=false
      bSeparateFillAlpha=true
      outlineMaterial=&M-FontShadow
      outlineSize=0
      outlineColor=<rgba 0.1094 0.0129 0 0.5>
     }
    }
    renderTransformPivot=<0.5 -0.5>
    shadowColorAndOpacity=<rgba 0.0052 0.0052 0.0056 0.4>
    shadowOffset=<2 2>
   }
   slot {class=ScaleBoxSlot, properties {horizontalAlignment=HAlign_Left, verticalAlignment=VAlign_Top}}
  }
  HorizontalBoxXPAmount {
   widgetClassPath=&HorizontalBox
   parent=VerticalBox_Container
   widget=HorizontalBoxXPAmount
   properties {visibility=HitTestInvisible, renderTransformPivot=<0.5 -0.5>}
   slot {class=VerticalBoxSlot, properties {horizontalAlignment=HAlign_Left, padding=<ltrb 0 -12 0 0>}}
  }
  XPAmountSwitcher {
   widgetClassPath=&WidgetSwitcher
   bIsVariable=true
   parent=HorizontalBoxXPAmount
   widget=XPAmountSwitcher
   slot {class=HorizontalBoxSlot, properties {horizontalAlignment=HAlign_Center, verticalAlignment=VAlign_Center, padding=<ltrb 0 0 -2 -1>}}
  }
  HorizontalBox_1 {widgetClassPath=&HorizontalBox, parent=XPAmountSwitcher, widget=HorizontalBox_1, slot {class=WidgetSwitcherSlot}}
  Text_+ {
   widgetClassPath=&CommonTextBlock
   parent=HorizontalBox_1
   widget=Text_+
   properties {
    mobileFontSizeMultiplier=1.05
    style=&TextStyle-HUD-XPDescriptor_C
    text=+
    font {
     bForceMonospaced=false
     bMaterialIsStencil=false
     fontMaterial=null
     fontObject=&Burbank
     letterSpacing=0
     monospacedWidth=1
     size=20
     skewAmount=0
     typefaceFontName=Burbank-PlayerPower
     outlineSettings {
      bApplyOutlineToDropShadows=false
      bMiteredCorners=false
      bSeparateFillAlpha=true
      outlineMaterial=&M-FontShadow
      outlineSize=0
      outlineColor=<rgba 0.1094 0.0129 0 0.5>
     }
    }
    shadowColorAndOpacity=<rgba 0.0052 0.0052 0.0056 0.4>
    shadowOffset=<2 2>
   }
   slot {class=HorizontalBoxSlot}
  }
  Text_Amount {
   widgetClassPath=/Script/CommonUI.CommonNumericTextBlock
   bIsVariable=true
   parent=HorizontalBox_1
   widget=Text_Amount
   properties {
    currentNumericValue=1000
    justification=Center
    mobileFontSizeMultiplier=1.05
    style=&TextStyle-HUD-XPDescriptor_C
    text=Text Block
    font {
     bForceMonospaced=false
     bMaterialIsStencil=false
     fontMaterial=null
     fontObject=&Burbank
     letterSpacing=0
     monospacedWidth=1
     size=20
     skewAmount=0
     typefaceFontName=Burbank-PlayerPower
     outlineSettings {
      bApplyOutlineToDropShadows=false
      bMiteredCorners=false
      bSeparateFillAlpha=true
      outlineMaterial=&M-FontShadow
      outlineSize=0
      outlineColor=<rgba 0.1094 0.0129 0 0.5>
     }
    }
    shadowColorAndOpacity=<rgba 0.0052 0.0052 0.0056 0.4>
    shadowOffset=<2 2>
   }
   slot {class=HorizontalBoxSlot, properties {horizontalAlignment=HAlign_Right}}
  }
  Text_XP {
   widgetClassPath=&CommonTextBlock
   parent=HorizontalBox_1
   widget=Text_XP
   properties {
    mobileFontSizeMultiplier=1.05
    style=&TextStyle-HUD-XPDescriptor_C
    text=XP
    textTransformPolicy=ToUpper
    font {
     bForceMonospaced=false
     bMaterialIsStencil=false
     fontMaterial=null
     fontObject=&Burbank
     letterSpacing=0
     monospacedWidth=1
     size=20
     skewAmount=0
     typefaceFontName=Burbank-PlayerPower
     outlineSettings {
      bApplyOutlineToDropShadows=false
      bMiteredCorners=false
      bSeparateFillAlpha=true
      outlineMaterial=&M-FontShadow
      outlineSize=0
      outlineColor=<rgba 0.1094 0.0129 0 0.5>
     }
    }
    shadowColorAndOpacity=<rgba 0.0052 0.0052 0.0056 0.4>
    shadowOffset=<2 2>
   }
   slot {class=HorizontalBoxSlot, properties {horizontalAlignment=HAlign_Center, verticalAlignment=VAlign_Center, padding=<ltrb 0 0 0 -1>}}
  }
  VerticalBox_0 {
   widgetClassPath=&VerticalBox
   parent=XPAmountSwitcher
   widget=VerticalBox_0
   slot {class=WidgetSwitcherSlot, properties {horizontalAlignment=HAlign_Center, verticalAlignment=VAlign_Center}}
  }
  Text_RewardSize {
   widgetClassPath=&CommonTextBlock
   bIsVariable=true
   parent=VerticalBox_0
   widget=Text_RewardSize
   properties {
    mobileFontSizeMultiplier=1.05
    style=&TextStyle-HUD-XPDescriptor_C
    text="<Large XP Reward>"
    font {
     bForceMonospaced=false
     bMaterialIsStencil=false
     fontMaterial=null
     fontObject=&Burbank
     letterSpacing=0
     monospacedWidth=1
     size=20
     skewAmount=0
     typefaceFontName=Burbank-PlayerPower
     outlineSettings {
      bApplyOutlineToDropShadows=false
      bMiteredCorners=false
      bSeparateFillAlpha=true
      outlineMaterial=&M-FontShadow
      outlineSize=0
      outlineColor=<rgba 0.1094 0.0129 0 0.5>
     }
    }
    shadowColorAndOpacity=<rgba 0.0052 0.0052 0.0056 0.4>
    shadowOffset=<2 2>
   }
   slot {class=VerticalBoxSlot, properties {padding=<ltrb 0 0 0 -8>}}
  }
  ScaleBox_2 {
   widgetClassPath=&ScaleBox
   parent=VerticalBox_0
   widget=ScaleBox_2
   properties {stretch=UserSpecified, userSpecifiedScale=0.75}
   slot {class=VerticalBoxSlot, properties {horizontalAlignment=HAlign_Center, verticalAlignment=VAlign_Center}}
  }
  CommonTextBlock_TestMode {
   widgetClassPath=&CommonTextBlock
   parent=ScaleBox_2
   widget=CommonTextBlock_TestMode
   properties {
    mobileFontSizeMultiplier=1.05
    style=&TextStyle-HUD-XPDescriptor_C
    text="[Test Mode]"
    textTransformPolicy=ToUpper
    font {
     bForceMonospaced=false
     bMaterialIsStencil=false
     fontMaterial=null
     fontObject=&Burbank
     letterSpacing=0
     monospacedWidth=1
     size=20
     skewAmount=0
     typefaceFontName=Burbank-PlayerPower
     outlineSettings {
      bApplyOutlineToDropShadows=false
      bMiteredCorners=false
      bSeparateFillAlpha=true
      outlineMaterial=&M-FontShadow
      outlineSize=0
      outlineColor=<rgba 0.1094 0.0129 0 0.5>
     }
    }
    renderTransformPivot=<0.2 0.75>
    shadowColorAndOpacity=<rgba 0.0052 0.0052 0.0056 0.4>
    shadowOffset=<2 2>
   }
   slot {class=ScaleBoxSlot}
  }
  TutorialHighlight {widgetClassPath=&TutorialHighlight}
 ]
}
