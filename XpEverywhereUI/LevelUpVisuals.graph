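# Level-up banner widget for the XP HUD, animating the icon, LEVEL UP text and level number over layered beams, glows and particles.
# EventStartAnim plays NewIntro, which drives transforms, opacity and material parameters; its start triggers a delayed sound.
# Animation-finished bindings chain intro into NewOutro, then broadcast DispatchComplete.
# Extras include a flare animation above level 100, a mobile icon upscale, reverse-play reset and a placeholder print-only reward function.
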WidgetBlueprint /XpEverywhereUI/LevelUpVisuals
animations [
 Hah {
  binding_count=3
  length_seconds=0.7104
  object=&self:Hah
  sequence {bindings [
    TriFX_1 {
     id=#1
     possessed_class=Image
     tracks [
      {
       type=MovieScene2DTransformTrack
       sections [{
         type=MovieScene2DTransformSection
         channels [
          Scale.X {type=MovieSceneScriptingFloatChannel, keys [[1.1443,0],[2.884,1],[13.61,1.6241]]}
          Scale.Y {type=MovieSceneScriptingFloatChannel, keys [[1.1443,0],[2.884,1],[13.61,1.6241]]}
         ]
        }]
      }
      {
       type=MovieSceneFloatTrack
       display_name=Render Opacity
       sections [{type=MovieSceneFloatSection, channels [None {type=MovieSceneScriptingFloatChannel, keys [[1.1443,0],[2.1443,1],[10.1443,1],[14.208,0]]}]}]
      }
     ]
    }
    TriFX_3 {
     id=#2
     possessed_class=Image
     tracks [{
       type=MovieScene2DTransformTrack
       sections [{
         type=MovieScene2DTransformSection
         channels [
          Scale.X {type=MovieSceneScriptingFloatChannel, keys [[0,1],[1.0617,0.7059],[2.7453,1.0082],[13.0637,1]]}
          Scale.Y {type=MovieSceneScriptingFloatChannel, keys [[0,1],[1.0617,0.7059],[2.7453,1.0082],[13.0637,1]]}
         ]
        }]
      }]
    }
    HorizontalFlare {
     id=#3
     possessed_class=Image
     tracks [
      {
       type=MovieSceneFloatTrack
       display_name=Render Opacity
       sections [{
         type=MovieSceneFloatSection
         channels [None {type=MovieSceneScriptingFloatChannel, keys [[1.1443,0],[2.1443,0.3296],[10.1443,0.3296],[14.208,0]]}]
        }]
      }
      {
       type=MovieScene2DTransformTrack
       sections [{
         type=MovieScene2DTransformSection
         channels [
          Scale.X {type=MovieSceneScriptingFloatChannel, keys [[1.1443,0.6164],[2.884,1.3069],[14.026,1.7761]]}
          Scale.Y {type=MovieSceneScriptingFloatChannel, keys [[1.1443,0.6164],[2.884,1.3069],[14.026,1.7761]]}
         ]
        }]
      }
     ]
    }
   ]}
 }
 NewOutro {
  binding_count=1
  length_seconds=0.5
  object=&self:NewOutro
  sequence {bindings [All {
     id=#4
     possessed_class=SizeBox
     tracks [{
       type=MovieSceneFloatTrack
       display_name=Render Opacity
       sections [{type=MovieSceneFloatSection, channels [None {type=MovieSceneScriptingFloatChannel, keys [[0,1],[10,0]]}]}]
      }]
    }]}
 }
 NewIntro {
  binding_count=24
  length_seconds=4.4167
  object=&self:NewIntro
  sequence {
   bindings [
    TriFX_2 {
     id=#5
     possessed_class=Image
     tracks [{
       type=MovieSceneWidgetMaterialTrack
       sections [{type=MovieSceneParameterSection, channels [TriScale {type=MovieSceneScriptingFloatChannel, keys [[15.991,0.65],[65.992,0]]}]}]
      }]
    }
    All {
     id=#6
     possessed_class=SizeBox
     tracks [{
       type=MovieSceneFloatTrack
       display_name=Render Opacity
       sections [{type=MovieSceneFloatSection, channels [None {type=MovieSceneScriptingFloatChannel, keys [[6,0],[58,1]]}]}]
      }]
    }
    LevelNumber {
     id=#7
     display_name=AnimBoundLevelBox
     possessed_class=CommonTextBlock
     tracks [
      {
       type=MovieSceneFloatTrack
       display_name=Render Opacity
       sections [{type=MovieSceneFloatSection, channels [None {type=MovieSceneScriptingFloatChannel, keys [[6,0],[166.246,0],[186.277,1]]}]}]
      }
      {
       type=MovieScene2DTransformTrack
       sections [{type=MovieScene2DTransformSection, channels [Translation.X {type=MovieSceneScriptingFloatChannel, keys [[166.246,-64],[186.277,0]]}]}]
      }
      {
       type=MovieSceneColorTrack
       display_name=Content Color and Opacity
       sections [{
         type=MovieSceneColorSection
         channels [
          Color.R {type=MovieSceneScriptingFloatChannel, keys [[166.246,1],[186.277,1],[213.938,0],[254,0]]}
          Color.G {type=MovieSceneScriptingFloatChannel, keys [[166.246,1],[186.277,1],[213.938,0]]}
          Color.B {type=MovieSceneScriptingFloatChannel, keys [[166.246,1],[186.277,1],[213.938,0]]}
          Color.A {type=MovieSceneScriptingFloatChannel, const [1,3]}
         ]
        }]
      }
     ]
    }
    ColorGradientBacking {
     id=#8
     possessed_class=Image
     tracks [
      {
       type=MovieSceneFloatTrack
       display_name=Render Opacity
       sections [{type=MovieSceneFloatSection, channels [None {type=MovieSceneScriptingFloatChannel, keys [[32,0],[61.991,1],[197,0.6]]}]}]
      }
      {
       type=MovieScene2DTransformTrack
       sections [{
         type=MovieScene2DTransformSection
         end_frame=196
         start_frame=32
         channels [
          Translation.X {type=MovieSceneScriptingFloatChannel, keys [[32,-220],[61.991,3],[196.991,-69]]}
          Translation.Y {type=MovieSceneScriptingFloatChannel, const [0,3]}
          Angle {type=MovieSceneScriptingFloatChannel, const [0,3]}
          Scale.X {type=MovieSceneScriptingFloatChannel, keys [[32,0.2],[61.991,1],[196.991,0.75]]}
          Scale.Y {type=MovieSceneScriptingFloatChannel, const [1,3]}
          Shear.X {type=MovieSceneScriptingFloatChannel, const [0,3]}
          Shear.Y {type=MovieSceneScriptingFloatChannel, const [0,3]}
         ]
        }]
      }
     ]
    }
    LEVELUPTEXT {
     id=#9
     possessed_class=CommonTextBlock
     tracks [
      {
       type=MovieSceneFloatTrack
       display_name=Render Opacity
       sections [{type=MovieSceneFloatSection, channels [None {type=MovieSceneScriptingFloatChannel, keys [[39.991,0],[45,1],[162,1],[183,0]]}]}]
      }
      {
       type=MovieSceneColorTrack
       display_name=Color and Opacity
       sections [{
         type=MovieSceneColorSection
         channels [
          Color.R {type=MovieSceneScriptingFloatChannel, keys [[48.991,1],[68.985,1],[79,0],[135,0]]}
          Color.G {type=MovieSceneScriptingFloatChannel, keys [[48.991,1],[68.985,1],[79,0],[135,0]]}
          Color.B {type=MovieSceneScriptingFloatChannel, keys [[48.991,1],[68.985,1],[78.985,0.1302],[79,0],[135,0]]}
          Color.A {type=MovieSceneScriptingFloatChannel, thinned_from=5, keys [[48.991,1],[68.985,1],[79,0.8544],[135,0.8]]}
         ]
        }]
      }
      {
       type=MovieScene2DTransformTrack
       sections [{
         type=MovieScene2DTransformSection
         channels [
          Translation.X {type=MovieSceneScriptingFloatChannel, keys [[39.991,-80],[48.991,0],[162,0],[183,128]]}
          Translation.Y {type=MovieSceneScriptingFloatChannel, const [1,1]}
          Angle {type=MovieSceneScriptingFloatChannel, const [0,1]}
          Scale.X {type=MovieSceneScriptingFloatChannel, keys [[6,1],[40,1],[45,1.075],[60,1],[154.31,1]]}
          Scale.Y {type=MovieSceneScriptingFloatChannel, keys [[6,1],[40,1],[45,1.0262],[60,1],[154.31,1]]}
          Shear.X {type=MovieSceneScriptingFloatChannel, const [-12,1]}
          Shear.Y {type=MovieSceneScriptingFloatChannel, const [0,1]}
         ]
        }]
      }
     ]
    }
    Energy_In {
     id=#10
     possessed_class=Image
     tracks [{
       type=MovieSceneFloatTrack
       display_name=Render Opacity
       sections [{type=MovieSceneFloatSection, channels [None {type=MovieSceneScriptingFloatChannel, keys [[1,0],[21.428,1],[29.222,0]]}]}]
      }]
    }
    Icon {
     id=#11
     possessed_class=Image
     tracks [
      {
       type=MovieSceneFloatTrack
       display_name=Render Opacity
       sections [{type=MovieSceneFloatSection, channels [None {type=MovieSceneScriptingFloatChannel, keys [[19.49,0],[21.437,1]]}]}]
      }
      {
       type=MovieScene2DTransformTrack
       sections [{
         type=MovieScene2DTransformSection
         channels [
          Translation.Y {type=MovieSceneScriptingFloatChannel, keys [[13.646,-144],[21.429,8],[26.294,-16],[36.999,0]]}
          Angle {type=MovieSceneScriptingFloatChannel}
          Scale.X {
           type=MovieSceneScriptingFloatChannel
           keys [[13.646,0.25],[21.428,0.6],[23.374,1.1],[25.315,0.95],[30.181,1],[43.811,1],[46.731,1.25],[55.489,1]]
          }
          Scale.Y {type=MovieSceneScriptingFloatChannel, keys [[13.646,4],[20.455,3],[22.401,0.9],[25.315,1.25],[30.181,1],[43.811,1],[46.731,1.25],[55.489,1]]}
          Shear.X {type=MovieSceneScriptingFloatChannel}
          Shear.Y {type=MovieSceneScriptingFloatChannel}
         ]
        }]
      }
      {
       type=MovieSceneWidgetMaterialTrack
       sections [{
         type=MovieSceneParameterSection
         channels ["Tint Icon (False)" {type=MovieSceneScriptingFloatChannel, keys [[19.488,1],[47.704,1],[67.167,0]]}]
        }]
      }
     ]
    }
    TriFX_3 {
     id=#12
     possessed_class=Image
     tracks [{
       type=MovieSceneWidgetMaterialTrack
       sections [{type=MovieSceneParameterSection, channels [TriScale {type=MovieSceneScriptingFloatChannel, keys [[15.991,0.65],[65.992,0]]}]}]
      }]
    }
    Shadow {
     id=#13
     possessed_class=Image
     tracks [{
       type=MovieSceneFloatTrack
       display_name=Render Opacity
       sections [{type=MovieSceneFloatSection, channels [None {type=MovieSceneScriptingFloatChannel, keys [[30,0],[76.991,1]]}]}]
      }]
    }
    SoftGlow {
     id=#14
     possessed_class=Image
     tracks [{
       type=MovieSceneFloatTrack
       display_name=Render Opacity
       sections [{type=MovieSceneFloatSection, end_frame=91, start_frame=62, channels [None {type=MovieSceneScriptingFloatChannel, keys [[62,0],[91,0.4]]}]}]
      }]
    }
    TriFX_1 {
     id=#15
     possessed_class=Image
     tracks [{
       type=MovieSceneWidgetMaterialTrack
       sections [{type=MovieSceneParameterSection, channels [TriScale {type=MovieSceneScriptingFloatChannel, keys [[15.991,0.65],[65.992,0]]}]}]
      }]
    }
    RedGlow {
     id=#16
     possessed_class=Image
     tracks [
      {
       type=MovieSceneFloatTrack
       display_name=Render Opacity
       sections [{
         type=MovieSceneFloatSection
         channels [None {type=MovieSceneScriptingFloatChannel, keys [[1,0],[6.839,0.125],[31.169,0.05],[43.81,0.05],[47.713,0.5],[109.023,0.125]]}]
        }]
      }
      {
       type=MovieScene2DTransformTrack
       sections [{
         type=MovieScene2DTransformSection
         channels [
          Scale.X {type=MovieSceneScriptingFloatChannel, const [1,1]}
          Scale.Y {type=MovieSceneScriptingFloatChannel, keys [[47.713,1],[109.023,0.666]]}
         ]
        }]
      }
     ]
    }
    Swoosh {
     id=#17
     possessed_class=Image
     tracks [
      {
       type=MovieScene2DTransformTrack
       sections [{
         type=MovieScene2DTransformSection
         channels [
          Translation.X {type=MovieSceneScriptingFloatChannel}
          Translation.Y {type=MovieSceneScriptingFloatChannel, keys [[0.992,-512],[11.99,-248],[19,-32]]}
          Angle {type=MovieSceneScriptingFloatChannel}
          Scale.X {type=MovieSceneScriptingFloatChannel}
          Scale.Y {type=MovieSceneScriptingFloatChannel, thinned_from=4, keys [[0.991,10],[11.99,16],[19,1.3]]}
          Shear.X {type=MovieSceneScriptingFloatChannel}
          Shear.Y {type=MovieSceneScriptingFloatChannel}
         ]
        }]
      }
      {
       type=MovieSceneWidgetMaterialTrack
       sections [{type=MovieSceneParameterSection, channels [Intensity {type=MovieSceneScriptingFloatChannel, keys [[0.991,0],[14,0.9],[16.006,0]]}]}]
      }
      {
       type=MovieSceneFloatTrack
       display_name=Render Opacity
       sections [{
         type=MovieSceneFloatSection
         end_frame=19
         start_frame=1
         channels [None {type=MovieSceneScriptingFloatChannel, keys [[1,0],[5.999,1],[14.01,1],[16.01,0]]}]
        }]
      }
     ]
    }
    RadialGlow {
     id=#18
     possessed_class=Image
     tracks [
      {
       type=MovieSceneFloatTrack
       display_name=Render Opacity
       sections [{
         type=MovieSceneFloatSection
         channels [None {type=MovieSceneScriptingFloatChannel, keys [[1,0],[3.92,1],[15.589,1],[31.16,0],[49.65,0.5],[81.774,0.333],[112.916,0.1]]}]
        }]
      }
      {
       type=MovieScene2DTransformTrack
       sections [{
         type=MovieScene2DTransformSection
         channels [
          Scale.X {
           type=MovieSceneScriptingFloatChannel
           keys [[1,0.05],[10.732,0.125],[17.544,0.25],[31.16,1],[43.811,0.333],[49.65,1.5],[81.765,1],[112.285,1]]
          }
          Scale.Y {
           type=MovieSceneScriptingFloatChannel
           thinned_from=8
           keys [[1,0.05],[10.732,0.125],[17.544,0.25],[31.16,1],[43.811,0.333],[49.65,0.5],[112.285,0.5]]
          }
         ]
        }]
      }
     ]
    }
    AdditiveHorizontalBeam {
     id=#19
     possessed_class=Image
     tracks [
      {
       type=MovieSceneWidgetMaterialTrack
       sections [{type=MovieSceneParameterSection, channels [Animate_Alpha {type=MovieSceneScriptingFloatChannel, keys [[35,0],[55,1.25],[120,1]]}]}]
      }
      {
       type=MovieScene2DTransformTrack
       sections [{type=MovieScene2DTransformSection, channels [Scale.X {type=MovieSceneScriptingFloatChannel, keys [[34.992,0.5],[55,1.324],[120,0.8]]}]}]
      }
     ]
    }
    Particles {
     id=#20
     possessed_class=Image
     tracks [{
       type=MovieSceneFloatTrack
       display_name=Render Opacity
       sections [{type=MovieSceneFloatSection, end_frame=93, start_frame=31, channels [None {type=MovieSceneScriptingFloatChannel, keys [[31,0],[93,1]]}]}]
      }]
    }
    FlareNoise_Large {
     id=#21
     possessed_class=Image
     tracks [
      {
       type=MovieSceneFloatTrack
       display_name=Render Opacity
       sections [{type=MovieSceneFloatSection, channels [None {type=MovieSceneScriptingFloatChannel, keys [[43.811,0],[46.731,1],[60.364,0.7]]}]}]
      }
      {
       type=MovieScene2DTransformTrack
       sections [{
         type=MovieScene2DTransformSection
         channels [
          Scale.X {type=MovieSceneScriptingFloatChannel, keys [[43.81,1],[46.731,1.75],[110.961,1]]}
          Scale.Y {type=MovieSceneScriptingFloatChannel, keys [[43.81,1],[46.731,2],[110.961,1]]}
         ]
        }]
      }
      {
       type=MovieSceneWidgetMaterialTrack
       sections [{
         type=MovieSceneParameterSection
         channels [
          Pow {type=MovieSceneScriptingFloatChannel, const [1.25,3]}
          Mult {type=MovieSceneScriptingFloatChannel, keys [[43.81,4],[46.731,3],[96.363,2]]}
         ]
        }]
      }
     ]
    }
    Beam {
     id=#22
     possessed_class=Image
     tracks [
      {
       type=MovieSceneFloatTrack
       display_name=Render Opacity
       sections [{type=MovieSceneFloatSection, end_frame=77, start_frame=31, channels [None {type=MovieSceneScriptingFloatChannel, keys [[31,0],[77,1]]}]}]
      }
      {
       type=MovieScene2DTransformTrack
       sections [{
         type=MovieScene2DTransformSection
         end_frame=76
         start_frame=31
         channels [
          Translation.X {type=MovieSceneScriptingFloatChannel, keys [[31,-268],[76,-134]]}
          Translation.Y {type=MovieSceneScriptingFloatChannel}
          Angle {type=MovieSceneScriptingFloatChannel}
          Scale.X {type=MovieSceneScriptingFloatChannel, keys [[31,0.4],[76,0.7]]}
          Scale.Y {type=MovieSceneScriptingFloatChannel}
          Shear.X {type=MovieSceneScriptingFloatChannel}
          Shear.Y {type=MovieSceneScriptingFloatChannel}
         ]
        }]
      }
     ]
    }
    SizeBoxTriangles {
     id=#23
     possessed_class=SizeBox
     tracks [
      {
       type=MovieSceneFloatTrack
       display_name=Render Opacity
       sections [{type=MovieSceneFloatSection, channels [None {type=MovieSceneScriptingFloatChannel, keys [[34.991,0],[38.991,0.5],[180,0.125]]}]}]
      }
      {
       type=MovieScene2DTransformTrack
       sections [{
         type=MovieScene2DTransformSection
         channels [
          Translation.X {type=MovieSceneScriptingFloatChannel, keys [[34.991,-256],[62.991,0]]}
          Translation.Y {type=MovieSceneScriptingFloatChannel, const [0,1]}
          Scale.X {type=MovieSceneScriptingFloatChannel, keys [[22.991,1],[34.991,0.75],[42.991,1]]}
          Scale.Y {type=MovieSceneScriptingFloatChannel, keys [[22.991,1],[34.991,0.75],[42.991,1]]}
          Shear.X {type=MovieSceneScriptingFloatChannel}
          Shear.Y {type=MovieSceneScriptingFloatChannel}
         ]
        }]
      }
     ]
    }
    Energy_Out {
     id=#24
     possessed_class=Image
     tracks [
      {
       type=MovieSceneFloatTrack
       display_name=Render Opacity
       sections [{type=MovieSceneFloatSection, channels [None {type=MovieSceneScriptingFloatChannel, keys [[43.811,0],[61.328,1],[255,0]]}]}]
      }
      {
       type=MovieScene2DTransformTrack
       sections [{
         type=MovieScene2DTransformSection
         channels [
          Scale.X {type=MovieSceneScriptingFloatChannel, keys [[43.811,0.5],[67.167,1]]}
          Scale.Y {type=MovieSceneScriptingFloatChannel, keys [[43.811,0.5],[67.167,1]]}
         ]
        }]
      }
     ]
    }
    RiftEffect {
     id=#25
     possessed_class=Image
     tracks [
      {
       type=MovieSceneFloatTrack
       display_name=Render Opacity
       sections [{
         type=MovieSceneFloatSection
         end_frame=55
         start_frame=-1
         channels [None {type=MovieSceneScriptingFloatChannel, keys [[0,0],[40,0],[55,1]]}]
        }]
      }
      {
       type=MovieScene2DTransformTrack
       sections [{
         type=MovieScene2DTransformSection
         end_frame=55
         start_frame=40
         channels [
          Translation.X {type=MovieSceneScriptingFloatChannel, keys [[40,-42],[55,287]]}
          Translation.Y {type=MovieSceneScriptingFloatChannel}
          Angle {type=MovieSceneScriptingFloatChannel}
          Scale.X {type=MovieSceneScriptingFloatChannel, keys [[40,0.45],[55,3.8]]}
          Scale.Y {type=MovieSceneScriptingFloatChannel}
          Shear.X {type=MovieSceneScriptingFloatChannel}
          Shear.Y {type=MovieSceneScriptingFloatChannel}
         ]
        }]
      }
     ]
    }
    OverlayTris {
     id=#26
     possessed_class=Overlay
     tracks [
      {
       type=MovieSceneFloatTrack
       display_name=Render Opacity
       sections [{
         type=MovieSceneFloatSection
         channels [None {type=MovieSceneScriptingFloatChannel, keys [[0.992,0],[12.996,0],[17.996,1],[60.997,1],[62.997,0]]}]
        }]
      }
      {
       type=MovieScene2DTransformTrack
       sections [{
         type=MovieScene2DTransformSection
         channels [
          Translation.X {type=MovieSceneScriptingFloatChannel}
          Translation.Y {type=MovieSceneScriptingFloatChannel}
          Angle {type=MovieSceneScriptingFloatChannel}
          Scale.X {type=MovieSceneScriptingFloatChannel, keys [[12.996,0],[60.996,3]]}
          Scale.Y {type=MovieSceneScriptingFloatChannel, keys [[12.996,0],[60.996,3]]}
          Shear.X {type=MovieSceneScriptingFloatChannel}
          Shear.Y {type=MovieSceneScriptingFloatChannel}
         ]
        }]
      }
     ]
    }
    FlareNoise_Small {
     id=#27
     possessed_class=Image
     tracks [
      {
       type=MovieSceneFloatTrack
       display_name=Render Opacity
       sections [{type=MovieSceneFloatSection, channels [None {type=MovieSceneScriptingFloatChannel, keys [[43.81,0],[46.73,1]]}]}]
      }
      {
       type=MovieSceneWidgetMaterialTrack
       sections [{
         type=MovieSceneParameterSection
         channels [
          Pow {type=MovieSceneScriptingFloatChannel, keys [[43.81,1.25],[46.73,1.5],[96.363,1.25]]}
          Mult {type=MovieSceneScriptingFloatChannel, keys [[43.81,4],[46.73,4],[96.363,2]]}
         ]
        }]
      }
      {
       type=MovieScene2DTransformTrack
       sections [{
         type=MovieScene2DTransformSection
         channels [
          Scale.X {type=MovieSceneScriptingFloatChannel, keys [[43.81,1],[46.731,2],[110.961,1]]}
          Scale.Y {type=MovieSceneScriptingFloatChannel, keys [[43.81,1],[46.731,3],[110.961,1]]}
         ]
        }]
      }
     ]
    }
    Burst {
     id=#28
     possessed_class=Image
     tracks [{
       type=MovieSceneWidgetMaterialTrack
       sections [{
         type=MovieSceneParameterSection
         channels [
          Alpha {type=MovieSceneScriptingFloatChannel, keys [[1,0],[7.997,0],[10.997,1],[15.992,1],[30.992,0]]}
          Grow {type=MovieSceneScriptingFloatChannel, keys [[1,0],[7.997,0],[19.997,1]]}
          RadialDensity {type=MovieSceneScriptingFloatChannel, keys [[1,0],[7.997,0],[13.996,2],[30.999,0.25]]}
          Radius {type=MovieSceneScriptingFloatChannel, keys [[1,0],[7.996,0],[13.996,0.4],[30.992,0.75]]}
          RingAlpha {type=MovieSceneScriptingFloatChannel, keys [[1,0],[7.996,0],[25.996,1]]}
          RingSize {type=MovieSceneScriptingFloatChannel, keys [[1,0],[10.996,0],[25.996,1]]}
          RingThicc {type=MovieSceneScriptingFloatChannel, thinned_from=4, keys [[-10,0],[10.996,0],[25.996,1]]}
          Flare {type=MovieSceneScriptingFloatChannel, keys [[1,0],[5.996,0],[12.996,1],[31,0]]}
          FlareAlpha {type=MovieSceneScriptingFloatChannel, keys [[1,0],[5.997,0],[13.996,1],[19.997,1],[31,0]]}
         ]
        }]
      }]
    }
   ]
   display_rate {numerator=60}
   tracks [{
     type=MovieSceneEventTrack
     display_name=MusicEvents
     sections [{
       type=MovieSceneEventTriggerSection
       channels [None {
         type=MovieSceneScriptingEventChannel
         keys [[74, "<Struct 'MovieSceneEvent' (0x469ad595000) {}>"], [120, "<Struct 'MovieSceneEvent' (0x469ad5955a0) {}>"]]
        }]
      }]
    }]
  }
 }
]
graphs [
 DispatchComplete {nodes [DispatchComplete() {is=title K2Node_FunctionEntry}]}
 DispatchShowReward {nodes [DispatchShowReward() {is=title K2Node_FunctionEntry}]}
 EventGraph {nodes [
   EventStartAnim() {is=title K2Node_CustomEvent}
   > PlayAnimation(InAnimation=Get NewIntro() -> (NewIntro:Widget Animation Object Reference))
   n1 = EventIntroAnimComplete() {is=title K2Node_CustomEvent}
   > PlayAnimation(InAnimation=Get NewOutro() -> (NewOutro:Widget Animation Object Reference))
   n2 = EventComplete() {is=title K2Node_CustomEvent}
   > Call Dispatch Complete!(execute:Exec, self:Level Up Visuals Object Reference) -> (then:Exec) {is=title K2Node_CallDelegate}
   n3 = Play XP Intro Sound() {is=title K2Node_CustomEvent}
   > Delay(Duration=0.2)
   > SpawnSound2D!(
    execute
    Sound:Sound Base Object Reference=Get SFX_LevelUp() -> (SFX_LevelUp:Sound Base Object Reference)
    VolumeMultiplier:float=1
    PitchMultiplier:float=1
    StartTime:float=0
    ConcurrencySettings:Sound Concurrency Object Reference
    bPersistAcrossLevelTransition:bool=false
    bAutoDestroy:bool=true
   ) -> (then, ReturnValue:Audio Component Object Reference)
   EventLayeredPrestige() {is=title K2Node_CustomEvent}
   > Branch(Condition="integer > integer"(A=Get Level() -> (Level:int=0), B=100) {is=title K2Node_CallFunction})
   > PlayAnimation(InAnimation=Get Hah() -> (Hah:Widget Animation Object Reference))
   Event OnInitialized()
   > Branch(Condition=IsMobileGame())
   > SetRenderScale!(execute, self:Widget Object Reference=Get IconTint() -> (IconTint:Border Object Reference), Scale:Vector 2D Structure=(X=1.5,Y=1.5)) -> (then)
   Event Construct()
   > BindToAnimationFinished(Animation=Get NewIntro(), Delegate=@n1)
   > BindToAnimationFinished(Animation=Get NewOutro(), Delegate=@n2)
   > BindToAnimationStarted!(execute, self:User Widget Object Reference, Animation:Widget Animation Object Reference=Get NewIntro(), Delegate:Delegate=@n3) -> (then)
  ]}
 FireLevelReward {nodes [
   FireLevelReward() {is=title K2Node_FunctionEntry}
   > PrintString!(execute, InString="Enjoy your level rewards, cool guy", bPrintToScreen, bPrintToLog, TextColor=(R=1,G=0.3237,B=0,A=1), Duration, Key) -> (then)
  ]}
 Post100LayeredAnim {nodes [
   Post100LayeredAnim() {is=title K2Node_FunctionEntry}
   > EventLayeredPrestige() {is=title K2Node_CallFunction}
  ]}
 ResetAnimationState {nodes [
   ResetAnimationState() {is=title K2Node_FunctionEntry}
   n1 = Get NewIntro()
   > PlayAnimation(InAnimation=@n1, StartAtTime=GetEndTime(self=@n1), PlayMode=Reverse)
  ]}
 Sequencer Events {nodes [BringBackWorldMusic() {is=title K2Node_CustomEvent}]}
]
variables [
 DispatchComplete {}
 DispatchShowReward {}
 Level {type=integer}
 SFX_LevelUp {type=SoundBase, default=/Game/Sounds/Fort_UI_Sounds/Gameplay/XP/MSS_XP_LevelUp}
]
widgets {
 info {inheritedWidgetCount=0, parentClass=&UserWidget, rootWidgetClass=&Overlay, widgetCount=41}
 widgets [
  Overlay_2 {widgetClassPath=&Overlay, widget=Overlay_2, properties {visibility=HitTestInvisible}}
  All {
   widgetClassPath=&SizeBox
   parent=Overlay_2
   widget=All
   properties {bOverride_HeightOverride=true, bOverride_WidthOverride=true, heightOverride=64, widthOverride=896}
   slot {class=OverlaySlot, properties {verticalAlignment=VAlign_Center}}
  }
  Overlay_TopContainer {
   widgetClassPath=&Overlay
   parent=All
   widget=Overlay_TopContainer
   properties {visibility=HitTestInvisible}
   slot {class=SizeBoxSlot}
  }
  BackingElements {
   widgetClassPath=&Overlay
   bIsVariable=true
   parent=Overlay_TopContainer
   widget=BackingElements
   properties {renderTransform {angle=0, scale=<1.2 1>, shear=<0 0>, translation=<-32 0>}}
   slot {class=OverlaySlot, properties {horizontalAlignment=HAlign_Fill, verticalAlignment=VAlign_Fill}}
  }
  Shadow {
   widgetClassPath=&Image
   bIsVariable=true
   parent=BackingElements
   widget=Shadow
   properties {
    visibility=SelfHitTestInvisible
    brush {
     bIsDynamicallyLoaded=false
     drawAs=Image
     imageType=NoImage
     mirroring=NoMirror
     resourceName=None
     resourceObject=/Game/Athena/HUD/XP/Materials/M_UI_LevelUP_Darken
     tiling=NoTile
     imageSize=<640 32>
     margin=<ltrb 0 0 0 0>
     outlineSettings {
      bUseBrushTransparency=false
      roundingType=HalfHeightRadius
      width=0
      color {colorUseRule=UseColor_Specified, specifiedColor=<rgba 0 0 0 0>}
      cornerRadii=<0 0 0 0>
     }
     tintColor {colorUseRule=UseColor_Specified, specifiedColor=<rgba 1 1 1 1>}
     uVRegion {bIsValid=false, max=<0 0>, min=<0 0>}
    }
    renderTransform {angle=0, scale=<1 3>, shear=<0 0>, translation=<0 0>}
   }
   slot {class=OverlaySlot, properties {verticalAlignment=VAlign_Fill}}
  }
  Beam {
   widgetClassPath=&Image
   bIsVariable=true
   parent=BackingElements
   widget=Beam
   properties {
    brush {
     bIsDynamicallyLoaded=false
     drawAs=Image
     imageType=FullColor
     mirroring=NoMirror
     resourceName=None
     resourceObject=/Game/Athena/HUD/XP/Materials/M_UI_LevelUP_Geadient_Beam_Soft
     tiling=NoTile
     imageSize=<32 32>
     margin=<ltrb 0 0 0 0>
     outlineSettings {
      bUseBrushTransparency=false
      roundingType=HalfHeightRadius
      width=0
      color {colorUseRule=UseColor_Specified, specifiedColor=<rgba 0 0 0 0>}
      cornerRadii=<0 0 0 0>
     }
     tintColor {colorUseRule=UseColor_Specified, specifiedColor=<rgba 1 1 1 1>}
     uVRegion {bIsValid=false, max=<0 0>, min=<0 0>}
    }
    renderTransform {angle=0, scale=<0.7 1>, shear=<0 0>, translation=<-134 0>}
   }
   slot {class=OverlaySlot, properties {horizontalAlignment=HAlign_Fill, verticalAlignment=VAlign_Fill}}
  }
  BeamSoft {
   widgetClassPath=&Image
   bIsVariable=true
   parent=BackingElements
   widget=BeamSoft
   properties {
    renderOpacity=0.5
    brush {
     bIsDynamicallyLoaded=false
     drawAs=Image
     imageType=FullColor
     mirroring=NoMirror
     resourceName=None
     resourceObject=&M_UI_LevelUP_Geadient_Beam_Soft
     tiling=NoTile
     imageSize=<32 32>
     margin=<ltrb 0 0 0 0>
     outlineSettings {
      bUseBrushTransparency=false
      roundingType=HalfHeightRadius
      width=0
      color {colorUseRule=UseColor_Specified, specifiedColor=<rgba 0 0 0 0>}
      cornerRadii=<0 0 0 0>
     }
     tintColor {colorUseRule=UseColor_Specified, specifiedColor=<rgba 1 1 1 1>}
     uVRegion {bIsValid=false, max=<0 0>, min=<0 0>}
    }
    renderTransform {angle=0, scale=<0.4 3.8>, shear=<0 0>, translation=<-259 0>}
   }
   slot {class=OverlaySlot, properties {horizontalAlignment=HAlign_Fill, verticalAlignment=VAlign_Fill}}
  }
  ColorGradientBacking {
   widgetClassPath=&Image
   bIsVariable=true
   parent=BackingElements
   widget=ColorGradientBacking
   properties {
    renderOpacity=0.6
    visibility=SelfHitTestInvisible
    brush {
     bIsDynamicallyLoaded=false
     drawAs=Image
     imageType=FullColor
     mirroring=NoMirror
     resourceName=None
     resourceObject=/Game/Athena/HUD/XP/Materials/M_UI_LevelUP_Gradient_MultiProduct
     tiling=NoTile
     imageSize=<576 32>
     margin=<ltrb 0 0 0 0>
     outlineSettings {
      bUseBrushTransparency=false
      roundingType=HalfHeightRadius
      width=0
      color {colorUseRule=UseColor_Specified, specifiedColor=<rgba 0 0 0 0>}
      cornerRadii=<0 0 0 0>
     }
     tintColor {colorUseRule=UseColor_Specified, specifiedColor=<rgba 1 1 1 0.5>}
     uVRegion {bIsValid=false, max=<0 0>, min=<0 0>}
    }
    renderTransform {angle=0, scale=<0.2 1>, shear=<0 0>, translation=<-220 0>}
   }
   slot {class=OverlaySlot, properties {verticalAlignment=VAlign_Fill}}
  }
  RiftEffect {
   widgetClassPath=&Image
   bIsVariable=true
   parent=BackingElements
   widget=RiftEffect
   properties {
    renderOpacity=0
    brush {
     bIsDynamicallyLoaded=false
     drawAs=Image
     imageType=FullColor
     mirroring=NoMirror
     resourceName=None
     resourceObject=/Game/Athena/HUD/XP/Materials/MI_UI_RiftEffect_Tendrils_2
     tiling=NoTile
     imageSize=<200 32>
     margin=<ltrb 0 0 0 0>
     outlineSettings {
      bUseBrushTransparency=false
      roundingType=HalfHeightRadius
      width=0
      color {colorUseRule=UseColor_Specified, specifiedColor=<rgba 0 0 0 0>}
      cornerRadii=<0 0 0 0>
     }
     tintColor {colorUseRule=UseColor_Specified, specifiedColor=<rgba 1 1 1 0.8>}
     uVRegion {bIsValid=false, max=<0 0>, min=<0 0>}
    }
    colorAndOpacity=<rgba 1 1 1 0.9>
    renderTransform {angle=0, scale=<3.7573 1.5>, shear=<0 0>, translation=<282.8083 4>}
   }
   slot {class=OverlaySlot, properties {verticalAlignment=VAlign_Fill}}
  }
  SoftGlow {
   widgetClassPath=&Image
   bIsVariable=true
   parent=BackingElements
   widget=SoftGlow
   properties {
    renderOpacity=0.4
    brush {
     bIsDynamicallyLoaded=false
     drawAs=Image
     imageType=FullColor
     mirroring=NoMirror
     resourceName=None
     resourceObject=/Game/Athena/HUD/XP/Materials/M_UI_LevelUP_Gradient_Soft
     tiling=NoTile
     imageSize=<32 139.74>
     margin=<ltrb 0 0 0 0>
     outlineSettings {
      bUseBrushTransparency=false
      roundingType=HalfHeightRadius
      width=0
      color {colorUseRule=UseColor_Specified, specifiedColor=<rgba 0 0 0 0>}
      cornerRadii=<0 0 0 0>
     }
     tintColor {colorUseRule=UseColor_Specified, specifiedColor=<rgba 1 1 1 0.3>}
     uVRegion {bIsValid=false, max=<0 0>, min=<0 0>}
    }
    renderTransform {angle=0, scale=<0.95 1>, shear=<0 0>, translation=<-188 0>}
   }
   slot {class=OverlaySlot, properties {horizontalAlignment=HAlign_Fill, verticalAlignment=VAlign_Center}}
  }
  SizeBoxParticles {
   widgetClassPath=&SizeBox
   parent=BackingElements
   widget=SizeBoxParticles
   properties {bOverride_WidthOverride=true, widthOverride=300}
   slot {class=OverlaySlot, properties {verticalAlignment=VAlign_Fill}}
  }
  Particles {
   widgetClassPath=&Image
   bIsVariable=true
   parent=SizeBoxParticles
   widget=Particles
   properties {
    brush {
     bIsDynamicallyLoaded=false
     drawAs=Image
     imageType=FullColor
     mirroring=NoMirror
     resourceName=None
     resourceObject=/Game/Athena/HUD/XP/Materials/MI_UI_LevelUp_Particles_Inst
     tiling=NoTile
     imageSize=<800 32>
     margin=<ltrb 0 0 0 0>
     outlineSettings {
      bUseBrushTransparency=false
      roundingType=HalfHeightRadius
      width=0
      color {colorUseRule=UseColor_Specified, specifiedColor=<rgba 0 0 0 0>}
      cornerRadii=<0 0 0 0>
     }
     tintColor {colorUseRule=UseColor_Specified, specifiedColor=<rgba 1 1 1 1>}
     uVRegion {bIsValid=false, max=<0 0>, min=<0 0>}
    }
    renderTransform {angle=0, scale=<1.8 1.8>, shear=<0 0>, translation=<52 0>}
   }
   slot {class=SizeBoxSlot}
  }
  SizeBoxTriangles {
   widgetClassPath=&SizeBox
   parent=BackingElements
   widget=SizeBoxTriangles
   properties {heightOverride=223, renderOpacity=0.1875, widthOverride=214.3544, renderTransformPivot=<0 0.5>}
   slot {class=OverlaySlot, properties {horizontalAlignment=HAlign_Fill, verticalAlignment=VAlign_Fill}}
  }
  ImageTriangleR {
   widgetClassPath=&Image
   parent=SizeBoxTriangles
   widget=ImageTriangleR
   properties {
    visibility=SelfHitTestInvisible
    brush {
     bIsDynamicallyLoaded=false
     drawAs=Image
     imageType=NoImage
     mirroring=NoMirror
     resourceName=None
     resourceObject=/Game/UI/HUD/Materials/M_FX_TriangleSpiral_LevelUp
     tiling=NoTile
     imageSize=<640 85>
     margin=<ltrb 0 0 0 0>
     outlineSettings {
      bUseBrushTransparency=false
      roundingType=HalfHeightRadius
      width=0
      color {colorUseRule=UseColor_Specified, specifiedColor=<rgba 0 0 0 0>}
      cornerRadii=<0 0 0 0>
     }
     tintColor {colorUseRule=UseColor_Specified, specifiedColor=<rgba 0.941 0.2033 1 1>}
     uVRegion {bIsValid=false, max=<0 0>, min=<0 0>}
    }
    renderTransform {angle=0, scale=<1.125 1>, shear=<0 0>, translation=<0 0>}
    renderTransformPivot=<0 1>
   }
   slot {class=SizeBoxSlot}
  }
  AdditiveHorizontalBeam {
   widgetClassPath=&Image
   bIsVariable=true
   parent=BackingElements
   widget=AdditiveHorizontalBeam
   properties {
    renderOpacity=0.5
    visibility=SelfHitTestInvisible
    brush {
     bIsDynamicallyLoaded=false
     drawAs=Image
     imageType=FullColor
     mirroring=NoMirror
     resourceName=None
     resourceObject=/Game/Athena/HUD/XP/Materials/M_UI_LevelUP_Rift
     tiling=NoTile
     imageSize=<640 32>
     margin=<ltrb 0 0 0 0>
     outlineSettings {
      bUseBrushTransparency=false
      roundingType=HalfHeightRadius
      width=0
      color {colorUseRule=UseColor_Specified, specifiedColor=<rgba 0 0 0 0>}
      cornerRadii=<0 0 0 0>
     }
     tintColor {colorUseRule=UseColor_Specified, specifiedColor=<rgba 1 1 1 0.5>}
     uVRegion {bIsValid=false, max=<0 0>, min=<0 0>}
    }
    renderTransform {angle=0, scale=<1.45 1>, shear=<0 0>, translation=<0 0>}
    renderTransformPivot=<0 0.5>
   }
   slot {class=OverlaySlot, properties {horizontalAlignment=HAlign_Fill, verticalAlignment=VAlign_Fill}}
  }
  Overlay_1 {
   widgetClassPath=&Overlay
   parent=Overlay_TopContainer
   widget=Overlay_1
   slot {class=OverlaySlot, properties {horizontalAlignment=HAlign_Fill, verticalAlignment=VAlign_Center, padding=<ltrb 10 0 0 0>}}
  }
  Overlay_0 {
   widgetClassPath=&Overlay
   parent=Overlay_1
   widget=Overlay_0
   slot {class=OverlaySlot, properties {verticalAlignment=VAlign_Center}}
  }
  RedGlow {
   widgetClassPath=&Image
   bIsVariable=true
   parent=Overlay_0
   widget=RedGlow
   properties {
    renderOpacity=0.2
    brush {
     bIsDynamicallyLoaded=false
     drawAs=Image
     imageType=FullColor
     mirroring=NoMirror
     resourceName=None
     resourceObject=/Game/Athena/HUD/LTM/Wax/Materials/M_UI_TextGlow_Add_Gradient_Orange_LEVELUP_MultiPlatform
     tiling=NoTile
     imageSize=<512 500>
     margin=<ltrb 0 0 0 0>
     outlineSettings {
      bUseBrushTransparency=false
      roundingType=HalfHeightRadius
      width=0
      color {colorUseRule=UseColor_Specified, specifiedColor=<rgba 0 0 0 0>}
      cornerRadii=<0 0 0 0>
     }
     tintColor {colorUseRule=UseColor_Specified, specifiedColor=<rgba 1 1 1 1>}
     uVRegion {bIsValid=false, max=<0 0>, min=<0 0>}
    }
   }
   slot {class=OverlaySlot, properties {horizontalAlignment=HAlign_Center, verticalAlignment=VAlign_Center, padding=<ltrb -512 -512 -512 -512>}}
  }
  FlareNoise_Large {
   widgetClassPath=&Image
   bIsVariable=true
   parent=Overlay_0
   widget=FlareNoise_Large
   properties {
    renderOpacity=0.6
    visibility=SelfHitTestInvisible
    brush {
     bIsDynamicallyLoaded=false
     drawAs=Image
     imageType=FullColor
     mirroring=NoMirror
     resourceName=None
     resourceObject=/Game/Athena/HUD/XP/Materials/M_UI_LevelUp_LightShaft_Inst_Rift
     tiling=NoTile
     imageSize=<256 144>
     margin=<ltrb 0 0 0 0>
     outlineSettings {
      bUseBrushTransparency=false
      roundingType=HalfHeightRadius
      width=0
      color {colorUseRule=UseColor_Specified, specifiedColor=<rgba 0 0 0 0>}
      cornerRadii=<0 0 0 0>
     }
     tintColor {colorUseRule=UseColor_Specified, specifiedColor=<rgba 1 1 1 1>}
     uVRegion {bIsValid=false, max=<0 0>, min=<0 0>}
    }
   }
   slot {class=OverlaySlot, properties {horizontalAlignment=HAlign_Center, verticalAlignment=VAlign_Center, padding=<ltrb -256 -256 -256 -256>}}
  }
  FlareNoise_Small {
   widgetClassPath=&Image
   bIsVariable=true
   parent=Overlay_0
   widget=FlareNoise_Small
   properties {
    renderOpacity=0.5
    visibility=SelfHitTestInvisible
    brush {
     bIsDynamicallyLoaded=false
     drawAs=Image
     imageType=FullColor
     mirroring=NoMirror
     resourceName=None
     resourceObject=&M_UI_LevelUp_LightShaft_Inst_Rift
     tiling=NoTile
     imageSize=<96 96>
     margin=<ltrb 0 0 0 0>
     outlineSettings {
      bUseBrushTransparency=false
      roundingType=HalfHeightRadius
      width=0
      color {colorUseRule=UseColor_Specified, specifiedColor=<rgba 0 0 0 0>}
      cornerRadii=<0 0 0 0>
     }
     tintColor {colorUseRule=UseColor_Specified, specifiedColor=<rgba 1 1 1 1>}
     uVRegion {bIsValid=false, max=<0 0>, min=<0 0>}
    }
    renderTransform {angle=0, scale=<1 -1>, shear=<0 0>, translation=<0 0>}
   }
   slot {class=OverlaySlot, properties {horizontalAlignment=HAlign_Center, verticalAlignment=VAlign_Center, padding=<ltrb -256 -256 -256 -256>}}
  }
  Energy_Out {
   widgetClassPath=&Image
   bIsVariable=true
   parent=Overlay_0
   widget=Energy_Out
   properties {
    visibility=SelfHitTestInvisible
    brush {
     bIsDynamicallyLoaded=false
     drawAs=Image
     imageType=FullColor
     mirroring=NoMirror
     resourceName=None
     resourceObject=/Game/Athena/UI/InGame/Interrogation/Materials/M_UI_Starburst_LevelUP_white
     tiling=NoTile
     imageSize=<512 384>
     margin=<ltrb 0 0 0 0>
     outlineSettings {
      bUseBrushTransparency=false
      roundingType=HalfHeightRadius
      width=0
      color {colorUseRule=UseColor_Specified, specifiedColor=<rgba 0 0 0 0>}
      cornerRadii=<0 0 0 0>
     }
     tintColor {colorUseRule=UseColor_Specified, specifiedColor=<rgba 0.8196 0.3582 1 1>}
     uVRegion {bIsValid=false, max=<0 0>, min=<0 0>}
    }
   }
   slot {class=OverlaySlot, properties {horizontalAlignment=HAlign_Center, verticalAlignment=VAlign_Center, padding=<ltrb -512 -512 -512 -512>}}
  }
  Energy_In {
   widgetClassPath=&Image
   bIsVariable=true
   parent=Overlay_0
   widget=Energy_In
   properties {
    visibility=SelfHitTestInvisible
    brush {
     bIsDynamicallyLoaded=false
     drawAs=Image
     imageType=FullColor
     mirroring=NoMirror
     resourceName=None
     resourceObject=/Game/Athena/UI/InGame/Interrogation/Materials/M_UI_Starburst_LevelUP_Inward_White
     tiling=NoTile
     imageSize=<640 384>
     margin=<ltrb 0 0 0 0>
     outlineSettings {
      bUseBrushTransparency=false
      roundingType=HalfHeightRadius
      width=0
      color {colorUseRule=UseColor_Specified, specifiedColor=<rgba 0 0 0 0>}
      cornerRadii=<0 0 0 0>
     }
     tintColor {colorUseRule=UseColor_Specified, specifiedColor=<rgba 1 1 1 1>}
     uVRegion {bIsValid=false, max=<0 0>, min=<0 0>}
    }
   }
   slot {class=OverlaySlot, properties {horizontalAlignment=HAlign_Center, verticalAlignment=VAlign_Center, padding=<ltrb -512 -512 -512 -512>}}
  }
  RadialGlow {
   widgetClassPath=&Image
   bIsVariable=true
   parent=Overlay_0
   widget=RadialGlow
   properties {
    renderOpacity=0
    brush {
     bIsDynamicallyLoaded=false
     drawAs=Image
     imageType=FullColor
     mirroring=NoMirror
     resourceName=None
     resourceObject=/Game/Athena/HUD/XP/Materials/M_UI_LevelUP_Gradient_Glow_White
     tiling=NoTile
     imageSize=<256 256>
     margin=<ltrb 0 0 0 0>
     outlineSettings {
      bUseBrushTransparency=false
      roundingType=HalfHeightRadius
      width=0
      color {colorUseRule=UseColor_Specified, specifiedColor=<rgba 0 0 0 0>}
      cornerRadii=<0 0 0 0>
     }
     tintColor {colorUseRule=UseColor_Specified, specifiedColor=<rgba 1 1 1 1>}
     uVRegion {bIsValid=false, max=<0 0>, min=<0 0>}
    }
   }
   slot {class=OverlaySlot, properties {horizontalAlignment=HAlign_Center, verticalAlignment=VAlign_Center, padding=<ltrb -256 -256 -256 -256>}}
  }
  HorizontalFlare {
   widgetClassPath=&Image
   bIsVariable=true
   parent=Overlay_0
   widget=HorizontalFlare
   properties {
    renderOpacity=0
    visibility=SelfHitTestInvisible
    brush {
     bIsDynamicallyLoaded=false
     drawAs=Image
     imageType=FullColor
     mirroring=NoMirror
     resourceName=None
     resourceObject=/Game/Athena/HUD/XP/Materials/M_UI_AnimatedLevelUpGlow_Rift
     tiling=NoTile
     imageSize=<640 32>
     margin=<ltrb 0 0 0 0>
     outlineSettings {
      bUseBrushTransparency=false
      roundingType=HalfHeightRadius
      width=0
      color {colorUseRule=UseColor_Specified, specifiedColor=<rgba 0 0 0 0>}
      cornerRadii=<0 0 0 0>
     }
     tintColor {colorUseRule=UseColor_Foreground, specifiedColor=<rgba 1 1 1 1>}
     uVRegion {bIsValid=false, max=<0 0>, min=<0 0>}
    }
    renderTransform {angle=0, scale=<1 1.1011>, shear=<0 0>, translation=<94 0>}
   }
   slot {class=OverlaySlot, properties {horizontalAlignment=HAlign_Center, verticalAlignment=VAlign_Center, padding=<ltrb -640 -640 -640 -640>}}
  }
  IconTint {
   widgetClassPath=&Border
   bIsVariable=true
   parent=Overlay_0
   widget=IconTint
   properties {
    background {
     bIsDynamicallyLoaded=false
     drawAs=NoDrawType
     imageType=NoImage
     mirroring=NoMirror
     resourceName=None
     resourceObject=null
     tiling=NoTile
     imageSize=<32 32>
     margin=<ltrb 0 0 0 0>
     outlineSettings {
      bUseBrushTransparency=false
      roundingType=HalfHeightRadius
      width=0
      color {colorUseRule=UseColor_Specified, specifiedColor=<rgba 0 0 0 0>}
      cornerRadii=<0 0 0 0>
     }
     tintColor {colorUseRule=UseColor_Specified, specifiedColor=<rgba 1 1 1 1>}
     uVRegion {bIsValid=false, max=<0 0>, min=<0 0>}
    }
    padding=<ltrb 0 0 0 0>
   }
   slot {class=OverlaySlot, properties {verticalAlignment=VAlign_Center}}
  }
  Icon {
   widgetClassPath=&Image
   bIsVariable=true
   parent=IconTint
   widget=Icon
   properties {brush {
     bIsDynamicallyLoaded=false
     drawAs=Image
     imageType=FullColor
     mirroring=NoMirror
     resourceName=None
     resourceObject=/Game/UI/Foundation/XPBoost/MI_UI_UniversalXP
     tiling=NoTile
     imageSize=<56 56>
     margin=<ltrb 0 0 0 0>
     outlineSettings {
      bUseBrushTransparency=false
      roundingType=HalfHeightRadius
      width=0
      color {colorUseRule=UseColor_Specified, specifiedColor=<rgba 0 0 0 0>}
      cornerRadii=<0 0 0 0>
     }
     tintColor {colorUseRule=UseColor_Specified, specifiedColor=<rgba 1 1 1 1>}
     uVRegion {bIsValid=false, max=<0 0>, min=<0 0>}
    }}
   slot {class=BorderSlot, properties {padding=<ltrb 0 0 0 0>}}
  }
  AnimBoundLevelBox {
   widgetClassPath=&Border
   bIsVariable=true
   parent=Overlay_1
   widget=AnimBoundLevelBox
   properties {
    renderOpacity=0
    visibility=SelfHitTestInvisible
    background {
     bIsDynamicallyLoaded=false
     drawAs=NoDrawType
     imageType=NoImage
     mirroring=NoMirror
     resourceName=None
     resourceObject=null
     tiling=NoTile
     imageSize=<32 32>
     margin=<ltrb 0 0 0 0>
     outlineSettings {
      bUseBrushTransparency=false
      roundingType=HalfHeightRadius
      width=0
      color {colorUseRule=UseColor_Specified, specifiedColor=<rgba 0 0 0 0>}
      cornerRadii=<0 0 0 0>
     }
     tintColor {colorUseRule=UseColor_Specified, specifiedColor=<rgba 1 1 1 1>}
     uVRegion {bIsValid=false, max=<0 0>, min=<0 0>}
    }
    padding=<ltrb 0 0 0 0>
   }
   slot {class=OverlaySlot, properties {verticalAlignment=VAlign_Center, padding=<ltrb 72 -2 0 0>}}
  }
  LevelNumberBox {
   widgetClassPath=&HorizontalBox
   parent=AnimBoundLevelBox
   widget=LevelNumberBox
   slot {class=BorderSlot, properties {padding=<ltrb 0 0 0 0>}}
  }
  LevelNumber_1 {
   widgetClassPath=&CommonTextBlock
   bIsVariable=true
   parent=LevelNumberBox
   widget=LevelNumber_1
   properties {
    mobileFontSizeMultiplier=1.6
    text=LVL
    textTransformPolicy=ToUpper
    colorAndOpacity {colorUseRule=UseColor_Specified, specifiedColor=<rgba 1 1 1 0.9>}
    font {
     bForceMonospaced=false
     bMaterialIsStencil=false
     fontMaterial=null
     fontObject=&HeadingNow
     letterSpacing=-60
     monospacedWidth=1
     size=19.5
     skewAmount=0
     typefaceFontName=86 Bold
     outlineSettings {
      bApplyOutlineToDropShadows=false
      bMiteredCorners=false
      bSeparateFillAlpha=false
      outlineMaterial=/Game/Athena/HUD/LTM/Wax/Materials/M_UI_TextGlow_Add_Gradient_Orange_LEVELUP
      outlineSize=0
      outlineColor=<rgba 1 1 1 1>
     }
    }
    renderTransform {angle=0, scale=<1 1>, shear=<-12 0>, translation=<0 0>}
    shadowOffset=<0 0>
   }
   slot {class=HorizontalBoxSlot, properties {horizontalAlignment=HAlign_Left, verticalAlignment=VAlign_Center}}
  }
  LevelNumber {
   widgetClassPath=&CommonTextBlock
   bIsVariable=true
   parent=LevelNumberBox
   widget=LevelNumber
   properties {
    mobileFontSizeMultiplier=1.2
    text=222
    textTransformPolicy=ToUpper
    colorAndOpacity {colorUseRule=UseColor_Specified, specifiedColor=<rgba 1 1 1 0.9>}
    font {
     bForceMonospaced=false
     bMaterialIsStencil=false
     fontMaterial=null
     fontObject=&HeadingNow
     letterSpacing=0
     monospacedWidth=1
     size=19.5
     skewAmount=0
     typefaceFontName=86 Bold
     outlineSettings {
      bApplyOutlineToDropShadows=false
      bMiteredCorners=false
      bSeparateFillAlpha=false
      outlineMaterial=&M_UI_TextGlow_Add_Gradient_Orange_LEVELUP
      outlineSize=0
      outlineColor=<rgba 1 1 1 1>
     }
    }
    renderTransform {angle=0, scale=<1 1>, shear=<-12 0>, translation=<0 0>}
    shadowOffset=<0 0>
   }
   slot {class=HorizontalBoxSlot, properties {horizontalAlignment=HAlign_Left, verticalAlignment=VAlign_Center, padding=<ltrb 6 0 0 0>}}
  }
  LEVELUPTEXT {
   widgetClassPath=&CommonTextBlock
   bIsVariable=true
   parent=Overlay_1
   widget=LEVELUPTEXT
   properties {
    mobileFontSizeMultiplier=1.2
    text="Level Up!"
    textTransformPolicy=ToUpper
    colorAndOpacity {colorUseRule=UseColor_Specified, specifiedColor=<rgba 0 0 0 0.9>}
    font {
     bForceMonospaced=false
     bMaterialIsStencil=false
     fontMaterial=null
     fontObject=&HeadingNow
     letterSpacing=128
     monospacedWidth=1
     size=19.5
     skewAmount=0
     typefaceFontName=86 Bold
     outlineSettings {
      bApplyOutlineToDropShadows=false
      bMiteredCorners=false
      bSeparateFillAlpha=false
      outlineMaterial=&M_UI_TextGlow_Add_Gradient_Orange_LEVELUP
      outlineSize=0
      outlineColor=<rgba 1 1 1 1>
     }
    }
    renderTransform {angle=0, scale=<1 1>, shear=<-12 0>, translation=<1 1>}
    renderTransformPivot=<0 0.5>
    shadowOffset=<0 0>
   }
   slot {class=OverlaySlot, properties {verticalAlignment=VAlign_Center, padding=<ltrb 72 -2 0 0>}}
  }
  OverlayFX {
   widgetClassPath=&Overlay
   bIsVariable=true
   parent=Overlay_TopContainer
   widget=OverlayFX
   slot {class=OverlaySlot, properties {verticalAlignment=VAlign_Center, padding=<ltrb 24 -6 0 0>}}
  }
  SizeBoxSwoosh {
   widgetClassPath=&SizeBox
   parent=OverlayFX
   widget=SizeBoxSwoosh
   properties {bOverride_HeightOverride=true, bOverride_WidthOverride=true, heightOverride=24, visibility=Collapsed, widthOverride=24}
   slot {class=OverlaySlot, properties {horizontalAlignment=HAlign_Center, verticalAlignment=VAlign_Center, padding=<ltrb 0 0 60 0>}}
  }
  Swoosh {
   widgetClassPath=&Image
   bIsVariable=true
   parent=SizeBoxSwoosh
   widget=Swoosh
   properties {
    renderOpacity=0
    visibility=SelfHitTestInvisible
    brush {
     bIsDynamicallyLoaded=false
     drawAs=Image
     imageType=FullColor
     mirroring=NoMirror
     resourceName=None
     resourceObject=&M_UI_LevelUP_Gradient_Soft
     tiling=NoTile
     imageSize=<32 32>
     margin=<ltrb 0 0 0 0>
     outlineSettings {
      bUseBrushTransparency=false
      roundingType=HalfHeightRadius
      width=0
      color {colorUseRule=UseColor_Specified, specifiedColor=<rgba 0 0 0 0>}
      cornerRadii=<0 0 0 0>
     }
     tintColor {colorUseRule=UseColor_Specified, specifiedColor=<rgba 1 1 1 1>}
     uVRegion {bIsValid=false, max=<0 0>, min=<0 0>}
    }
    renderTransform {angle=180, scale=<2 1.3>, shear=<0 0>, translation=<0 -32>}
    renderTransformPivot=<0.5 1>
   }
   slot {class=SizeBoxSlot, properties {horizontalAlignment=HAlign_Center, verticalAlignment=VAlign_Center}}
  }
  SizeBoxBurst {
   widgetClassPath=&SizeBox
   parent=OverlayFX
   widget=SizeBoxBurst
   properties {bOverride_HeightOverride=true, bOverride_WidthOverride=true, heightOverride=24, widthOverride=24}
   slot {class=OverlaySlot, properties {horizontalAlignment=HAlign_Right, verticalAlignment=VAlign_Center, padding=<ltrb 0 0 120 0>}}
  }
  Burst {
   widgetClassPath=&Image
   bIsVariable=true
   parent=SizeBoxBurst
   widget=Burst
   properties {
    visibility=SelfHitTestInvisible
    brush {
     bIsDynamicallyLoaded=false
     drawAs=Image
     imageType=FullColor
     mirroring=NoMirror
     resourceName=None
     resourceObject=/Game/UI/Frontend/HeroLoadout/Art/Materials/FX/M_UI_HUD_LevelupBrst_Rift
     tiling=NoTile
     imageSize=<32 32>
     margin=<ltrb 0 0 0 0>
     outlineSettings {
      bUseBrushTransparency=false
      roundingType=HalfHeightRadius
      width=0
      color {colorUseRule=UseColor_Specified, specifiedColor=<rgba 0 0 0 0>}
      cornerRadii=<0 0 0 0>
     }
     tintColor {colorUseRule=UseColor_Specified, specifiedColor=<rgba 1 1 1 1>}
     uVRegion {bIsValid=false, max=<0 0>, min=<0 0>}
    }
    renderTransform {angle=0, scale=<18 18>, shear=<0 0>, translation=<0 0>}
   }
   slot {class=SizeBoxSlot}
  }
  SizeBoxTris {
   widgetClassPath=&SizeBox
   parent=OverlayFX
   widget=SizeBoxTris
   properties {
    bOverride_HeightOverride=true
    bOverride_WidthOverride=true
    heightOverride=24
    widthOverride=24
    renderTransform {angle=0, scale=<2 2>, shear=<0 0>, translation=<4 0>}
   }
   slot {class=OverlaySlot, properties {horizontalAlignment=HAlign_Right, verticalAlignment=VAlign_Center, padding=<ltrb 0 0 120 0>}}
  }
  OverlayTris {
   widgetClassPath=&Overlay
   bIsVariable=true
   parent=SizeBoxTris
   widget=OverlayTris
   properties {renderOpacity=0, renderTransform {angle=0, scale=<0.9 0.9>, shear=<0 0>, translation=<0 0>}}
   slot {class=SizeBoxSlot}
  }
  TriFX_1 {
   widgetClassPath=&Image
   bIsVariable=true
   parent=OverlayTris
   widget=TriFX_1
   properties {
    renderOpacity=0.8
    visibility=SelfHitTestInvisible
    brush {
     bIsDynamicallyLoaded=false
     drawAs=Image
     imageType=NoImage
     mirroring=NoMirror
     resourceName=None
     resourceObject=/FortUILibrary/MaterialInstances/Effects/M_UI_Tri_FX_HUDLevelup
     tiling=NoTile
     imageSize=<32 16>
     margin=<ltrb 0 0 0 0>
     outlineSettings {
      bUseBrushTransparency=false
      roundingType=HalfHeightRadius
      width=0
      color {colorUseRule=UseColor_Specified, specifiedColor=<rgba 0 0 0 0>}
      cornerRadii=<0 0 0 0>
     }
     tintColor {colorUseRule=UseColor_Specified, specifiedColor=<rgba 1 1 1 1>}
     uVRegion {bIsValid=false, max=<0 0>, min=<0 0>}
    }
    renderTransform {angle=90, scale=<1.5 3>, shear=<0 0>, translation=<-14 0>}
    renderTransformPivot=<1 0.5>
   }
   slot {class=OverlaySlot, properties {horizontalAlignment=HAlign_Center, verticalAlignment=VAlign_Center}}
  }
  TriFX_2 {
   widgetClassPath=&Image
   bIsVariable=true
   parent=OverlayTris
   widget=TriFX_2
   properties {
    renderOpacity=0.8
    visibility=SelfHitTestInvisible
    brush {
     bIsDynamicallyLoaded=false
     drawAs=Image
     imageType=NoImage
     mirroring=NoMirror
     resourceName=None
     resourceObject=&M_UI_Tri_FX_HUDLevelup
     tiling=NoTile
     imageSize=<32 16>
     margin=<ltrb 0 0 0 0>
     outlineSettings {
      bUseBrushTransparency=false
      roundingType=HalfHeightRadius
      width=0
      color {colorUseRule=UseColor_Specified, specifiedColor=<rgba 0 0 0 0>}
      cornerRadii=<0 0 0 0>
     }
     tintColor {colorUseRule=UseColor_Specified, specifiedColor=<rgba 1 1 1 1>}
     uVRegion {bIsValid=false, max=<0 0>, min=<0 0>}
    }
    renderTransform {angle=210, scale=<1.5 3>, shear=<0 0>, translation=<-14 0>}
    renderTransformPivot=<1 0.5>
   }
   slot {class=OverlaySlot, properties {horizontalAlignment=HAlign_Center, verticalAlignment=VAlign_Center}}
  }
  TriFX_3 {
   widgetClassPath=&Image
   bIsVariable=true
   parent=OverlayTris
   widget=TriFX_3
   properties {
    renderOpacity=0.8
    visibility=SelfHitTestInvisible
    brush {
     bIsDynamicallyLoaded=false
     drawAs=Image
     imageType=NoImage
     mirroring=NoMirror
     resourceName=None
     resourceObject=&M_UI_Tri_FX_HUDLevelup
     tiling=NoTile
     imageSize=<32 16>
     margin=<ltrb 0 0 0 0>
     outlineSettings {
      bUseBrushTransparency=false
      roundingType=HalfHeightRadius
      width=0
      color {colorUseRule=UseColor_Specified, specifiedColor=<rgba 0 0 0 0>}
      cornerRadii=<0 0 0 0>
     }
     tintColor {colorUseRule=UseColor_Specified, specifiedColor=<rgba 1 1 1 1>}
     uVRegion {bIsValid=false, max=<0 0>, min=<0 0>}
    }
    renderTransform {angle=330, scale=<1.5 3>, shear=<0 0>, translation=<-14 0>}
    renderTransformPivot=<1 0.5>
   }
   slot {class=OverlaySlot, properties {horizontalAlignment=HAlign_Center, verticalAlignment=VAlign_Center}}
  }
 ]
}
