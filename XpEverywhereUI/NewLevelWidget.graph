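# Level-up HUD widget built on XpEverywhereLevelUpWidget that hosts the LevelUpVisuals child to present a new level.
# OnShowWidget sets the level number, starts the child's animation and marks the prioritized widget's content active.
# The child's dispatchers trigger reward display, then completion, which clears active content.
# Construct binds to HUD post-game placement changes, and a victory-placement event plays a one-time top-padding animation to shift the visuals down.
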WidgetBlueprint /XpEverywhereUI/NewLevelWidget
animations [placementOffset {
  binding_count=2
  length_seconds=0.082
  object=&self:placementOffset
  sequence {bindings [
    OverlaySlot_0 {
     id=#1
     parent_id=#2
     possessed_class=OverlaySlot
     tracks [{
       type=MovieSceneMarginTrack
       display_name=Padding
       sections [{
         type=MovieSceneMarginSection
         channels [
          Left {type=MovieSceneScriptingFloatChannel}
          Top {type=MovieSceneScriptingFloatChannel, keys [[0,0],[1.6403,94.8122]]}
          Right {type=MovieSceneScriptingFloatChannel}
          Bottom {type=MovieSceneScriptingFloatChannel}
         ]
        }]
      }]
    }
    LevelUpVisuals {id=#2, possessed_class=LevelUpVisuals_C, tracks []}
   ]}
 }]
graphs [EventGraph {nodes [
   n1 = ShowLevelReward() {is=title K2Node_CustomEvent}
   > DisplayLevelUpRewards(self:Xp Everywhere Level Up Widget Object Reference)
   n2 = OutroFinished() {is=title K2Node_CustomEvent}
   > DoneLevelingUp(self:Xp Everywhere Level Up Widget Object Reference)
   > SetHasActiveContent!(execute, self:Fort Prioritized Widget Interface, bNewActive:bool=false) -> (then)
   n3 = EventMoveforEoM!() -> (OutputDelegate, then, Placement:Name) {is=title K2Node_CustomEvent}
   n4 = Get isPlaced() -> (isPlaced:bool)
   > Branch(Condition=AND Boolean(A="Equal (Name)"!(A:Name=@n3.Placement, B:Name=PlacementVictory) -> (ReturnValue:bool) {is=title K2Node_CallFunction}, B=NOT Boolean(A=@n4)))
   > PlayAnimation(InAnimation=Get placementOffset() -> (placementOffset:Widget Animation Object Reference))
   > Set isPlaced(isPlaced:bool=true) -> (Output_Get:bool)
   n5 = Event OnShowWidget!() -> (OutputDelegate, then, NewLevel:int=0) {is=title K2Node_Event}
   n6 = Get LevelUpVisuals() -> (LevelUpVisuals:Level Up Visuals Object Reference)
   > "SetText (Text)"(
    self=Get LevelNumber!(self:Level Up Visuals Object Reference=@n6) -> (LevelNumber:Common Text Object Reference)
    InText="To Text (Integer)"(Value=@n5.NewLevel)
   )
   > Set Level!(execute, Level:int=@n5.NewLevel, self:Level Up Visuals Object Reference=@n6) -> (then, Output_Get:int=0)
   n7 = Get LevelUpVisuals()
   > EventStartAnim(self:Level Up Visuals Object Reference=@n7)
   > Bind Event to Dispatch Show Reward(self:Level Up Visuals Object Reference=@n7, Delegate=@n1) {is=title K2Node_AddDelegate}
   > Bind Event to Dispatch Complete(self:Level Up Visuals Object Reference=@n7, Delegate=@n2) {is=title K2Node_AddDelegate}
   > SetHasActiveContent(bNewActive=true)
   Event Construct()
   > Bind Event to On Post Game Widget Placement Change(
    self:Athena HUDContext Object Reference=Get AthenaHUDContext!() -> (ReturnValue:Athena HUDContext Object Reference) {is=title K2Node_GetBlueprintContextEx}
    Delegate=Create Event()
   ) {is=title K2Node_AddDelegate}
   > SetHasActiveContent(bNewActive=false)
  ]}]
variables [isPlaced {type=boolean}]
widgets {
 info {inheritedWidgetCount=0, parentClass=/Script/XpEverywhereUI.XpEverywhereLevelUpWidget, rootWidgetClass=&Overlay, widgetCount=2}
 widgets [
  Overlay_0 {widgetClassPath=&Overlay, widget=Overlay_0}
  LevelUpVisuals {widgetClassPath=&/LevelUpVisuals, bIsVariable=true, parent=Overlay_0, widget=LevelUpVisuals, slot {class=OverlaySlot}}
 ]
}
