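# Widget Blueprint for the XP Everywhere bar, an animated notification showing XP earned, level, XP to next level and rested-XP state.
# OnXpBarUpdated computes start/stop fractions, capped at full on level-up, and feeds text and dynamic material parameters.
# Timeline animations reveal the fill, delta glow and text; outros are quick when leveling, otherwise slow, and skipped for always-active, manually updated bars.
# SetHasActiveContent marks the prioritized widget active during display.
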WidgetBlueprint /XpEverywhereUI/XP_Everywhere_Bar
animations [
 ProgressSequence {
  binding_count=9
  length_seconds=2.25
  object=&self:ProgressSequence
  sequence {bindings [
    RestedXPTextBox {
     id=#1
     possessed_class=Border
     tracks [
      {
       type=MovieSceneFloatTrack
       display_name=Render Opacity
       sections [{type=MovieSceneFloatSection, channels [None {type=MovieSceneScriptingFloatChannel, keys [[1,0],[8,1]]}]}]
      }
      {
       type=MovieScene2DTransformTrack
       sections [{
         type=MovieScene2DTransformSection
         channels [
          Translation.X {type=MovieSceneScriptingFloatChannel, keys [[0,-128],[8,0]]}
          Translation.Y {type=MovieSceneScriptingFloatChannel, const [-40,1]}
         ]
        }]
      }
     ]
    }
    SocialAvatarIcon {
     id=#2
     possessed_class=Image
     tracks [
      {
       type=MovieSceneFloatTrack
       display_name=Render Opacity
       sections [{type=MovieSceneFloatSection, channels [None {type=MovieSceneScriptingFloatChannel, keys [[2,0],[5,1]]}]}]
      }
      {
       type=MovieScene2DTransformTrack
       sections [{
         type=MovieScene2DTransformSection
         channels [
          Scale.X {type=MovieSceneScriptingFloatChannel, keys [[2,0.8],[7,1]]}
          Scale.Y {type=MovieSceneScriptingFloatChannel, keys [[2,0.8],[7,1]]}
         ]
        }]
      }
      {
       type=MovieSceneWidgetMaterialTrack
       sections [{type=MovieSceneParameterSection, channels [PortraitScale {type=MovieSceneScriptingFloatChannel, keys [[0,1.75],[10,1.35]]}]}]
      }
     ]
    }
    Backing_Color {
     id=#3
     possessed_class=Image
     tracks [{
       type=MovieSceneWidgetMaterialTrack
       sections [{
         type=MovieSceneParameterSection
         channels [
          Hover {type=MovieSceneScriptingFloatChannel, keys [[0,0],[2,1],[20,0]]}
          AnimateAlpha {type=MovieSceneScriptingFloatChannel, keys [[0,0],[5,1],[30,0]]}
         ]
        }]
      }]
    }
    CurrentLevelText_2 {
     id=#4
     possessed_class=CommonRichTextBlock
     tracks [
      {
       type=MovieSceneFloatTrack
       display_name=Render Opacity
       sections [{type=MovieSceneFloatSection, channels [None {type=MovieSceneScriptingFloatChannel, keys [[2,0],[7,1]]}]}]
      }
      {
       type=MovieScene2DTransformTrack
       sections [{type=MovieScene2DTransformSection, channels [Translation.X {type=MovieSceneScriptingFloatChannel, keys [[2,-24],[7,0]]}]}]
      }
     ]
    }
    Backing_Outline {
     id=#5
     possessed_class=Image
     tracks [{
       type=MovieSceneWidgetMaterialTrack
       sections [{
         type=MovieSceneParameterSection
         channels [
          AnimateAlpha {type=MovieSceneScriptingFloatChannel, keys [[4,1],[7,0],[30,1]]}
          Density {type=MovieSceneScriptingFloatChannel, keys [[4,3],[7,4]]}
         ]
        }]
      }]
    }
    ProgressBar {
     id=#6
     possessed_class=Image
     tracks [
      {
       type=MovieSceneWidgetMaterialTrack
       sections [{
         type=MovieSceneParameterSection
         channels [
          Animate_Progress {type=MovieSceneScriptingFloatChannel, keys [[27,0],[45,1]]}
          Animate_Delta {type=MovieSceneScriptingFloatChannel, keys [[17,0],[22,1]]}
          Animate_DeltaColor {type=MovieSceneScriptingFloatChannel, keys [[27,0],[37,1]]}
         ]
        }]
      }
      {
       type=MovieSceneFloatTrack
       display_name=Render Opacity
       sections [{type=MovieSceneFloatSection, channels [None {type=MovieSceneScriptingFloatChannel, keys [[1,0],[10,1]]}]}]
      }
     ]
    }
    ProgressBar_DeltaGlow {
     id=#7
     possessed_class=Image
     tracks [
      {
       type=MovieSceneFloatTrack
       display_name=Render Opacity
       sections [{type=MovieSceneFloatSection, channels [None {type=MovieSceneScriptingFloatChannel, keys [[17,0],[19,1],[33,0]]}]}]
      }
      {
       type=MovieSceneWidgetMaterialTrack
       sections [{
         type=MovieSceneParameterSection
         channels [
          Animate_Delta {type=MovieSceneScriptingFloatChannel, keys [[17,0],[22,1]]}
          Animate_DeltaColor {type=MovieSceneScriptingFloatChannel, keys [[17,1],[21,0.9055],[25,1]]}
         ]
        }]
      }
     ]
    }
    All {
     id=#8
     possessed_class=Border
     tracks [
      {
       type=MovieSceneFloatTrack
       display_name=Render Opacity
       sections [{type=MovieSceneFloatSection, channels [None {type=MovieSceneScriptingFloatChannel, keys [[0,0],[2,1],[45,1]]}]}]
      }
      {
       type=MovieScene2DTransformTrack
       sections [{
         type=MovieScene2DTransformSection
         channels [
          Translation.X {type=MovieSceneScriptingFloatChannel, keys [[0,-512],[6,0]]}
          Scale.X {type=MovieSceneScriptingFloatChannel, keys [[1,0.8],[5,1]]}
          Scale.Y {type=MovieSceneScriptingFloatChannel, keys [[1,0.8],[5,1]]}
         ]
        }]
      }
     ]
    }
    DeltaGlowScaleController {
     id=#9
     possessed_class=Border
     tracks [{
       type=MovieScene2DTransformTrack
       sections [{type=MovieScene2DTransformSection, channels [Scale.Y {type=MovieSceneScriptingFloatChannel, keys [[17,1],[30,2.5]]}]}]
      }]
    }
   ]}
 }
 QuickOutro {
  binding_count=1
  length_seconds=0.2
  object=&self:QuickOutro
  sequence {bindings [All {
     id=#8
     possessed_class=Border
     tracks [
      {
       type=MovieSceneFloatTrack
       display_name=Render Opacity
       sections [{type=MovieSceneFloatSection, channels [None {type=MovieSceneScriptingFloatChannel, keys [[0,1],[4,0]]}]}]
      }
      {
       type=MovieScene2DTransformTrack
       sections [{
         type=MovieScene2DTransformSection
         channels [
          Translation.X {type=MovieSceneScriptingFloatChannel, keys [[0,0],[4,-512]]}
          Scale.X {type=MovieSceneScriptingFloatChannel, keys [[0,1],[4,0.75]]}
          Scale.Y {type=MovieSceneScriptingFloatChannel, keys [[0,1],[4,0.75]]}
         ]
        }]
      }
     ]
    }]}
 }
 ShowXPEarnedAndNextLevel {
  binding_count=2
  length_seconds=2.8
  object=&self:ShowXPEarnedAndNextLevel
  sequence {bindings [
    XpRemainingText_2 {
     id=#10
     possessed_class=CommonRichTextBlock
     tracks [
      {
       type=MovieSceneFloatTrack
       display_name=Render Opacity
       sections [{type=MovieSceneFloatSection, channels [None {type=MovieSceneScriptingFloatChannel, keys [[50,0],[56,1]]}]}]
      }
      {
       type=MovieScene2DTransformTrack
       sections [{type=MovieScene2DTransformSection, channels [Translation.X {type=MovieSceneScriptingFloatChannel, keys [[51,-80],[56,0]]}]}]
      }
     ]
    }
    XPAmountBox {
     id=#11
     possessed_class=HorizontalBox
     tracks [
      {
       type=MovieScene2DTransformTrack
       sections [{type=MovieScene2DTransformSection, channels [Translation.X {type=MovieSceneScriptingFloatChannel, keys [[17,-64],[22,0],[43,0],[48,16]]}]}]
      }
      {
       type=MovieSceneFloatTrack
       display_name=Render Opacity
       sections [{type=MovieSceneFloatSection, channels [None {type=MovieSceneScriptingFloatChannel, keys [[0,0],[17,0],[22,1],[43,1],[48,0]]}]}]
      }
     ]
    }
   ]}
 }
 SlowOutro {
  binding_count=1
  length_seconds=2.5
  object=&self:SlowOutro
  sequence {bindings [All {
     id=#8
     possessed_class=Border
     tracks [
      {
       type=MovieSceneFloatTrack
       display_name=Render Opacity
       sections [{type=MovieSceneFloatSection, channels [None {type=MovieSceneScriptingFloatChannel, keys [[0,1],[40,1],[48,0]]}]}]
      }
      {
       type=MovieScene2DTransformTrack
       sections [{
         type=MovieScene2DTransformSection
         channels [
          Translation.X {type=MovieSceneScriptingFloatChannel, keys [[40,0],[50,-512]]}
          Scale.X {type=MovieSceneScriptingFloatChannel, keys [[40,1],[50,0.75]]}
          Scale.Y {type=MovieSceneScriptingFloatChannel, keys [[40,1],[50,0.75]]}
         ]
        }]
      }
     ]
    }]}
 }
]
graphs [
 EventGraph {nodes [
   n1 = UpdateText!() -> (OutputDelegate, then, DisplayedLevel:int, XpForDisplayedLevel:int) {is=title K2Node_CustomEvent}
   > SetText(
    self=Get CurrentLevelText_2() -> (CurrentLevelText_2:Common Rich Text Block Object Reference)
    InText=Format Text!(Format="LOCTEXT(\"LVL <text color=\\\"FFF\\\" size=\\\"17\\\">{level}</>\")", level:int=@n1.DisplayedLevel) -> (Result)
   )
   n2 = Get TotalXpRequiredForLevel() -> (TotalXpRequiredForLevel:int=0)
   > SetText(
    self=Get XpRemainingText_2() -> (XpRemainingText_2:Common Rich Text Block Object Reference)
    InText=Format Text!(
     Format="LOCTEXT(\"<text color=\\\"33aaff\\\">{RemainingXP} XP</> TO LVL {level}\")"
     RemainingXP:int=int - int(A=@n2, B=@n1.XpForDisplayedLevel)
     level=int + int(A=@n1.DisplayedLevel)
    ) -> (Result)
   )
   n3 = "Equal (Integer)"(A=Get Shown Level() -> (Shown Level:int=0), B=Get New Desired Level() -> (New Desired Level:int=0))
   n4 = int - int(A=Get TotalXpRequiredForLevel() -> (TotalXpRequiredForLevel=0), B=Get Shown XP() -> (Shown XP:int=0))
   n5 = int - int(A=Get NewDesiredXP() -> (NewDesiredXP:int=0), B=Get Shown XP() -> (Shown XP=0))
   > SetText(
    self=Get XpAmount() -> (XpAmount:Common Rich Text Block Object Reference)
    InText=Format Text!(Format="LOCTEXT(\"+{EarnedXP}\")", EarnedXP:int=SelectInt(A=@n5, B=@n4, bPickA=@n3)) -> (Result)
   )
   OutroFinished() {is=title K2Node_CustomEvent}
   > DoneUpdatingXpBar(self:Xp Everywhere Bar Object Reference)
   > SetHasActiveContent!(execute, self:Fort Prioritized Widget Interface, bNewActive:bool=false) -> (then)
   TriggerOutro() {is=title K2Node_CustomEvent}
   n6 = Get bXpBarAlwaysActive() -> (bXpBarAlwaysActive:bool)
   n7 = Get bIsXpBarManuallyUpdated() -> (bIsXpBarManuallyUpdated:bool)
   > n8 = Branch(Condition=OR Boolean(A=NOT Boolean(A=@n6), B=NOT Boolean(A=@n7)))
   > PlayAnimation(InAnimation=Select(
     Option 0:Widget Animation Object Reference=Get SlowOutro() -> (SlowOutro:Widget Animation Object Reference)
     Option 1:Widget Animation Object Reference=Get QuickOutro() -> (QuickOutro:Widget Animation Object Reference)
     Index=Get IsLevelingUp() -> (IsLevelingUp:bool)
    ) -> (ReturnValue:Widget Animation Object Reference))
   OutroFinished(execute=@n8.else) {is=title K2Node_CallFunction}
   n9 = Event OnXpBarUpdated!() -> (
    OutputDelegate
    then
    ShownXP:int=0
    ShownLevel:int=0
    NewDesiredXP:int=0
    NewDesiredLevel:int=0
    NewRemainingRestXP:int=0
    NewRestedXpConsumed:int=0
   ) {is=title K2Node_Event}
   > Set Shown XP(Shown XP:int=@n9.ShownXP) -> (Output_Get:int=0)
   > Set Shown Level(Shown Level:int=@n9.ShownLevel) -> (Output_Get:int=0)
   > Set NewDesiredXP(NewDesiredXP:int=@n9.NewDesiredXP) -> (Output_Get:int=0)
   > n10 = Set New Desired Level(New Desired Level:int=@n9.NewDesiredLevel) -> (Output_Get:int=0)
   > Set IsLevelingUp(IsLevelingUp:bool="integer > integer"(A=@n10, B=Get Shown Level() -> (Shown Level=0))) -> (Output_Get:bool)
   > SetRestedXPVisualState!(execute, self, IsRested:bool="integer > integer"(A=@n9.NewRemainingRestXP)) -> (then)
   > Set TotalXpRequiredForLevel(TotalXpRequiredForLevel:int=GetTotalXpRequiredForLevel!(self:Xp Everywhere Bar Object Reference, InLevel:int=Get Shown Level() -> (Shown Level=0)) -> (ReturnValue:int=0)) -> (Output_Get:int=0)
   n11 = Get TotalXpRequiredForLevel() -> (TotalXpRequiredForLevel=0)
   n12 = Get Shown XP() -> (Shown XP=0)
   > Set ProgressStartPosition(ProgressStartPosition:double=float / float(A="To Float (Integer)"(InInt=@n12), B="To Float (Integer)"(InInt=@n11))) -> (Output_Get:double)
   n13 = "To Float (Integer)"(InInt=Get TotalXpRequiredForLevel() -> (TotalXpRequiredForLevel=0))
   n14 = "To Float (Integer)"(InInt=Get NewDesiredXP() -> (NewDesiredXP=0))
   > Set ProgressStopPosition(ProgressStopPosition:double=SelectFloat(A=1, B=float / float(A=@n14, B=@n13), bPickA=Get IsLevelingUp())) -> (Output_Get:double)
   > UpdateText!(
    execute
    self
    DisplayedLevel:int=Get Shown Level() -> (Shown Level=0)
    XpForDisplayedLevel:int=SelectInt(A=Get TotalXpRequiredForLevel() -> (TotalXpRequiredForLevel=0), B=Get NewDesiredXP() -> (NewDesiredXP=0), bPickA=Get IsLevelingUp())
   ) -> (then) {is=title K2Node_CallFunction}
   > n15 = GetDynamicMaterial(self=Get ProgressBar() -> (ProgressBar:Image Object Reference))
   > SetScalarParameterValue(self=@n15, ParameterName=Progress_Start, Value=Get ProgressStartPosition() -> (ProgressStartPosition:double))
   > SetScalarParameterValue(self=@n15, ParameterName=Progress_Stop, Value=Get ProgressStopPosition() -> (ProgressStopPosition:double))
   > n16 = GetDynamicMaterial(self=Get ProgressBar_DeltaGlow() -> (ProgressBar_DeltaGlow:Image Object Reference))
   > SetScalarParameterValue(self=@n16, ParameterName=Progress_Start, Value=Get ProgressStartPosition())
   > SetScalarParameterValue(self=@n16, ParameterName=Progress_Stop, Value=Get ProgressStopPosition())
   > PlayAnimation(
    InAnimation=Get ProgressSequence() -> (ProgressSequence:Widget Animation Object Reference)
    StartAtTime=SelectFloat(A=Get AlwaysActiveProgressAnimStartTime() -> (AlwaysActiveProgressAnimStartTime:double), B=0, bPickA=Get bXpBarAlwaysActive())
   )
   > PlayAnimation(InAnimation=Get ShowXPEarnedAndNextLevel() -> (ShowXPEarnedAndNextLevel:Widget Animation Object Reference))
   > SetHasActiveContent(bNewActive=true)
   Event OnInitialized()
   > BindToAnimationFinished(Animation=Get ProgressSequence(), Delegate=Create Event())
   n17 = Create Event()
   > BindToAnimationFinished(Animation=Get SlowOutro(), Delegate=@n17)
   > BindToAnimationFinished(Animation=Get QuickOutro(), Delegate=@n17)
   Event Construct()
   > SetRenderOpacity(self=Get All() -> (All:Border Object Reference), InOpacity=0)
   > SetHasActiveContent(bNewActive=false)
  ]}
 SetRestedXPVisualState {nodes [
   n1 = SetRestedXPVisualState!() -> (then, IsRested:bool=~) {is=title K2Node_FunctionEntry}
   > n2 = GetDynamicMaterial(self=Get ProgressBar())
   > SetScalarParameterValue(self=@n2, ParameterName=IsSuperCharged, Value=Select(Option 0:double=0, Option 1:double=1, Index=@n1) -> (ReturnValue:double))
   > SetVisibility(
    self=Get RestedXPTextBox() -> (RestedXPTextBox:Border Object Reference)
    InVisibility=Select(Option 0:ESlateVisibility Enum=Collapsed, Option 1:ESlateVisibility Enum=SelfHitTestInvisible, Index=@n1) -> (ReturnValue:ESlateVisibility Enum)
   )
  ]}
 SetXpToNextLevelTextVisualState {nodes [
   SetXpToNextLevelTextVisualState() {is=title K2Node_FunctionEntry}
   n1 = Get ShowXPEarnedAndNextLevel()
   > PlayAnimation(InAnimation=@n1, StartAtTime=GetEndTime(self=@n1))
  ]}
]
variables [
 AdditionalDelay {type=number, default=1}
 AlwaysActiveProgressAnimStartTime {type=number, default=0.85}
 Anim Speed {type=number, default=1}
 HasLeveledAnimPlayed {type=boolean}
 IsLevelingUp {type=boolean}
 New Desired Level {type=integer}
 NewDesiredXP {type=integer}
 ProgressBarMID {type=MaterialInstanceDynamic}
 ProgressStartPosition {type=number}
 ProgressStopPosition {type=number}
 Shown Level {type=integer}
 Shown XP {type=integer}
 TotalXpRequiredForLevel {type=integer}
 XpEarned {type=number}
 XpFillRate {type=integer, default=15000}
]
widgets {
 info {inheritedWidgetCount=1, parentClass=/Script/XpEverywhereUI.XpEverywhereBar, rootWidgetClass=&Border, widgetCount=31}
 widgets [
  All {
   widgetClassPath=&Border
   bIsVariable=true
   widget=All
   properties {
    visibility=SelfHitTestInvisible
    background {
     bIsDynamicallyLoaded=false
     drawAs=NoDrawType
     imageType=NoImage
     mirroring=NoMirror
     resourceName=None
     resourceObject=/FortUILibrary/Foundation/Art_ToolKit/Tools/M_UI_XP_Everywhere_Backing_DropShadow
     tiling=NoTile
     imageSize=<40 40>
     margin=<ltrb 0.5 0.5 0.5 0.5>
     outlineSettings {
      bUseBrushTransparency=false
      roundingType=HalfHeightRadius
      width=0
      color {colorUseRule=UseColor_Specified, specifiedColor=<rgba 0 0 0 0>}
      cornerRadii=<0 0 0 0>
     }
     tintColor {colorUseRule=UseColor_Specified, specifiedColor=<rgba 1 1 1 1>}
     uVRegion {bIsValid=false, max=<0 0>, min=<0 0>}
    }
    padding=<ltrb 16 16 16 16>
    renderTransformPivot=<0 0.5>
   }
  }
  Size {
   widgetClassPath=&SizeBox
   parent=All
   widget=Size
   properties {bOverride_HeightOverride=true, bOverride_WidthOverride=true, heightOverride=64, widthOverride=430}
   slot {class=BorderSlot, properties {padding=<ltrb 16 16 16 16>}}
  }
  Overlay_2 {widgetClassPath=&Overlay, parent=Size, widget=Overlay_2, slot {class=SizeBoxSlot}}
  Backing_Darken {
   widgetClassPath=&Image
   bIsVariable=true
   parent=Overlay_2
   widget=Backing_Darken
   properties {
    renderOpacity=0
    visibility=SelfHitTestInvisible
    brush {
     bIsDynamicallyLoaded=false
     drawAs=Box
     imageType=NoImage
     mirroring=NoMirror
     resourceName=None
     resourceObject=/FortUILibrary/Foundation/Art_ToolKit/Tools/M_UI_XP_Everywhere_Backing_Darken
     tiling=NoTile
     imageSize=<16 16>
     margin=<ltrb 0.5 0.5 0.5 0.5>
     outlineSettings {
      bUseBrushTransparency=false
      roundingType=HalfHeightRadius
      width=0
      color {colorUseRule=UseColor_Specified, specifiedColor=<rgba 0 0 0 0>}
      cornerRadii=<0 0 0 0>
     }
     tintColor {colorUseRule=UseColor_Specified, specifiedColor=<rgba 1 1 1 1>}
     uVRegion {bIsValid=false, max=<0 0>, min=<0 0>}
    }
    colorAndOpacity=<rgba 1 1 1 0.96>
   }
   slot {class=OverlaySlot, properties {horizontalAlignment=HAlign_Fill, verticalAlignment=VAlign_Fill}}
  }
  Backing_Color {
   widgetClassPath=&Image
   bIsVariable=true
   parent=Overlay_2
   widget=Backing_Color
   properties {
    visibility=SelfHitTestInvisible
    brush {
     bIsDynamicallyLoaded=false
     drawAs=Box
     imageType=FullColor
     mirroring=NoMirror
     resourceName=None
     resourceObject=/FortniteCoreUI/Athena/HUD/Quests/Art/M_Ui_Quest_XPNotification_DarkPurple
     tiling=NoTile
     imageSize=<64 64>
     margin=<ltrb 0.3 0.3 0.3 0.3>
     outlineSettings {
      bUseBrushTransparency=false
      roundingType=HalfHeightRadius
      width=0
      color {colorUseRule=UseColor_Specified, specifiedColor=<rgba 0 0 0 0>}
      cornerRadii=<0 0 0 0>
     }
     tintColor {colorUseRule=UseColor_Specified, specifiedColor=<rgba 1 1 1 1>}
     uVRegion {bIsValid=false, max=<0 0>, min=<0 0>}
    }
   }
   slot {class=OverlaySlot, properties {horizontalAlignment=HAlign_Fill, verticalAlignment=VAlign_Fill}}
  }
  Backing_Outline {
   widgetClassPath=&Image
   bIsVariable=true
   parent=Overlay_2
   widget=Backing_Outline
   properties {
    renderOpacity=0.1
    visibility=SelfHitTestInvisible
    brush {
     bIsDynamicallyLoaded=false
     drawAs=Box
     imageType=NoImage
     mirroring=NoMirror
     resourceName=None
     resourceObject=/FortUILibrary/Foundation/Art_ToolKit/Tools/M_UI_XP_Everywhere_BackingGlow
     tiling=NoTile
     imageSize=<16 16>
     margin=<ltrb 0.5 0.5 0.5 0.5>
     outlineSettings {
      bUseBrushTransparency=false
      roundingType=HalfHeightRadius
      width=0
      color {colorUseRule=UseColor_Specified, specifiedColor=<rgba 0 0 0 0>}
      cornerRadii=<0 0 0 0>
     }
     tintColor {colorUseRule=UseColor_Specified, specifiedColor=<rgba 1 1 1 1>}
     uVRegion {bIsValid=false, max=<0 0>, min=<0 0>}
    }
    renderTransform {angle=0, scale=<1.005 1.025>, shear=<0 0>, translation=<0 0>}
   }
   slot {class=OverlaySlot, properties {horizontalAlignment=HAlign_Fill, verticalAlignment=VAlign_Fill}}
  }
  Content {
   widgetClassPath=&HorizontalBox
   parent=Overlay_2
   widget=Content
   slot {class=OverlaySlot, properties {horizontalAlignment=HAlign_Fill, verticalAlignment=VAlign_Center, padding=<ltrb 8 0 12 0>}}
  }
  SocialAvatarIcon {
   widgetClassPath=/Script/FortniteUI.FortSocialAvatarIcon
   bInherited=true
   bIsVariable=true
   parent=Content
   widget=SocialAvatarIcon
   properties {
    avatarMaterial=/Game/UI/SocialPanel/Art/Materials/UI_UserPortrait_Buffet_XP
    brush {
     bIsDynamicallyLoaded=false
     drawAs=Image
     imageType=NoImage
     mirroring=NoMirror
     resourceName=None
     resourceObject=&UI_UserPortrait_Buffet_XP
     tiling=NoTile
     imageSize=<48 48>
     margin=<ltrb 0 0 0 0>
     outlineSettings {
      bUseBrushTransparency=false
      roundingType=HalfHeightRadius
      width=0
      color {colorUseRule=UseColor_Specified, specifiedColor=<rgba 0 0 0 0>}
      cornerRadii=<0 0 0 0>
     }
     tintColor {colorUseRule=UseColor_Specified, specifiedColor=<rgba 1 1 1 1>}
     uVRegion {bIsValid=false, max=<0 0>, min=<0 0>}
    }
   }
   slot {class=HorizontalBoxSlot, properties {horizontalAlignment=HAlign_Center, verticalAlignment=VAlign_Center}}
  }
  ProgressBarAndText {
   widgetClassPath=&VerticalBox
   parent=Content
   widget=ProgressBarAndText
   slot {class=HorizontalBoxSlot, properties {verticalAlignment=VAlign_Bottom, padding=<ltrb 2 0 0 16>, size {sizeRule=Fill, value=1}}}
  }
  Text {
   widgetClassPath=&HorizontalBox
   parent=ProgressBarAndText
   widget=Text
   properties {renderTransform {angle=0, scale=<1 1>, shear=<-12 0>, translation=<0 0>}}
   slot {class=VerticalBoxSlot, properties {padding=<ltrb 0 0 2 4>, size {sizeRule=Fill, value=1}}}
  }
  CommonRichTextBlock_164 {
   widgetClassPath=&CommonRichTextBlock
   bIsVariable=true
   parent=Text
   widget=CommonRichTextBlock_164
   properties {
    bOverrideDefaultStyle=true
    defaultTextStyleOverride {
     overflowPolicy=Clip
     transformPolicy=None
     colorAndOpacity {colorUseRule=UseColor_Specified, specifiedColor=<rgba 1 1 1 1>}
     font {
      bForceMonospaced=false
      bMaterialIsStencil=false
      fontMaterial=null
      fontObject=&NotoSans
      letterSpacing=0
      monospacedWidth=1
      size=24
      skewAmount=0
      typefaceFontName=NotoSans-Regular
      outlineSettings {
       bApplyOutlineToDropShadows=false
       bMiteredCorners=false
       bSeparateFillAlpha=false
       outlineMaterial=null
       outlineSize=0
       outlineColor=<rgba 0 0 0 1>
      }
     }
     highlightColor {colorUseRule=UseColor_Specified, specifiedColor=<rgba 0 0 0 1>}
     highlightShape {
      bIsDynamicallyLoaded=false
      drawAs=Image
      imageType=NoImage
      mirroring=NoMirror
      resourceName=None
      resourceObject=null
      tiling=NoTile
      imageSize=<32 32>
      margin=<ltrb 0 0 0 0>
      outlineSettings {
       bUseBrushTransparency=false
       roundingType=HalfHeightRadius
       width=0
       color {colorUseRule=UseColor_Specified, specifiedColor=<rgba 0 0 0 0>}
       cornerRadii=<0 0 0 0>
      }
      tintColor {colorUseRule=UseColor_Specified, specifiedColor=<rgba 1 1 1 1>}
      uVRegion {bIsValid=false, max=<0 0>, min=<0 0>}
     }
     selectedBackgroundColor {colorUseRule=UseColor_Foreground, specifiedColor=<rgba 1 0 1 1>}
     shadowColorAndOpacity=<rgba 0 0 0 1>
     shadowOffset=<0 0>
     strikeBrush {
      bIsDynamicallyLoaded=false
      drawAs=Image
      imageType=NoImage
      mirroring=NoMirror
      resourceName=None
      resourceObject=null
      tiling=NoTile
      imageSize=<32 32>
      margin=<ltrb 0 0 0 0>
      outlineSettings {
       bUseBrushTransparency=false
       roundingType=HalfHeightRadius
       width=0
       color {colorUseRule=UseColor_Specified, specifiedColor=<rgba 0 0 0 0>}
       cornerRadii=<0 0 0 0>
      }
      tintColor {colorUseRule=UseColor_Specified, specifiedColor=<rgba 1 1 1 1>}
      uVRegion {bIsValid=false, max=<0 0>, min=<0 0>}
     }
     underlineBrush {
      bIsDynamicallyLoaded=false
      drawAs=Image
      imageType=NoImage
      mirroring=NoMirror
      resourceName=None
      resourceObject=null
      tiling=NoTile
      imageSize=<32 32>
      margin=<ltrb 0 0 0 0>
      outlineSettings {
       bUseBrushTransparency=false
       roundingType=HalfHeightRadius
       width=0
       color {colorUseRule=UseColor_Specified, specifiedColor=<rgba 0 0 0 0>}
       cornerRadii=<0 0 0 0>
      }
      tintColor {colorUseRule=UseColor_Specified, specifiedColor=<rgba 1 1 1 1>}
      uVRegion {bIsValid=false, max=<0 0>, min=<0 0>}
     }
    }
   }
   slot {class=HorizontalBoxSlot}
  }
  CurrentLevelText_2 {
   widgetClassPath=&CommonRichTextBlock
   bIsVariable=true
   parent=Text
   widget=CurrentLevelText_2
   properties {
    bOverrideDefaultStyle=true
    mobileTextBlockScale=1.1
    text="LVL <text color=\"FFF\" size=\"16\">X</>"
    textTransformPolicy=ToUpper
    defaultTextStyleOverride {
     overflowPolicy=Clip
     transformPolicy=None
     colorAndOpacity {colorUseRule=UseColor_Specified, specifiedColor=<rgba 0.0875 0.1006 1 1>}
     font {
      bForceMonospaced=false
      bMaterialIsStencil=false
      fontMaterial=null
      fontObject=&HeadingNow
      letterSpacing=0
      monospacedWidth=1
      size=12
      skewAmount=0
      typefaceFontName=86 Bold
      outlineSettings {
       bApplyOutlineToDropShadows=false
       bMiteredCorners=false
       bSeparateFillAlpha=false
       outlineMaterial=null
       outlineSize=0
       outlineColor=<rgba 0 0 0 1>
      }
     }
     highlightColor {colorUseRule=UseColor_Specified, specifiedColor=<rgba 0 0 0 1>}
     highlightShape {
      bIsDynamicallyLoaded=false
      drawAs=Image
      imageType=NoImage
      mirroring=NoMirror
      resourceName=None
      resourceObject=null
      tiling=NoTile
      imageSize=<32 32>
      margin=<ltrb 0 0 0 0>
      outlineSettings {
       bUseBrushTransparency=false
       roundingType=HalfHeightRadius
       width=0
       color {colorUseRule=UseColor_Specified, specifiedColor=<rgba 0 0 0 0>}
       cornerRadii=<0 0 0 0>
      }
      tintColor {colorUseRule=UseColor_Specified, specifiedColor=<rgba 1 1 1 1>}
      uVRegion {bIsValid=false, max=<0 0>, min=<0 0>}
     }
     selectedBackgroundColor {colorUseRule=UseColor_Foreground, specifiedColor=<rgba 1 0 1 1>}
     shadowColorAndOpacity=<rgba 0 0 0 1>
     shadowOffset=<0 0>
     strikeBrush {
      bIsDynamicallyLoaded=false
      drawAs=Image
      imageType=NoImage
      mirroring=NoMirror
      resourceName=None
      resourceObject=null
      tiling=NoTile
      imageSize=<32 32>
      margin=<ltrb 0 0 0 0>
      outlineSettings {
       bUseBrushTransparency=false
       roundingType=HalfHeightRadius
       width=0
       color {colorUseRule=UseColor_Specified, specifiedColor=<rgba 0 0 0 0>}
       cornerRadii=<0 0 0 0>
      }
      tintColor {colorUseRule=UseColor_Specified, specifiedColor=<rgba 1 1 1 1>}
      uVRegion {bIsValid=false, max=<0 0>, min=<0 0>}
     }
     underlineBrush {
      bIsDynamicallyLoaded=false
      drawAs=Image
      imageType=NoImage
      mirroring=NoMirror
      resourceName=None
      resourceObject=null
      tiling=NoTile
      imageSize=<32 32>
      margin=<ltrb 0 0 0 0>
      outlineSettings {
       bUseBrushTransparency=false
       roundingType=HalfHeightRadius
       width=0
       color {colorUseRule=UseColor_Specified, specifiedColor=<rgba 0 0 0 0>}
       cornerRadii=<0 0 0 0>
      }
      tintColor {colorUseRule=UseColor_Specified, specifiedColor=<rgba 1 1 1 1>}
      uVRegion {bIsValid=false, max=<0 0>, min=<0 0>}
     }
    }
   }
   slot {class=HorizontalBoxSlot, properties {verticalAlignment=VAlign_Center, padding=<ltrb 6 2 0 0>}}
  }
  Overlay_1 {
   widgetClassPath=&Overlay
   parent=Text
   widget=Overlay_1
   slot {class=HorizontalBoxSlot, properties {horizontalAlignment=HAlign_Right, verticalAlignment=VAlign_Bottom, size {sizeRule=Fill, value=1}}}
  }
  XpRemainingText_2 {
   widgetClassPath=&CommonRichTextBlock
   bIsVariable=true
   parent=Overlay_1
   widget=XpRemainingText_2
   properties {
    bOverrideDefaultStyle=true
    renderOpacity=0
    text="<text color=\"40BFFF\">0 XP</> TO LVL 0"
    defaultTextStyleOverride {
     overflowPolicy=Clip
     transformPolicy=None
     colorAndOpacity {colorUseRule=UseColor_Specified, specifiedColor=<rgba 0.0875 0.1006 1 1>}
     font {
      bForceMonospaced=false
      bMaterialIsStencil=false
      fontMaterial=null
      fontObject=&HeadingNow
      letterSpacing=0
      monospacedWidth=1
      size=12
      skewAmount=0
      typefaceFontName=86 Bold
      outlineSettings {
       bApplyOutlineToDropShadows=false
       bMiteredCorners=false
       bSeparateFillAlpha=false
       outlineMaterial=null
       outlineSize=0
       outlineColor=<rgba 0 0 0 1>
      }
     }
     highlightColor {colorUseRule=UseColor_Specified, specifiedColor=<rgba 0 0 0 1>}
     highlightShape {
      bIsDynamicallyLoaded=false
      drawAs=Image
      imageType=NoImage
      mirroring=NoMirror
      resourceName=None
      resourceObject=null
      tiling=NoTile
      imageSize=<32 32>
      margin=<ltrb 0 0 0 0>
      outlineSettings {
       bUseBrushTransparency=false
       roundingType=HalfHeightRadius
       width=0
       color {colorUseRule=UseColor_Specified, specifiedColor=<rgba 0 0 0 0>}
       cornerRadii=<0 0 0 0>
      }
      tintColor {colorUseRule=UseColor_Specified, specifiedColor=<rgba 1 1 1 1>}
      uVRegion {bIsValid=false, max=<0 0>, min=<0 0>}
     }
     selectedBackgroundColor {colorUseRule=UseColor_Foreground, specifiedColor=<rgba 1 0 1 1>}
     shadowColorAndOpacity=<rgba 0 0 0 1>
     shadowOffset=<0 0>
     strikeBrush {
      bIsDynamicallyLoaded=false
      drawAs=Image
      imageType=NoImage
      mirroring=NoMirror
      resourceName=None
      resourceObject=null
      tiling=NoTile
      imageSize=<32 32>
      margin=<ltrb 0 0 0 0>
      outlineSettings {
       bUseBrushTransparency=false
       roundingType=HalfHeightRadius
       width=0
       color {colorUseRule=UseColor_Specified, specifiedColor=<rgba 0 0 0 0>}
       cornerRadii=<0 0 0 0>
      }
      tintColor {colorUseRule=UseColor_Specified, specifiedColor=<rgba 1 1 1 1>}
      uVRegion {bIsValid=false, max=<0 0>, min=<0 0>}
     }
     underlineBrush {
      bIsDynamicallyLoaded=false
      drawAs=Image
      imageType=NoImage
      mirroring=NoMirror
      resourceName=None
      resourceObject=null
      tiling=NoTile
      imageSize=<32 32>
      margin=<ltrb 0 0 0 0>
      outlineSettings {
       bUseBrushTransparency=false
       roundingType=HalfHeightRadius
       width=0
       color {colorUseRule=UseColor_Specified, specifiedColor=<rgba 0 0 0 0>}
       cornerRadii=<0 0 0 0>
      }
      tintColor {colorUseRule=UseColor_Specified, specifiedColor=<rgba 1 1 1 1>}
      uVRegion {bIsValid=false, max=<0 0>, min=<0 0>}
     }
    }
   }
   slot {class=OverlaySlot}
  }
  XPAmountBox {
   widgetClassPath=&HorizontalBox
   bIsVariable=true
   parent=Overlay_1
   widget=XPAmountBox
   properties {renderOpacity=0}
   slot {class=OverlaySlot, properties {horizontalAlignment=HAlign_Right, verticalAlignment=VAlign_Fill}}
  }
  XpAmount {
   widgetClassPath=&CommonRichTextBlock
   bIsVariable=true
   parent=XPAmountBox
   widget=XpAmount
   properties {
    bOverrideDefaultStyle=true
    justification=Right
    text=+999
    defaultTextStyleOverride {
     overflowPolicy=Clip
     transformPolicy=None
     colorAndOpacity {colorUseRule=UseColor_Specified, specifiedColor=<rgba 0.0513 0.521 1 1>}
     font {
      bForceMonospaced=false
      bMaterialIsStencil=false
      fontMaterial=null
      fontObject=&HeadingNow
      letterSpacing=0
      monospacedWidth=1
      size=12
      skewAmount=0
      typefaceFontName=86 Bold
      outlineSettings {
       bApplyOutlineToDropShadows=false
       bMiteredCorners=false
       bSeparateFillAlpha=false
       outlineMaterial=null
       outlineSize=0
       outlineColor=<rgba 0 0 0 1>
      }
     }
     highlightColor {colorUseRule=UseColor_Specified, specifiedColor=<rgba 0 0 0 1>}
     highlightShape {
      bIsDynamicallyLoaded=false
      drawAs=Image
      imageType=NoImage
      mirroring=NoMirror
      resourceName=None
      resourceObject=null
      tiling=NoTile
      imageSize=<32 32>
      margin=<ltrb 0 0 0 0>
      outlineSettings {
       bUseBrushTransparency=false
       roundingType=HalfHeightRadius
       width=0
       color {colorUseRule=UseColor_Specified, specifiedColor=<rgba 0 0 0 0>}
       cornerRadii=<0 0 0 0>
      }
      tintColor {colorUseRule=UseColor_Specified, specifiedColor=<rgba 1 1 1 1>}
      uVRegion {bIsValid=false, max=<0 0>, min=<0 0>}
     }
     selectedBackgroundColor {colorUseRule=UseColor_Foreground, specifiedColor=<rgba 1 0 1 1>}
     shadowColorAndOpacity=<rgba 0 0 0 1>
     shadowOffset=<0 0>
     strikeBrush {
      bIsDynamicallyLoaded=false
      drawAs=Image
      imageType=NoImage
      mirroring=NoMirror
      resourceName=None
      resourceObject=null
      tiling=NoTile
      imageSize=<32 32>
      margin=<ltrb 0 0 0 0>
      outlineSettings {
       bUseBrushTransparency=false
       roundingType=HalfHeightRadius
       width=0
       color {colorUseRule=UseColor_Specified, specifiedColor=<rgba 0 0 0 0>}
       cornerRadii=<0 0 0 0>
      }
      tintColor {colorUseRule=UseColor_Specified, specifiedColor=<rgba 1 1 1 1>}
      uVRegion {bIsValid=false, max=<0 0>, min=<0 0>}
     }
     underlineBrush {
      bIsDynamicallyLoaded=false
      drawAs=Image
      imageType=NoImage
      mirroring=NoMirror
      resourceName=None
      resourceObject=null
      tiling=NoTile
      imageSize=<32 32>
      margin=<ltrb 0 0 0 0>
      outlineSettings {
       bUseBrushTransparency=false
       roundingType=HalfHeightRadius
       width=0
       color {colorUseRule=UseColor_Specified, specifiedColor=<rgba 0 0 0 0>}
       cornerRadii=<0 0 0 0>
      }
      tintColor {colorUseRule=UseColor_Specified, specifiedColor=<rgba 1 1 1 1>}
      uVRegion {bIsValid=false, max=<0 0>, min=<0 0>}
     }
    }
   }
   slot {class=HorizontalBoxSlot, properties {horizontalAlignment=HAlign_Left, verticalAlignment=VAlign_Top}}
  }
  XPIcon {
   widgetClassPath=&Image
   bIsVariable=true
   parent=XPAmountBox
   widget=XPIcon
   properties {
    visibility=SelfHitTestInvisible
    brush {
     bIsDynamicallyLoaded=false
     drawAs=Image
     imageType=FullColor
     mirroring=NoMirror
     resourceName=None
     resourceObject=/Game/UI/Foundation/Textures/Icons/Items/UniversalXP/T_UI_UniversalXP_Icon_Standard_64
     tiling=NoTile
     imageSize=<24 24>
     margin=<ltrb 0 0 0 0>
     outlineSettings {
      bUseBrushTransparency=false
      roundingType=HalfHeightRadius
      width=0
      color {colorUseRule=UseColor_Specified, specifiedColor=<rgba 0 0 0 0>}
      cornerRadii=<0 0 0 0>
     }
     tintColor {colorUseRule=UseColor_Specified, specifiedColor=<rgba 1 1 1 1>}
     uVRegion {bIsValid=false, max=<0 0>, min=<0 0>}
    }
    renderTransform {angle=0, scale=<1.333 1.333>, shear=<12 0>, translation=<0 -1>}
   }
   slot {class=HorizontalBoxSlot, properties {verticalAlignment=VAlign_Center, padding=<ltrb 6 0 0 0>}}
  }
  ProgressBarSize {
   widgetClassPath=&SizeBox
   parent=ProgressBarAndText
   widget=ProgressBarSize
   properties {
    bOverride_HeightOverride=true
    heightOverride=10
    widthOverride=500
    renderTransform {angle=0, scale=<1 1>, shear=<-12 0>, translation=<0 0>}
   }
   slot {class=VerticalBoxSlot}
  }
  Overlay_0 {widgetClassPath=&Overlay, parent=ProgressBarSize, widget=Overlay_0, slot {class=SizeBoxSlot}}
  ProgressBar {
   widgetClassPath=&Image
   bIsVariable=true
   parent=Overlay_0
   widget=ProgressBar
   properties {
    visibility=SelfHitTestInvisible
    brush {
     bIsDynamicallyLoaded=false
     drawAs=Image
     imageType=NoImage
     mirroring=NoMirror
     resourceName=None
     resourceObject=/FortUILibrary/Foundation/Art_ToolKit/ProgressBars/M_UI_XP_Everywhere_ProgressBar
     tiling=NoTile
     imageSize=<32 32>
     margin=<ltrb 0 0 0 0>
     outlineSettings {
      bUseBrushTransparency=false
      roundingType=HalfHeightRadius
      width=0
      color {colorUseRule=UseColor_Specified, specifiedColor=<rgba 0 0 0 0>}
      cornerRadii=<0 0 0 0>
     }
     tintColor {colorUseRule=UseColor_Specified, specifiedColor=<rgba 1 1 1 1>}
     uVRegion {bIsValid=false, max=<0 0>, min=<0 0>}
    }
   }
   slot {class=OverlaySlot, properties {horizontalAlignment=HAlign_Fill, verticalAlignment=VAlign_Fill}}
  }
  DeltaGlowScaleController {
   widgetClassPath=&Border
   parent=Overlay_0
   widget=DeltaGlowScaleController
   properties {
    visibility=SelfHitTestInvisible
    background {
     bIsDynamicallyLoaded=false
     drawAs=NoDrawType
     imageType=NoImage
     mirroring=NoMirror
     resourceName=None
     resourceObject=null
     tiling=NoTile
     imageSize=<32 32>
     margin=<ltrb 0 0 0 0>
     outlineSettings {
      bUseBrushTransparency=false
      roundingType=HalfHeightRadius
      width=0
      color {colorUseRule=UseColor_Specified, specifiedColor=<rgba 0 0 0 0>}
      cornerRadii=<0 0 0 0>
     }
     tintColor {colorUseRule=UseColor_Specified, specifiedColor=<rgba 1 1 1 1>}
     uVRegion {bIsValid=false, max=<0 0>, min=<0 0>}
    }
    padding=<ltrb 0 0 0 0>
   }
   slot {class=OverlaySlot, properties {horizontalAlignment=HAlign_Fill, verticalAlignment=VAlign_Fill}}
  }
  ProgressBar_DeltaGlow {
   widgetClassPath=&Image
   bIsVariable=true
   parent=DeltaGlowScaleController
   widget=ProgressBar_DeltaGlow
   properties {
    renderOpacity=0
    visibility=SelfHitTestInvisible
    brush {
     bIsDynamicallyLoaded=false
     drawAs=Image
     imageType=NoImage
     mirroring=NoMirror
     resourceName=None
     resourceObject=/FortUILibrary/Foundation/Art_ToolKit/ProgressBars/M_UI_XP_Everywhere_ProgressBar_Glow
     tiling=NoTile
     imageSize=<32 32>
     margin=<ltrb 0 0 0 0>
     outlineSettings {
      bUseBrushTransparency=false
      roundingType=HalfHeightRadius
      width=0
      color {colorUseRule=UseColor_Specified, specifiedColor=<rgba 0 0 0 0>}
      cornerRadii=<0 0 0 0>
     }
     tintColor {colorUseRule=UseColor_Specified, specifiedColor=<rgba 1 1 1 1>}
     uVRegion {bIsValid=false, max=<0 0>, min=<0 0>}
    }
    renderTransform {angle=0, scale=<1 20>, shear=<0 0>, translation=<0 0>}
   }
   slot {class=BorderSlot, properties {padding=<ltrb 0 0 0 0>}}
  }
  RestedXPTextBox {
   widgetClassPath=&Border
   bIsVariable=true
   parent=Overlay_2
   widget=RestedXPTextBox
   properties {
    visibility=Collapsed
    background {
     bIsDynamicallyLoaded=false
     drawAs=NoDrawType
     imageType=NoImage
     mirroring=NoMirror
     resourceName=None
     resourceObject=null
     tiling=NoTile
     imageSize=<32 32>
     margin=<ltrb 0 0 0 0>
     outlineSettings {
      bUseBrushTransparency=false
      roundingType=HalfHeightRadius
      width=0
      color {colorUseRule=UseColor_Specified, specifiedColor=<rgba 0 0 0 0>}
      cornerRadii=<0 0 0 0>
     }
     tintColor {colorUseRule=UseColor_Specified, specifiedColor=<rgba 1 1 1 1>}
     uVRegion {bIsValid=false, max=<0 0>, min=<0 0>}
    }
    padding=<ltrb 0 0 0 0>
    renderTransform {angle=0, scale=<1 1>, shear=<0 0>, translation=<0 -36>}
   }
   slot {class=OverlaySlot, properties {padding=<ltrb 18 0 0 0>}}
  }
  Overlay_4 {widgetClassPath=&Overlay, parent=RestedXPTextBox, widget=Overlay_4, slot {class=BorderSlot, properties {padding=<ltrb 0 0 0 0>}}}
  Image_SuperchargedGlow {
   widgetClassPath=&Image
   parent=Overlay_4
   widget=Image_SuperchargedGlow
   properties {
    visibility=SelfHitTestInvisible
    brush {
     bIsDynamicallyLoaded=false
     drawAs=Box
     imageType=NoImage
     mirroring=NoMirror
     resourceName=None
     resourceObject=/Game/Athena/HUD/LTM/Wax/Materials/M_UI_TextGlow_Add_Gradient_Purple_XP
     tiling=NoTile
     imageSize=<48 48>
     margin=<ltrb 0.5 0.5 0.5 0.5>
     outlineSettings {
      bUseBrushTransparency=false
      roundingType=HalfHeightRadius
      width=0
      color {colorUseRule=UseColor_Specified, specifiedColor=<rgba 0 0 0 0>}
      cornerRadii=<0 0 0 0>
     }
     tintColor {colorUseRule=UseColor_Specified, specifiedColor=<rgba 1 1 1 1>}
     uVRegion {bIsValid=false, max=<0 0>, min=<0 0>}
    }
    colorAndOpacity=<rgba 1 1 1 0.0625>
    renderTransform {angle=0, scale=<1.25 1.333>, shear=<0 0>, translation=<0 0>}
   }
   slot {class=OverlaySlot, properties {horizontalAlignment=HAlign_Fill, verticalAlignment=VAlign_Fill}}
  }
  HorizontalBox_1 {
   widgetClassPath=&HorizontalBox
   parent=Overlay_4
   widget=HorizontalBox_1
   slot {class=OverlaySlot, properties {verticalAlignment=VAlign_Center}}
  }
  Overlay_3 {
   widgetClassPath=&Overlay
   parent=HorizontalBox_1
   widget=Overlay_3
   properties {renderTransform {angle=0, scale=<1 1>, shear=<0 0>, translation=<0 1>}}
   slot {class=HorizontalBoxSlot}
  }
  Image_XPLightning {
   widgetClassPath=&Image
   parent=Overlay_3
   widget=Image_XPLightning
   properties {
    renderOpacity=0
    visibility=Collapsed
    brush {
     bIsDynamicallyLoaded=false
     drawAs=Image
     imageType=NoImage
     mirroring=NoMirror
     resourceName=None
     resourceObject=/Game/Athena/HUD/XP/Materials/M_UI_XPLightning
     tiling=NoTile
     imageSize=<32 32>
     margin=<ltrb 0 0 0 0>
     outlineSettings {
      bUseBrushTransparency=false
      roundingType=HalfHeightRadius
      width=0
      color {colorUseRule=UseColor_Specified, specifiedColor=<rgba 0 0 0 0>}
      cornerRadii=<0 0 0 0>
     }
     tintColor {colorUseRule=UseColor_Specified, specifiedColor=<rgba 1 1 1 1>}
     uVRegion {bIsValid=false, max=<0 0>, min=<0 0>}
    }
    renderTransform {angle=0, scale=<2 2>, shear=<0 0>, translation=<0 0>}
   }
   slot {class=OverlaySlot}
  }
  Image_XPGlow {
   widgetClassPath=&Image
   parent=Overlay_3
   widget=Image_XPGlow
   properties {
    visibility=SelfHitTestInvisible
    brush {
     bIsDynamicallyLoaded=false
     drawAs=Image
     imageType=NoImage
     mirroring=NoMirror
     resourceName=None
     resourceObject=&M_UI_TextGlow_Add_Gradient_Purple_XP
     tiling=NoTile
     imageSize=<32 32>
     margin=<ltrb 0 0 0 0>
     outlineSettings {
      bUseBrushTransparency=false
      roundingType=HalfHeightRadius
      width=0
      color {colorUseRule=UseColor_Specified, specifiedColor=<rgba 0 0 0 0>}
      cornerRadii=<0 0 0 0>
     }
     tintColor {colorUseRule=UseColor_Specified, specifiedColor=<rgba 1 1 1 1>}
     uVRegion {bIsValid=false, max=<0 0>, min=<0 0>}
    }
    colorAndOpacity=<rgba 1 1 1 0.125>
    renderTransform {angle=0, scale=<2 2>, shear=<0 0>, translation=<0 0>}
   }
   slot {class=OverlaySlot, properties {horizontalAlignment=HAlign_Center, verticalAlignment=VAlign_Center}}
  }
  Image_XPIcon {
   widgetClassPath=&Image
   parent=Overlay_3
   widget=Image_XPIcon
   properties {
    brush {
     bIsDynamicallyLoaded=false
     drawAs=Image
     imageType=FullColor
     mirroring=NoMirror
     resourceName=None
     resourceObject=&T_UI_UniversalXP_Icon_Standard_64
     tiling=NoTile
     imageSize=<32 32>
     margin=<ltrb 0 0 0 0>
     outlineSettings {
      bUseBrushTransparency=false
      roundingType=HalfHeightRadius
      width=0
      color {colorUseRule=UseColor_Specified, specifiedColor=<rgba 0 0 0 0>}
      cornerRadii=<0 0 0 0>
     }
     tintColor {colorUseRule=UseColor_Specified, specifiedColor=<rgba 1 1 1 1>}
     uVRegion {bIsValid=false, max=<0 0>, min=<0 0>}
    }
    renderTransform {angle=0, scale=<1 1>, shear=<0 0>, translation=<0 1>}
   }
   slot {class=OverlaySlot}
  }
  CommonTextBlock_71 {
   widgetClassPath=&CommonTextBlock
   parent=HorizontalBox_1
   widget=CommonTextBlock_71
   properties {
    mobileFontSizeMultiplier=1.05
    text="Supercharged!"
    textTransformPolicy=ToUpper
    font {
     bForceMonospaced=false
     bMaterialIsStencil=false
     fontMaterial=null
     fontObject=&HeadingNow
     letterSpacing=0
     monospacedWidth=1
     size=8.25
     skewAmount=0
     typefaceFontName=86 Bold
     outlineSettings {
      bApplyOutlineToDropShadows=false
      bMiteredCorners=false
      bSeparateFillAlpha=false
      outlineMaterial=/Game/UI/Foundation/Materials/M-FontShadow
      outlineSize=2
      outlineColor=<rgba 0.0083 0 0.025 0.25>
     }
    }
    renderTransform {angle=0, scale=<1 1>, shear=<-12 0>, translation=<-5 1>}
    shadowOffset=<0 0>
   }
   slot {class=HorizontalBoxSlot, properties {verticalAlignment=VAlign_Top, padding=<ltrb 6 8 0 0>}}
  }
 ]
}
