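# Widget Blueprint that serves as the left-hand layout container for the XP Everywhere UI, grouping level-up and XP displays.
# A fixed-height SizeBox holds a vertical stack with LevelUpRewardsNotification in the upper slot.
# The lower slot uses a bottom-aligned Overlay to layer XP_Everywhere_Bar over NewLevelWidget.
# The three child widgets are exposed as variables, while the event graph holds only unconnected PreConstruct, Construct and Tick events.
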WidgetBlueprint /XpEverywhereUI/XP_Everywhere_LeftContainer
graphs [EventGraph {nodes [
   Event PreConstruct()
   Event Construct()
   Event Tick()
  ]}]
variables []
widgets {
 info {inheritedWidgetCount=0, parentClass=&UserWidget, rootWidgetClass=&SizeBox, widgetCount=8}
 widgets [
  SizeBox_2 {
   widgetClassPath=&SizeBox
   widget=SizeBox_2
   properties {
    bOverride_HeightOverride=true
    bOverride_MinDesiredWidth=true
    heightOverride=300
    minDesiredHeight=4
    minDesiredWidth=1000
    widthOverride=600
   }
  }
  VerticalBox_0 {widgetClassPath=&VerticalBox, parent=SizeBox_2, widget=VerticalBox_0, slot {class=SizeBoxSlot}}
  SizeBox_0 {
   widgetClassPath=&SizeBox
   parent=VerticalBox_0
   widget=SizeBox_0
   properties {bOverride_HeightOverride=true, heightOverride=96}
   slot {class=VerticalBoxSlot}
  }
  LevelUpRewardsNotification {
   widgetClassPath=&/LevelUpRewardsNotification
   bIsVariable=true
   parent=SizeBox_0
   widget=LevelUpRewardsNotification
   slot {class=SizeBoxSlot}
  }
  SizeBox_1 {
   widgetClassPath=&SizeBox
   parent=VerticalBox_0
   widget=SizeBox_1
   properties {bOverride_HeightOverride=true, heightOverride=160}
   slot {class=VerticalBoxSlot, properties {verticalAlignment=VAlign_Top}}
  }
  Overlay {
   widgetClassPath=&Overlay
   parent=SizeBox_1
   widget=Overlay
   slot {class=SizeBoxSlot, properties {verticalAlignment=VAlign_Bottom}}
  }
  NewLevelWidget {
   widgetClassPath=&/NewLevelWidget
   bIsVariable=true
   parent=Overlay
   widget=NewLevelWidget
   slot {class=OverlaySlot, properties {padding=<ltrb 0 16 0 0>}}
  }
  XP_Everywhere_Bar {
   widgetClassPath=&/XP_Everywhere_Bar
   bIsVariable=true
   parent=Overlay
   widget=XP_Everywhere_Bar
   slot {class=OverlaySlot, properties {verticalAlignment=VAlign_Bottom, padding=<ltrb -16 0 0 0>}}
  }
 ]
}
