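# Post-match XP screen progress bar, a child of WBP_UIKit_ProgressBar with no widgets of its own.
# PreConstruct pushes gradient and leading-circle toggles, progress, delta and circle colors, and behind-glow settings into the inherited ProgressBarMID material.
# AnimateProgressChange calls the parent, then, only when UseDeltaBar is false, plays ProgressAnimation, falling back to Anim_Progress_Base when invalid.
# Playback speed is the reciprocal of TimeDuration, via SafeDivide.
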WidgetBlueprint /FortniteCoreUI/HUD/EndGame/PostGameFlowScreens/DefaultScreens/XPScreen/WBP_PostMatch_ProgressBar
graphs [
 AnimateProgressChange {nodes [
   n1 = AnimateProgressChange!() -> (then, CurrentProgress:double, TargetProgress:double, TimeDuration:double, ProgressAnimation:Widget Animation Object Reference) {is=title K2Node_FunctionEntry}
   > Parent: AnimateProgressChange!(
    execute
    CurrentProgress:double=@n1
    TargetProgress:double=@n1.TargetProgress
    TimeDuration:double=@n1.TimeDuration
    ProgressAnimation:Widget Animation Object Reference=@n1.ProgressAnimation
   ) -> (then) {is=title K2Node_CallParentFunction}
   > n2 = Branch(Condition=Get UseDeltaBar() -> (UseDeltaBar:bool))
   n3 = Get ProgressAnimation() -> (ProgressAnimation:Widget Animation Object Reference)
   PlayAnimation(
    execute=@n2.else
    InAnimation=Select(
     Option 0:Widget Animation Object Reference=Get Anim_Progress_Base() -> (Anim_Progress_Base:Widget Animation Object Reference)
     Option 1:Widget Animation Object Reference=@n3
     Index=IsValid(Object=@n3)
    ) -> (ReturnValue:Widget Animation Object Reference)
    PlaybackSpeed=SafeDivide(A=1, B=Get TimeDuration() -> (TimeDuration:double))
   )
  ]}
 EventGraph {nodes [
   n1 = Event PreConstruct()
   > Parent: PreConstruct(IsDesignTime=@n1.IsDesignTime)
   > SetScalarParameterValue(
    self=Get ProgressBarMID() -> (ProgressBarMID:Material Instance Dynamic Object Reference)
    ParameterName=UseColorGradients
    Value="To Float (Boolean)"(InBool=Get UseColorGradients() -> (UseColorGradients:bool))
   )
   > SetScalarParameterValue(
    self=Get ProgressBarMID()
    ParameterName=UseLeadingCircle
    Value="To Float (Boolean)"(InBool=Get UseLeadingCircle() -> (UseLeadingCircle:bool))
   )
   > SetColorParameters!(
    execute
    self
    Progress Color 1:Linear Color Structure=Get Progress Color 1() -> (Progress Color 1:Linear Color Structure)
    Progress Color 2:Linear Color Structure=Get Progress Color 2() -> (Progress Color 2:Linear Color Structure)
    Delta 1 Color 1:Linear Color Structure=Get Delta 1 Color 1() -> (Delta 1 Color 1:Linear Color Structure)
    Delta 1 Color 2:Linear Color Structure=Get Delta 1 Color 2() -> (Delta 1 Color 2:Linear Color Structure)
    Delta 2 Color 1:Linear Color Structure=Get Delta 2 Color 1() -> (Delta 2 Color 1:Linear Color Structure)
    Delta 2 Color 2:Linear Color Structure=Get Delta 2 Color 2() -> (Delta 2 Color 2:Linear Color Structure)
    LeadingCircleColor:Linear Color Structure=Get Leading Circle Color() -> (Leading Circle Color:Linear Color Structure)
   ) -> (then)
   > SetBehindGlow!(
    execute
    self
    Color:Linear Color Structure=Get BehindGlowColor() -> (BehindGlowColor:Linear Color Structure)
    Density:float=Get BehindGlowDensity() -> (BehindGlowDensity:double)
    Radius:float=Get BehindGlowRadius() -> (BehindGlowRadius:double)
   ) -> (then)
  ]}
 SetBehindGlow {nodes [
   n1 = SetBehindGlow!() -> (then, Color:Linear Color Structure, Density:float=~, Radius:float=~) {is=title K2Node_FunctionEntry}
   n2 = Get ProgressBarMID()
   > SetVectorParameterValue(self=@n2, ParameterName=BehindGlowColor, Value=@n1)
   > SetScalarParameterValue(self=@n2, ParameterName=BehindGlowDensity, Value=@n1.Density)
   > SetScalarParameterValue(self=@n2, ParameterName=BehindGlowRadius, Value=@n1.Radius)
  ]}
 SetColorParameters {nodes [
   n1 = SetColorParameters!() -> (
    then
    Progress Color 1:Linear Color Structure
    Progress Color 2:Linear Color Structure
    Delta 1 Color 1:Linear Color Structure
    Delta 1 Color 2:Linear Color Structure
    Delta 2 Color 1:Linear Color Structure
    Delta 2 Color 2:Linear Color Structure
    LeadingCircleColor:Linear Color Structure
   ) {is=title K2Node_FunctionEntry}
   n2 = Get ProgressBarMID()
   > SetVectorParameterValue(self=@n2, ParameterName=Progress Color 1, Value=@n1)
   > SetVectorParameterValue(self=@n2, ParameterName=Progress Color 2, Value=@n1.Progress Color 2)
   > SetVectorParameterValue(self=@n2, ParameterName=Delta 1 Color 1, Value=@n1.Delta 1 Color 1)
   > SetVectorParameterValue(self=@n2, ParameterName=Delta 1 Color 2, Value=@n1.Delta 1 Color 2)
   > SetVectorParameterValue(self=@n2, ParameterName=Delta 2 Color 1, Value=@n1.Delta 2 Color 1)
   > SetVectorParameterValue(self=@n2, ParameterName=Delta 2 Color 2, Value=@n1.Delta 2 Color 2)
   > SetVectorParameterValue(self=@n2, ParameterName=LeadingCircleColor, Value=@n1.LeadingCircleColor)
  ]}
]
variables [
 BehindGlowColor {type=LinearColor, default=<rgba 0.4 0.1 1 0>}
 BehindGlowDensity {type=number, default=1.6667}
 BehindGlowRadius {type=number, default=0.5}
 Delta 1 Color 1 {type=LinearColor, default=<rgba 0 0.1912 0.2159 1>}
 Delta 1 Color 2 {type=LinearColor, default=<rgba 0 0.8632 0.9734 1>}
 Delta 2 Color 1 {type=LinearColor, default=<rgba 0.9647 0.7379 0.4342 1>}
 Delta 2 Color 2 {type=LinearColor, default=<rgba 0.9647 0.5457 0.1248 1>}
 Leading Circle Color {type=LinearColor, default=<rgba 1 1 1 1>}
 Progress Color 1 {type=LinearColor, default=<rgba 0.3888 0.035 1 1>}
 Progress Color 2 {type=LinearColor, default=<rgba 0.075 0.1058 1 1>}
 UseColorGradients {type=boolean}
 UseLeadingCircle {type=boolean, default=true}
]
widgets {
 info {
  inheritedWidgetCount=0
  parentClass=/FortUILibrary/UIKit/ProgressBar/WBP_UIKit_ProgressBar.WBP_UIKit_ProgressBar_C
  rootWidgetClass=null
  widgetCount=0
 }
 widgets []
}
