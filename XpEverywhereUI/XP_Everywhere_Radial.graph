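# Radial XP widget, derived from XpEverywhereBar, showing level, earned XP, XP to next level and a "Supercharged!" rested-XP badge.
# OnXpBarInitialized seeds the ring and plays the intro; OnXpBarUpdated computes fill endpoints (full on level-up), animates with pulse and sound, and starts a timer.
# Gameplay event router messages accumulate into the next fill's duration.
# Outros lead to DoneUpdatingXpBar; compact mode shrinks ring and fonts, hiding remaining XP.
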WidgetBlueprint /XpEverywhereUI/XP_Everywhere_Radial
animations [
 ProgressSequence {
  binding_count=4
  length_seconds=2.25
  object=&self:ProgressSequence
  sequence {bindings [
    Txt_CurrentLevel_Label {
     id=#1
     possessed_class=CommonTextBlock
     tracks [
      {
       type=MovieSceneFloatTrack
       display_name=Render Opacity
       sections [{type=MovieSceneFloatSection, channels [None {type=MovieSceneScriptingFloatChannel, keys [[2,0],[7,1]]}]}]
      }
      {
       type=MovieScene2DTransformTrack
       sections [{type=MovieScene2DTransformSection, channels [Translation.X {type=MovieSceneScriptingFloatChannel, keys [[2,-24],[7,0]]}]}]
      }
     ]
    }
    RestedXPTextBox {
     id=#2
     possessed_class=Border
     tracks [
      {
       type=MovieSceneFloatTrack
       display_name=Render Opacity
       sections [{type=MovieSceneFloatSection, channels [None {type=MovieSceneScriptingFloatChannel, keys [[1,0],[8,1]]}]}]
      }
      {
       type=MovieScene2DTransformTrack
       sections [{
         type=MovieScene2DTransformSection
         channels [
          Translation.X {type=MovieSceneScriptingFloatChannel, keys [[0,-128],[8,0]]}
          Translation.Y {type=MovieSceneScriptingFloatChannel, const [-40,1]}
         ]
        }]
      }
     ]
    }
    Txt_CurrentLevel {
     id=#3
     possessed_class=CommonTextBlock
     tracks [
      {
       type=MovieSceneFloatTrack
       display_name=Render Opacity
       sections [{type=MovieSceneFloatSection, channels [None {type=MovieSceneScriptingFloatChannel, keys [[2,0],[7,1]]}]}]
      }
      {
       type=MovieScene2DTransformTrack
       sections [{type=MovieScene2DTransformSection, channels [Translation.X {type=MovieSceneScriptingFloatChannel, keys [[2,-24],[7,0]]}]}]
      }
     ]
    }
    All {
     id=#4
     possessed_class=Border
     tracks [
      {
       type=MovieSceneFloatTrack
       display_name=Render Opacity
       sections [{type=MovieSceneFloatSection, channels [None {type=MovieSceneScriptingFloatChannel, keys [[0,0],[2,1],[45,1]]}]}]
      }
      {
       type=MovieScene2DTransformTrack
       sections [{
         type=MovieScene2DTransformSection
         channels [
          Translation.X {type=MovieSceneScriptingFloatChannel, keys [[0,-512],[6,0]]}
          Scale.X {type=MovieSceneScriptingFloatChannel, keys [[1,0.8],[5,1]]}
          Scale.Y {type=MovieSceneScriptingFloatChannel, keys [[1,0.8],[5,1]]}
         ]
        }]
      }
     ]
    }
   ]}
 }
 QuickOutro {
  binding_count=1
  length_seconds=0.2
  object=&self:QuickOutro
  sequence {bindings [All {
     id=#4
     possessed_class=Border
     tracks [
      {
       type=MovieSceneFloatTrack
       display_name=Render Opacity
       sections [{type=MovieSceneFloatSection, channels [None {type=MovieSceneScriptingFloatChannel, keys [[0,1],[4,0]]}]}]
      }
      {
       type=MovieScene2DTransformTrack
       sections [{
         type=MovieScene2DTransformSection
         channels [
          Translation.X {type=MovieSceneScriptingFloatChannel, keys [[0,0],[4,-512]]}
          Scale.X {type=MovieSceneScriptingFloatChannel, keys [[0,1],[4,0.75]]}
          Scale.Y {type=MovieSceneScriptingFloatChannel, keys [[0,1],[4,0.75]]}
         ]
        }]
      }
     ]
    }]}
 }
 ShowXPEarnedAndNextLevel {
  binding_count=3
  length_seconds=2.8
  object=&self:ShowXPEarnedAndNextLevel
  sequence {bindings [
    XPAmountBox {
     id=#5
     possessed_class=HorizontalBox
     tracks [
      {
       type=MovieScene2DTransformTrack
       sections [{type=MovieScene2DTransformSection, channels [Translation.X {type=MovieSceneScriptingFloatChannel, keys [[17,-64],[22,0],[43,0],[48,16]]}]}]
      }
      {
       type=MovieSceneFloatTrack
       display_name=Render Opacity
       sections [{type=MovieSceneFloatSection, channels [None {type=MovieSceneScriptingFloatChannel, keys [[0,0],[17,0],[22,1],[43,1],[48,0]]}]}]
      }
     ]
    }
    Txt_XPToNextLevel_Label {
     id=#6
     possessed_class=CommonTextBlock
     tracks [
      {
       type=MovieSceneFloatTrack
       display_name=Render Opacity
       sections [{type=MovieSceneFloatSection, channels [None {type=MovieSceneScriptingFloatChannel, keys [[50,0],[56,1]]}]}]
      }
      {
       type=MovieScene2DTransformTrack
       sections [{type=MovieScene2DTransformSection, channels [Translation.X {type=MovieSceneScriptingFloatChannel, keys [[51,-80],[56,0]]}]}]
      }
     ]
    }
    Txt_XPToNextLevel {
     id=#7
     possessed_class=CommonTextBlock
     tracks [
      {
       type=MovieSceneFloatTrack
       display_name=Render Opacity
       sections [{type=MovieSceneFloatSection, channels [None {type=MovieSceneScriptingFloatChannel, keys [[50,0],[56,1]]}]}]
      }
      {
       type=MovieScene2DTransformTrack
       sections [{type=MovieScene2DTransformSection, channels [Translation.X {type=MovieSceneScriptingFloatChannel, keys [[51,-80],[56,0]]}]}]
      }
     ]
    }
   ]}
 }
 SlowOutro {
  binding_count=1
  length_seconds=2.5
  object=&self:SlowOutro
  sequence {bindings [All {
     id=#4
     possessed_class=Border
     tracks [
      {
       type=MovieSceneFloatTrack
       display_name=Render Opacity
       sections [{type=MovieSceneFloatSection, channels [None {type=MovieSceneScriptingFloatChannel, keys [[0,1],[40,1],[48,0]]}]}]
      }
      {
       type=MovieScene2DTransformTrack
       sections [{
         type=MovieScene2DTransformSection
         channels [
          Translation.X {type=MovieSceneScriptingFloatChannel, keys [[40,0],[50,-512]]}
          Scale.X {type=MovieSceneScriptingFloatChannel, keys [[40,1],[50,0.75]]}
          Scale.Y {type=MovieSceneScriptingFloatChannel, keys [[40,1],[50,0.75]]}
         ]
        }]
      }
     ]
    }]}
 }
 QuickIntro {
  binding_count=5
  length_seconds=1
  object=&self:QuickIntro
  sequence {bindings [
    WBP_PostMatch_ProgressBar {
     id=#8
     possessed_class=WBP_PostMatch_ProgressBar_C
     tracks [
      {
       type=MovieScene2DTransformTrack
       sections [{
         type=MovieScene2DTransformSection
         channels [
          Scale.X {type=MovieSceneScriptingFloatChannel, keys [[2,2],[10,1]]}
          Scale.Y {type=MovieSceneScriptingFloatChannel, keys [[2,2],[10,1]]}
         ]
        }]
      }
      {
       type=MovieSceneFloatTrack
       display_name=Render Opacity
       sections [{type=MovieSceneFloatSection, channels [None {type=MovieSceneScriptingFloatChannel, keys [[2,0],[10,1]]}]}]
      }
     ]
    }
    Txt_CurrentLevel_Label {
     id=#9
     possessed_class=CommonTextBlock
     tracks [{
       type=MovieSceneFloatTrack
       display_name=Render Opacity
       sections [{type=MovieSceneFloatSection, channels [None {type=MovieSceneScriptingFloatChannel, keys [[8,0],[16,1]]}]}]
      }]
    }
    Txt_CurrentLevel {
     id=#10
     possessed_class=CommonTextBlock
     tracks [
      {
       type=MovieScene2DTransformTrack
       sections [{
         type=MovieScene2DTransformSection
         channels [
          Scale.X {type=MovieSceneScriptingFloatChannel, keys [[5,2],[13,1]]}
          Scale.Y {type=MovieSceneScriptingFloatChannel, keys [[5,2],[13,1]]}
         ]
        }]
      }
      {
       type=MovieSceneFloatTrack
       display_name=Render Opacity
       sections [{type=MovieSceneFloatSection, channels [None {type=MovieSceneScriptingFloatChannel, keys [[5,0],[13,1]]}]}]
      }
     ]
    }
    Image_Backing {
     id=#11
     possessed_class=Image
     tracks [{
       type=MovieSceneFloatTrack
       display_name=Render Opacity
       sections [{type=MovieSceneFloatSection, channels [None {type=MovieSceneScriptingFloatChannel, keys [[10,0],[19,10]]}]}]
      }]
    }
    Image_Glow {
     id=#12
     possessed_class=Image
     tracks [{
       type=MovieSceneFloatTrack
       display_name=Render Opacity
       sections [{type=MovieSceneFloatSection, channels [None {type=MovieSceneScriptingFloatChannel, keys [[0,0],[2,0],[20,1]]}]}]
      }]
    }
   ]}
 }
 ProgressPulse {
  binding_count=2
  length_seconds=1
  object=&self:ProgressPulse
  sequence {bindings [
    Image_halo {
     id=#13
     possessed_class=Image
     tracks [
      {
       type=MovieSceneWidgetMaterialTrack
       sections [{
         type=MovieSceneParameterSection
         channels [
          Inner Radial Manual {type=MovieSceneScriptingFloatChannel, keys [[0,0.5],[20,1]]}
          InnerRadial Opacity {type=MovieSceneScriptingFloatChannel, keys [[0,0],[2,1],[20,0]]}
         ]
        }]
      }
      {
       type=MovieSceneByteTrack
       display_name=Visibility
       sections [{type=MovieSceneByteSection, channels [None {type=MovieSceneScriptingByteChannel, keys [[0,4],[20,2]]}]}]
      }
     ]
    }
    Image_Glow_ProgressPulse {
     id=#14
     possessed_class=Image
     tracks [
      {
       type=MovieSceneWidgetMaterialTrack
       sections [{type=MovieSceneParameterSection, channels [Anim_Pulse {type=MovieSceneScriptingFloatChannel, keys [[0,0],[2,1],[20,0]]}]}]
      }
      {
       type=MovieSceneByteTrack
       display_name=Visibility
       sections [{type=MovieSceneByteSection, channels [None {type=MovieSceneScriptingByteChannel, keys [[0,4],[20,2]]}]}]
      }
      {
       type=MovieSceneFloatTrack
       display_name=Render Opacity
       sections [{type=MovieSceneFloatSection, channels [None {type=MovieSceneScriptingFloatChannel, keys [[0,0],[2,1],[20,0]]}]}]
      }
     ]
    }
   ]}
 }
]
graphs [
 DisplayIsCompact {nodes [
   DisplayIsCompact() {is=title K2Node_FunctionEntry}
   > SetFont(
    self=Get Txt_CurrentLevel_Label() -> (Txt_CurrentLevel_Label:Common Text Object Reference)
    InFontInfo=Select(
     Option 0:Slate Font Info Structure=Get Const_LvlLabelFont() -> (Const_LvlLabelFont:Slate Font Info Structure)
     Option 1:Slate Font Info Structure=Get Const_LvlLabelFont_Small() -> (Const_LvlLabelFont_Small:Slate Font Info Structure)
     Index=Get IsCompact() -> (IsCompact:bool)
    ) -> (ReturnValue:Slate Font Info Structure)
   )
   > SetFont(
    self=Get Txt_CurrentLevel() -> (Txt_CurrentLevel:Common Text Object Reference)
    InFontInfo=Select(
     Option 0=Get Const_LvlFont() -> (Const_LvlFont:Slate Font Info Structure)
     Option 1=Get Const_LvlFont_Small() -> (Const_LvlFont_Small:Slate Font Info Structure)
     Index=Get IsCompact()
    )
   )
   > SetVisibility(
    self=Get XPRemaining() -> (XPRemaining:Overlay Object Reference)
    InVisibility=Select(Option 0:ESlateVisibility Enum=SelfHitTestInvisible, Option 1:ESlateVisibility Enum=Collapsed, Index=Get IsCompact()) -> (ReturnValue:ESlateVisibility Enum)
   )
   n1 = Select(Option 0:double=400, Option 1:double=100, Index=Get IsCompact()) -> (ReturnValue:double)
   > SetSize!(
    execute
    self:WBP UIKit Progress Bar Base Object Reference=Get WBP_PostMatch_ProgressBar() -> (WBP_PostMatch_ProgressBar:WBP Post Match Progress Bar Object Reference)
    InWidth:double=@n1
    InHeight:double=@n1
   ) -> (then)
   > SetThickness!(
    execute
    self:WBP UIKit Progress Bar Base Object Reference=Get WBP_PostMatch_ProgressBar()
    Thickness_PX:double=Select(Option 0=25, Option 1=10, Index=Get IsCompact())
   ) -> (then)
   > SetOutline!(
    execute
    self:WBP UIKit Progress Bar Base Object Reference=Get WBP_PostMatch_ProgressBar()
    UseOutline:bool=true
    OutlinePX:double=Select(Option 0=4, Option 1=0, Index=Get IsCompact())
   ) -> (then)
   > SetPadding!(
    execute
    self:Overlay Slot Object Reference=SlotAsOverlaySlot(Widget=Get Txt_CurrentLevel_Label())
    InPadding_Left:float=0
    InPadding_Top:float=0
    InPadding_Right:float=0
    InPadding_Bottom:float=Select(Option 0=80, Option 1=13, Index=Get IsCompact())
   ) -> (then)
   > SetPadding(
    self=SlotAsOverlaySlot(Widget=Get Txt_CurrentLevel())
    InPadding_Left=0
    InPadding_Top=Select(Option 0=0, Option 1=8, Index=Get IsCompact())
    InPadding_Right=0
    InPadding_Bottom=0
   )
   > SetColorAndOpacity!(
    execute
    self=Get Txt_CurrentLevel_Label()
    InColorAndOpacity_SpecifiedColor:Linear Color Structure=Select(Option 0:Linear Color Structure=(R=1,G=1,B=1,A=1), Option 1:Linear Color Structure=(R=0.3281,G=0.3281,B=0.3281,A=1), Index=Get IsCompact()) -> (ReturnValue:Linear Color Structure)
    InColorAndOpacity_ColorUseRule:ESlateColorStylingMode Enum
   ) -> (then)
  ]}
 EventGraph {nodes [
   n1 = UpdateText!() -> (OutputDelegate, then, DisplayedLevel:int, XpForDisplayedLevel:int) {is=title K2Node_CustomEvent}
   > "SetText (Text)"(self=Get Txt_CurrentLevel(), InText="To Text (Integer)"(Value=@n1.DisplayedLevel))
   n2 = Get TotalXpRequiredForLevel() -> (TotalXpRequiredForLevel:int=0)
   > "SetText (Text)"(
    self=Get Txt_XPToNextLevel() -> (Txt_XPToNextLevel:Common Text Object Reference)
    InText=Format Text!(Format="LOCTEXT(\"{RemainingXP} XP\")", RemainingXP:int=int - int(A=@n2, B=@n1.XpForDisplayedLevel)) -> (Result)
   )
   > "SetText (Text)"(
    self=Get Txt_XPToNextLevel_Label() -> (Txt_XPToNextLevel_Label:Common Text Object Reference)
    InText=Format Text!(Format="LOCTEXT(\"TO LVL {level}\")", level:int=int + int(A=@n1.DisplayedLevel)) -> (Result)
   )
   n3 = "Equal (Integer)"(A=Get Shown Level() -> (Shown Level:int=0), B=Get New Desired Level() -> (New Desired Level:int=0))
   n4 = int - int(A=Get TotalXpRequiredForLevel() -> (TotalXpRequiredForLevel=0), B=Get Shown XP() -> (Shown XP:int=0))
   n5 = int - int(A=Get NewDesiredXP() -> (NewDesiredXP:int=0), B=Get Shown XP() -> (Shown XP=0))
   > SetText(
    self=Get XpAmount() -> (XpAmount:Common Rich Text Block Object Reference)
    InText=Format Text!(Format="LOCTEXT(\"+{EarnedXP}\")", EarnedXP:int=SelectInt(A=@n5, B=@n4, bPickA=@n3)) -> (Result)
   )
   OutroFinished() {is=title K2Node_CustomEvent}
   > DoneUpdatingXpBar(self:Xp Everywhere Bar Object Reference)
   TriggerOutro() {is=title K2Node_CustomEvent}
   n6 = Get bXpBarAlwaysActive() -> (bXpBarAlwaysActive:bool)
   n7 = Get bIsXpBarManuallyUpdated() -> (bIsXpBarManuallyUpdated:bool)
   > n8 = Branch(Condition=OR Boolean(A=NOT Boolean(A=@n6), B=NOT Boolean(A=@n7)))
   > PlayAnimation(InAnimation=Select(
     Option 0:Widget Animation Object Reference=Get SlowOutro() -> (SlowOutro:Widget Animation Object Reference)
     Option 1:Widget Animation Object Reference=Get QuickOutro() -> (QuickOutro:Widget Animation Object Reference)
     Index=Get IsLevelingUp() -> (IsLevelingUp:bool)
    ) -> (ReturnValue:Widget Animation Object Reference))
   OutroFinished(execute=@n8.else) {is=title K2Node_CallFunction}
   n9 = Event OnXpBarUpdated!() -> (
    OutputDelegate
    then
    ShownXP:int=0
    ShownLevel:int=0
    NewDesiredXP:int=0
    NewDesiredLevel:int=0
    NewRemainingRestXP:int=0
    NewRestedXpConsumed:int=0
   ) {is=title K2Node_Event}
   > Set Shown XP(Shown XP:int=@n9.ShownXP) -> (Output_Get:int=0)
   > Set Shown Level(Shown Level:int=@n9.ShownLevel) -> (Output_Get:int=0)
   > Set NewDesiredXP(NewDesiredXP:int=@n9.NewDesiredXP) -> (Output_Get:int=0)
   > n10 = Set New Desired Level(New Desired Level:int=@n9.NewDesiredLevel) -> (Output_Get:int=0)
   > Set IsLevelingUp(IsLevelingUp:bool="integer > integer"(A=@n10, B=Get Shown Level() -> (Shown Level=0))) -> (Output_Get:bool)
   > SetRestedXPVisualState!(execute, self, IsRested:bool="integer > integer"(A=@n9.NewRemainingRestXP)) -> (then)
   n11 = Get Shown Level() -> (Shown Level=0)
   > Set TotalXpRequiredForLevel(TotalXpRequiredForLevel:int=Select(
     Option 0:int=GetTotalXpRequiredForLevel!(self:Xp Everywhere Bar Object Reference, InLevel:int=@n11) -> (ReturnValue:int=0)
     Option 1:int=Get DEBUG_totalxpforlevel() -> (DEBUG_totalxpforlevel:int=0)
     Index=Get Is Design Time() -> (Is Design Time:bool)
    ) -> (ReturnValue:int)) -> (Output_Get:int=0)
   n12 = Get TotalXpRequiredForLevel() -> (TotalXpRequiredForLevel=0)
   n13 = Get Shown XP() -> (Shown XP=0)
   > Set ProgressStartPosition(ProgressStartPosition:double=float / float(A="To Float (Integer)"(InInt=@n13), B="To Float (Integer)"(InInt=@n12))) -> (Output_Get:double)
   n14 = "To Float (Integer)"(InInt=Get TotalXpRequiredForLevel() -> (TotalXpRequiredForLevel=0))
   n15 = "To Float (Integer)"(InInt=Get NewDesiredXP() -> (NewDesiredXP=0))
   > Set ProgressStopPosition(ProgressStopPosition:double=SelectFloat(A=1, B=float / float(A=@n15, B=@n14), bPickA=Get IsLevelingUp())) -> (Output_Get:double)
   > UpdateText!(
    execute
    self
    DisplayedLevel:int=Get Shown Level() -> (Shown Level=0)
    XpForDisplayedLevel:int=SelectInt(A=Get TotalXpRequiredForLevel() -> (TotalXpRequiredForLevel=0), B=Get NewDesiredXP() -> (NewDesiredXP=0), bPickA=Get IsLevelingUp())
   ) -> (then) {is=title K2Node_CallFunction}
   > SetProgress!(
    execute
    self:WBP UIKit Progress Bar Base Object Reference=Get WBP_PostMatch_ProgressBar()
    Progress:double=Get ProgressStartPosition() -> (ProgressStartPosition:double)
   ) -> (then)
   n16 = Get NextDuration() -> (NextDuration:double)
   > SetDurationTime!(
    execute
    self:WBP UIKit Progress Bar Base Object Reference=Get WBP_PostMatch_ProgressBar()
    DurationTime:double=Select(Option 0:double=Get DefaultDuration() -> (DefaultDuration:float), Option 1:double=@n16, Index="float > float"(A=@n16)) -> (ReturnValue:double)
   ) -> (then)
   > Set NextDuration(NextDuration:double=0) -> (Output_Get:double)
   > AnimateProgressBar()
   > n17 = Branch(Condition=IsUsingCompactUI())
   > n18 = PlaySound2D(Sound=/Game/Sounds/Fort_UI_Sounds/Gameplay/XP/Fort_XP_Score_Popup_Cue)
   PlayAnimation(execute=[@n18.then, @n17.else], InAnimation=Get ProgressPulse() -> (ProgressPulse:Widget Animation Object Reference))
   > Branch(Condition=NOT Boolean(A=Get Is Design Time()))
   > n19 = Set Timer by Event(
    Delegate=Create Event()
    Time=Get DurationTime!(self:WBP UIKit Progress Bar Base Object Reference=Get WBP_PostMatch_ProgressBar()) -> (DurationTime:double)
    bLooping=false
   )
   > Set OutroTimerHandle(OutroTimerHandle:Timer Handle Structure=@n19) -> (Output_Get:Timer Handle Structure)
   > PlayAnimation(InAnimation=Get ShowXPEarnedAndNextLevel() -> (ShowXPEarnedAndNextLevel:Widget Animation Object Reference))
   n20 = Event OnXpBarInitialized!() -> (OutputDelegate, then, CurrentXP:int=0, CurrentLevel:int=0, CurrentRemainingRestXP:int=0, RundownLineItemDuration:float) {is=title K2Node_Event}
   > Set Shown XP(Shown XP=@n20.CurrentXP) -> (Output_Get=0)
   > Set Shown Level(Shown Level=@n20.CurrentLevel) -> (Output_Get=0)
   > n21 = Set DefaultDuration(DefaultDuration:float=@n20.RundownLineItemDuration) -> (Output_Get:float)
   > SetDurationTime(self=Get WBP_PostMatch_ProgressBar(), DurationTime=@n21)
   n22 = Get Shown Level() -> (Shown Level=0)
   > Set TotalXpRequiredForLevel(TotalXpRequiredForLevel=Select(
     Option 0:int=GetTotalXpRequiredForLevel(InLevel=@n22) -> (ReturnValue=0)
     Option 1:int=Get DEBUG_totalxpforlevel() -> (DEBUG_totalxpforlevel=0)
     Index=Get Is Design Time()
    ) -> (ReturnValue:int)) -> (Output_Get=0)
   n23 = Get TotalXpRequiredForLevel() -> (TotalXpRequiredForLevel=0)
   n24 = Get Shown XP() -> (Shown XP=0)
   > Set ProgressStartPosition(ProgressStartPosition=float / float(A="To Float (Integer)"(InInt=@n24), B="To Float (Integer)"(InInt=@n23)))
   > UpdateText(DisplayedLevel=Get Shown Level() -> (Shown Level=0), XpForDisplayedLevel=Get Shown XP() -> (Shown XP=0))
   > SetProgress(self=Get WBP_PostMatch_ProgressBar(), Progress=Get ProgressStartPosition())
   > SetRenderOpacity(self=Get All() -> (All:Border Object Reference), InOpacity=1)
   > SetRenderOpacity(self=Get XPRemaining(), InOpacity=0)
   > PlayXPRadialArrivalIntro()
   n25 = Event OnSetIsCompact!() -> (OutputDelegate, then, IsCompact:bool) {is=title K2Node_Event}
   > SetIsSmallDisplay!(execute, self, IsCompact:bool=@n25.IsCompact) -> (then)
   Event OnInitialized()
   > BindToAnimationFinished(Animation=Get ProgressSequence() -> (ProgressSequence:Widget Animation Object Reference), Delegate=Create Event())
   n26 = Create Event()
   > BindToAnimationFinished(Animation=Get SlowOutro(), Delegate=@n26)
   > BindToAnimationFinished(Animation=Get QuickOutro(), Delegate=@n26)
   Event Construct()
   > SetRenderOpacity(self=Get All(), InOpacity=0)
   n27 = GetGlobalEventRouter()
   > n28 = StartListeningToEvent!(
    execute:Exec
    Target:Gameplay Event Router Component Object Reference=@n27
    EventType:Script Struct Object Reference=/FortniteCoreUI/HUD/EndGame/Data/S_UI_SetRundownProgressUpdateDuration
    Context:Object Reference
    EventBubblingRule:EEventBubblingRule Enum
    None:%=@n27
   ) -> (
    then:Exec
    AsyncTaskProxy:Async Action Start Listening To Event Object Reference
    OnEventReceived:Exec
    Payload:S UI Set Rundown Progress Update Duration Structure
   )
   > Set SetRundownDurationAsyncAction(SetRundownDurationAsyncAction:Async Action Start Listening To Event Object Reference=@n28) -> (Output_Get:Async Action Start Listening To Event Object Reference)
   Set NextDuration(
    execute=@n28.OnEventReceived
    NextDuration=float + float(
     A=Break S UI Set Rundown Progress Update Duration!(S_UI_SetRundownProgressUpdateDuration:"S UI Set Rundown Progress Update Duration Structure (by ref)"=@n28.Payload) -> (RundownItemDuration_2_#15:double=~) {is=title K2Node_BreakStruct}
     B=Get NextDuration()
    )
   )
   Event Destruct()
   > StopListeningToEvent(self:Async Action Start Listening To Event Object Reference=Get SetRundownDurationAsyncAction() -> (SetRundownDurationAsyncAction:Async Action Start Listening To Event Object Reference))
  ]}
 PlayXPRadialArrivalIntro {nodes [
   PlayXPRadialArrivalIntro() {is=title K2Node_FunctionEntry}
   > PlayAnimation(InAnimation=Get QuickIntro() -> (QuickIntro:Widget Animation Object Reference))
  ]}
 Sequencer Events {nodes [
   AnimateProgressBar() {is=title K2Node_CustomEvent}
   n1 = Get WBP_PostMatch_ProgressBar()
   > SetProgressTransitionWithCustomAnimation!(
    execute
    self:WBP UIKit Progress Bar Base Object Reference=@n1
    Progress:double=Get ProgressStopPosition() -> (ProgressStopPosition:double)
    ProgressAnimation:Widget Animation Object Reference=Get Anim_Progress_Base!(self:WBP UIKit Progress Bar Base Object Reference=@n1) -> (Anim_Progress_Base:Widget Animation Object Reference)
   ) -> (then)
  ]}
 SetIsSmallDisplay {nodes [
   n1 = SetIsSmallDisplay!() -> (then, IsCompact:bool=~) {is=title K2Node_FunctionEntry}
   > Set IsCompact(IsCompact:bool=@n1) -> (Output_Get:bool)
   > DisplayIsCompact() {is=title K2Node_CallFunction}
  ]}
 SetRestedXPVisualState {nodes [
   n1 = SetRestedXPVisualState!() -> (then, IsRested:bool=~) {is=title K2Node_FunctionEntry}
   > SetVisibility(
    self=Get RestedXPTextBox() -> (RestedXPTextBox:Border Object Reference)
    InVisibility=Select(Option 0:ESlateVisibility Enum=Collapsed, Option 1:ESlateVisibility Enum=SelfHitTestInvisible, Index=@n1) -> (ReturnValue:ESlateVisibility Enum)
   )
  ]}
 SetXpToNextLevelTextVisualState {nodes [
   SetXpToNextLevelTextVisualState() {is=title K2Node_FunctionEntry}
   n1 = Get ShowXPEarnedAndNextLevel()
   > PlayAnimation(InAnimation=@n1, StartAtTime=GetEndTime(self=@n1))
  ]}
]
variables [
 AdditionalDelay {type=number, default=1}
 AlwaysActiveProgressAnimStartTime {type=number, default=0.85}
 Anim Speed {type=number, default=0.05}
 Const_LvlFont {
  type=SlateFontInfo
  default {
   bForceMonospaced=false
   bMaterialIsStencil=false
   fontMaterial=null
   fontObject=&HeadingNow
   letterSpacing=0
   monospacedWidth=1
   size=96
   skewAmount=0
   typefaceFontName=86 Bold
   outlineSettings {
    bApplyOutlineToDropShadows=false
    bMiteredCorners=false
    bSeparateFillAlpha=false
    outlineMaterial=null
    outlineSize=0
    outlineColor=<rgba 0 0 0 1>
   }
  }
 }
 Const_LvlFont_Small {
  type=SlateFontInfo
  default {
   bForceMonospaced=false
   bMaterialIsStencil=false
   fontMaterial=null
   fontObject=&HeadingNow
   letterSpacing=0
   monospacedWidth=1
   size=18.75
   skewAmount=0
   typefaceFontName=86 Bold
   outlineSettings {
    bApplyOutlineToDropShadows=false
    bMiteredCorners=false
    bSeparateFillAlpha=false
    outlineMaterial=null
    outlineSize=0
    outlineColor=<rgba 0 0 0 1>
   }
  }
 }
 Const_LvlLabelFont {
  type=SlateFontInfo
  default {
   bForceMonospaced=false
   bMaterialIsStencil=false
   fontMaterial=null
   fontObject=&HeadingNow
   letterSpacing=0
   monospacedWidth=1
   size=25.5
   skewAmount=0
   typefaceFontName=86 Bold Italic
   outlineSettings {
    bApplyOutlineToDropShadows=false
    bMiteredCorners=false
    bSeparateFillAlpha=false
    outlineMaterial=null
    outlineSize=0
    outlineColor=<rgba 0 0 0 1>
   }
  }
 }
 Const_LvlLabelFont_Small {
  type=SlateFontInfo
  default {
   bForceMonospaced=false
   bMaterialIsStencil=false
   fontMaterial=null
   fontObject=&HeadingNow
   letterSpacing=0
   monospacedWidth=1
   size=11.25
   skewAmount=0
   typefaceFontName=75 Medium
   outlineSettings {
    bApplyOutlineToDropShadows=false
    bMiteredCorners=false
    bSeparateFillAlpha=false
    outlineMaterial=null
    outlineSize=0
    outlineColor=<rgba 0 0 0 1>
   }
  }
 }
 DEBUG_newdesiredlevel {type=integer, default=1}
 DEBUG_newdesiredxp {type=integer, default=700}
 DEBUG_showlevel {type=integer, default=1}
 DEBUG_shownxp {type=integer, default=500}
 DEBUG_totalxpforlevel {type=integer, default=1000}
 DefaultDuration {type=number}
 HasLeveledAnimPlayed {type=boolean}
 Is Design Time {type=boolean}
 IsCompact {type=boolean}
 IsLevelingUp {type=boolean}
 New Desired Level {type=integer}
 NewDesiredXP {type=integer}
 NextDuration {type=number}
 OutroTimerHandle {type=TimerHandle}
 ProgressBarMID {type=MaterialInstanceDynamic}
 ProgressStartPosition {type=number}
 ProgressStopPosition {type=number}
 SetRundownDurationAsyncAction {type=AsyncAction_StartListeningToEvent, default=null}
 Shown Level {type=integer}
 Shown XP {type=integer}
 TotalXpRequiredForLevel {type=integer}
 XpEarned {type=number}
 XpFillRate {type=integer, default=15000}
]
widgets {
 info {inheritedWidgetCount=1, parentClass=/Script/XpEverywhereUI.XpEverywhereBar, rootWidgetClass=&Border, widgetCount=25}
 widgets [
  All {
   widgetClassPath=&Border
   bIsVariable=true
   widget=All
   properties {
    visibility=SelfHitTestInvisible
    background {
     bIsDynamicallyLoaded=false
     drawAs=NoDrawType
     imageType=NoImage
     mirroring=NoMirror
     resourceName=None
     resourceObject=/FortUILibrary/Foundation/Art_ToolKit/Tools/M_UI_XP_Everywhere_Backing_DropShadow
     tiling=NoTile
     imageSize=<40 40>
     margin=<ltrb 0.5 0.5 0.5 0.5>
     outlineSettings {
      bUseBrushTransparency=false
      roundingType=HalfHeightRadius
      width=0
      color {colorUseRule=UseColor_Specified, specifiedColor=<rgba 0 0 0 0>}
      cornerRadii=<0 0 0 0>
     }
     tintColor {colorUseRule=UseColor_Specified, specifiedColor=<rgba 1 1 1 1>}
     uVRegion {bIsValid=false, max=<0 0>, min=<0 0>}
    }
    padding=<ltrb 16 16 16 16>
    renderTransformPivot=<0 0.5>
   }
  }
  Root {widgetClassPath=&Overlay, parent=All, widget=Root, slot {class=BorderSlot, properties {padding=<ltrb 16 16 16 16>}}}
  Image_Backing {
   widgetClassPath=&Image
   parent=Root
   widget=Image_Backing
   properties {
    visibility=SelfHitTestInvisible
    brush {
     bIsDynamicallyLoaded=false
     drawAs=Image
     imageType=FullColor
     mirroring=NoMirror
     resourceName=None
     resourceObject=/FortniteCoreUI/HUD/EndGame/PostGameFlowScreens/DefaultScreens/XPScreen/Materials/MI_PostMatch_RadialLevel
     tiling=NoTile
     imageSize=<64 64>
     margin=<ltrb 0 0 0 0>
     outlineSettings {
      bUseBrushTransparency=false
      roundingType=HalfHeightRadius
      width=0
      color {colorUseRule=UseColor_Specified, specifiedColor=<rgba 0 0 0 0>}
      cornerRadii=<0 0 0 0>
     }
     tintColor {colorUseRule=UseColor_Specified, specifiedColor=<rgba 1 1 1 1>}
     uVRegion {bIsValid=false, max=<0 0>, min=<0 0>}
    }
    renderTransform {angle=-90, scale=<1 1>, shear=<0 0>, translation=<0 0>}
   }
   slot {class=OverlaySlot, properties {horizontalAlignment=HAlign_Fill, verticalAlignment=VAlign_Fill}}
  }
  Image_Glow {
   widgetClassPath=&Image
   parent=Root
   widget=Image_Glow
   properties {
    visibility=SelfHitTestInvisible
    brush {
     bIsDynamicallyLoaded=false
     drawAs=Image
     imageType=FullColor
     mirroring=NoMirror
     resourceName=None
     resourceObject=/FortniteCoreUI/HUD/EndGame/PostGameFlowScreens/DefaultScreens/XPScreen/Materials/M_UI_PostMatchTextGlow
     tiling=NoTile
     imageSize=<64 64>
     margin=<ltrb 0 0 0 0>
     outlineSettings {
      bUseBrushTransparency=false
      roundingType=HalfHeightRadius
      width=0
      color {colorUseRule=UseColor_Specified, specifiedColor=<rgba 0 0 0 0>}
      cornerRadii=<0 0 0 0>
     }
     tintColor {colorUseRule=UseColor_Specified, specifiedColor=<rgba 0.333 0.333 0.333 1>}
     uVRegion {bIsValid=false, max=<0 0>, min=<0 0>}
    }
    renderTransform {angle=0, scale=<2 2>, shear=<0 0>, translation=<0 0>}
   }
   slot {class=OverlaySlot, properties {horizontalAlignment=HAlign_Fill, verticalAlignment=VAlign_Fill}}
  }
  Image_Glow_ProgressPulse {
   widgetClassPath=&Image
   parent=Root
   widget=Image_Glow_ProgressPulse
   properties {
    visibility=Hidden
    brush {
     bIsDynamicallyLoaded=false
     drawAs=Image
     imageType=FullColor
     mirroring=NoMirror
     resourceName=None
     resourceObject=&M_UI_PostMatchTextGlow
     tiling=NoTile
     imageSize=<64 64>
     margin=<ltrb 0 0 0 0>
     outlineSettings {
      bUseBrushTransparency=false
      roundingType=HalfHeightRadius
      width=0
      color {colorUseRule=UseColor_Specified, specifiedColor=<rgba 0 0 0 0>}
      cornerRadii=<0 0 0 0>
     }
     tintColor {colorUseRule=UseColor_Specified, specifiedColor=<rgba 0.5 0.5 0.5 1>}
     uVRegion {bIsValid=false, max=<0 0>, min=<0 0>}
    }
   }
   slot {class=OverlaySlot, properties {horizontalAlignment=HAlign_Fill, verticalAlignment=VAlign_Fill}}
  }
  Image_halo {
   widgetClassPath=&Image
   parent=Root
   widget=Image_halo
   properties {
    visibility=Hidden
    brush {
     bIsDynamicallyLoaded=false
     drawAs=Image
     imageType=FullColor
     mirroring=NoMirror
     resourceName=None
     resourceObject=/FortniteCoreUI/HUD/EndGame/PostGameFlowScreens/DefaultScreens/XPScreen/Materials/MI_PostMatch_LevelUp_PulseGlow
     tiling=NoTile
     imageSize=<64 64>
     margin=<ltrb 0 0 0 0>
     outlineSettings {
      bUseBrushTransparency=false
      roundingType=HalfHeightRadius
      width=0
      color {colorUseRule=UseColor_Specified, specifiedColor=<rgba 0 0 0 0>}
      cornerRadii=<0 0 0 0>
     }
     tintColor {colorUseRule=UseColor_Specified, specifiedColor=<rgba 0.25 0.25 0.25 1>}
     uVRegion {bIsValid=false, max=<0 0>, min=<0 0>}
    }
    renderTransform {angle=0, scale=<3 3>, shear=<0 0>, translation=<0 0>}
   }
   slot {class=OverlaySlot, properties {horizontalAlignment=HAlign_Fill, verticalAlignment=VAlign_Fill}}
  }
  WBP_PostMatch_ProgressBar {
   widgetClassPath=/FortniteCoreUI/HUD/EndGame/PostGameFlowScreens/DefaultScreens/XPScreen/WBP_PostMatch_ProgressBar
   bIsVariable=true
   parent=Root
   widget=WBP_PostMatch_ProgressBar
   properties {
    behindGlowDensity=0
    behindGlowRadius=0
    edgeOffset_PX=12
    glowIntensity=0.125
    progressValue=0.5
    thickness_PX=16
    colorA=<rgba 0.75 0.9167 1 1>
    colorB=<rgba 0.15 0.4333 1 1>
    colorC=<rgba 0.1 0.175 1 1>
    progress Color 1=<rgba 0.55 0.1 1 1>
    progress Color 2=<rgba 0.5 0.5417 1 1>
   }
   slot {class=OverlaySlot, properties {horizontalAlignment=HAlign_Fill, verticalAlignment=VAlign_Fill}}
  }
  Txt_CurrentLevel_Label {
   widgetClassPath=&CommonTextBlock
   bIsVariable=true
   parent=Root
   widget=Txt_CurrentLevel_Label
   properties {
    bIsScrollingEnabled=false
    text=LVL
    font {
     bForceMonospaced=false
     bMaterialIsStencil=false
     fontMaterial=null
     fontObject=&HeadingNow
     letterSpacing=0
     monospacedWidth=1
     size=18
     skewAmount=0
     typefaceFontName=86 Bold Italic
     outlineSettings {
      bApplyOutlineToDropShadows=false
      bMiteredCorners=false
      bSeparateFillAlpha=false
      outlineMaterial=null
      outlineSize=0
      outlineColor=<rgba 0 0 0 1>
     }
    }
    shadowOffset=<0 0>
   }
   slot {class=OverlaySlot, properties {horizontalAlignment=HAlign_Center, verticalAlignment=VAlign_Center, padding=<ltrb 0 0 0 80>}}
  }
  Txt_CurrentLevel {
   widgetClassPath=&CommonTextBlock
   bIsVariable=true
   parent=Root
   widget=Txt_CurrentLevel
   properties {
    text=26
    font {
     bForceMonospaced=false
     bMaterialIsStencil=false
     fontMaterial=null
     fontObject=&HeadingNow
     letterSpacing=0
     monospacedWidth=1
     size=96
     skewAmount=0
     typefaceFontName=86 Bold
     outlineSettings {
      bApplyOutlineToDropShadows=false
      bMiteredCorners=false
      bSeparateFillAlpha=true
      outlineMaterial=/FortniteCoreUI/HUD/EndGame/PostGameFlowScreens/DefaultScreens/XPScreen/Materials/MI_PostMatch_FontGlow
      outlineSize=64
      outlineColor=<rgba 1 1 1 1>
     }
    }
    shadowOffset=<0 0>
   }
   slot {class=OverlaySlot, properties {horizontalAlignment=HAlign_Center, verticalAlignment=VAlign_Center}}
  }
  RestedXPTextBox {
   widgetClassPath=&Border
   bIsVariable=true
   parent=Root
   widget=RestedXPTextBox
   properties {
    visibility=Collapsed
    background {
     bIsDynamicallyLoaded=false
     drawAs=NoDrawType
     imageType=NoImage
     mirroring=NoMirror
     resourceName=None
     resourceObject=null
     tiling=NoTile
     imageSize=<32 32>
     margin=<ltrb 0 0 0 0>
     outlineSettings {
      bUseBrushTransparency=false
      roundingType=HalfHeightRadius
      width=0
      color {colorUseRule=UseColor_Specified, specifiedColor=<rgba 0 0 0 0>}
      cornerRadii=<0 0 0 0>
     }
     tintColor {colorUseRule=UseColor_Specified, specifiedColor=<rgba 1 1 1 1>}
     uVRegion {bIsValid=false, max=<0 0>, min=<0 0>}
    }
    padding=<ltrb 0 0 0 0>
    renderTransform {angle=0, scale=<1 1>, shear=<0 0>, translation=<0 -48>}
   }
   slot {class=OverlaySlot, properties {horizontalAlignment=HAlign_Center}}
  }
  Overlay_4 {widgetClassPath=&Overlay, parent=RestedXPTextBox, widget=Overlay_4, slot {class=BorderSlot, properties {padding=<ltrb 0 0 0 0>}}}
  Image_SuperchargedGlow {
   widgetClassPath=&Image
   parent=Overlay_4
   widget=Image_SuperchargedGlow
   properties {
    visibility=SelfHitTestInvisible
    brush {
     bIsDynamicallyLoaded=false
     drawAs=Box
     imageType=NoImage
     mirroring=NoMirror
     resourceName=None
     resourceObject=/Game/Athena/HUD/LTM/Wax/Materials/M_UI_TextGlow_Add_Gradient_Purple_XP
     tiling=NoTile
     imageSize=<48 48>
     margin=<ltrb 0.5 0.5 0.5 0.5>
     outlineSettings {
      bUseBrushTransparency=false
      roundingType=HalfHeightRadius
      width=0
      color {colorUseRule=UseColor_Specified, specifiedColor=<rgba 0 0 0 0>}
      cornerRadii=<0 0 0 0>
     }
     tintColor {colorUseRule=UseColor_Specified, specifiedColor=<rgba 1 1 1 1>}
     uVRegion {bIsValid=false, max=<0 0>, min=<0 0>}
    }
    colorAndOpacity=<rgba 1 1 1 0.0625>
    renderTransform {angle=0, scale=<1 1>, shear=<-12 0>, translation=<0 0>}
   }
   slot {class=OverlaySlot, properties {horizontalAlignment=HAlign_Fill, verticalAlignment=VAlign_Fill, padding=<ltrb -8 -8 -12 -8>}}
  }
  HorizontalBox_1 {
   widgetClassPath=&HorizontalBox
   parent=Overlay_4
   widget=HorizontalBox_1
   slot {class=OverlaySlot, properties {verticalAlignment=VAlign_Center}}
  }
  Overlay_3 {
   widgetClassPath=&Overlay
   parent=HorizontalBox_1
   widget=Overlay_3
   properties {renderTransform {angle=0, scale=<1 1>, shear=<0 0>, translation=<0 1>}}
   slot {class=HorizontalBoxSlot}
  }
  Image_XPLightning {
   widgetClassPath=&Image
   parent=Overlay_3
   widget=Image_XPLightning
   properties {
    renderOpacity=0
    visibility=Collapsed
    brush {
     bIsDynamicallyLoaded=false
     drawAs=Image
     imageType=NoImage
     mirroring=NoMirror
     resourceName=None
     resourceObject=/Game/Athena/HUD/XP/Materials/M_UI_XPLightning
     tiling=NoTile
     imageSize=<32 32>
     margin=<ltrb 0 0 0 0>
     outlineSettings {
      bUseBrushTransparency=false
      roundingType=HalfHeightRadius
      width=0
      color {colorUseRule=UseColor_Specified, specifiedColor=<rgba 0 0 0 0>}
      cornerRadii=<0 0 0 0>
     }
     tintColor {colorUseRule=UseColor_Specified, specifiedColor=<rgba 1 1 1 1>}
     uVRegion {bIsValid=false, max=<0 0>, min=<0 0>}
    }
    renderTransform {angle=0, scale=<2 2>, shear=<0 0>, translation=<0 0>}
   }
   slot {class=OverlaySlot}
  }
  Image_XPGlow {
   widgetClassPath=&Image
   parent=Overlay_3
   widget=Image_XPGlow
   properties {
    visibility=SelfHitTestInvisible
    brush {
     bIsDynamicallyLoaded=false
     drawAs=Image
     imageType=NoImage
     mirroring=NoMirror
     resourceName=None
     resourceObject=&M_UI_TextGlow_Add_Gradient_Purple_XP
     tiling=NoTile
     imageSize=<32 32>
     margin=<ltrb 0 0 0 0>
     outlineSettings {
      bUseBrushTransparency=false
      roundingType=HalfHeightRadius
      width=0
      color {colorUseRule=UseColor_Specified, specifiedColor=<rgba 0 0 0 0>}
      cornerRadii=<0 0 0 0>
     }
     tintColor {colorUseRule=UseColor_Specified, specifiedColor=<rgba 1 1 1 1>}
     uVRegion {bIsValid=false, max=<0 0>, min=<0 0>}
    }
    colorAndOpacity=<rgba 1 1 1 0.125>
    renderTransform {angle=0, scale=<2 2>, shear=<0 0>, translation=<0 0>}
   }
   slot {class=OverlaySlot, properties {horizontalAlignment=HAlign_Center, verticalAlignment=VAlign_Center}}
  }
  Image_XPIcon {
   widgetClassPath=&Image
   parent=Overlay_3
   widget=Image_XPIcon
   properties {
    brush {
     bIsDynamicallyLoaded=false
     drawAs=Image
     imageType=FullColor
     mirroring=NoMirror
     resourceName=None
     resourceObject=/Game/UI/Foundation/Textures/Icons/Items/UniversalXP/T_UI_UniversalXP_Icon_Standard_64
     tiling=NoTile
     imageSize=<48 48>
     margin=<ltrb 0 0 0 0>
     outlineSettings {
      bUseBrushTransparency=false
      roundingType=HalfHeightRadius
      width=0
      color {colorUseRule=UseColor_Specified, specifiedColor=<rgba 0 0 0 0>}
      cornerRadii=<0 0 0 0>
     }
     tintColor {colorUseRule=UseColor_Specified, specifiedColor=<rgba 1 1 1 1>}
     uVRegion {bIsValid=false, max=<0 0>, min=<0 0>}
    }
    renderTransform {angle=0, scale=<1 1>, shear=<0 0>, translation=<0 1>}
   }
   slot {class=OverlaySlot}
  }
  CommonTextBlock_71 {
   widgetClassPath=&CommonTextBlock
   parent=HorizontalBox_1
   widget=CommonTextBlock_71
   properties {
    mobileFontSizeMultiplier=1.05
    text="Supercharged!"
    textTransformPolicy=ToUpper
    font {
     bForceMonospaced=false
     bMaterialIsStencil=false
     fontMaterial=null
     fontObject=&HeadingNow
     letterSpacing=0
     monospacedWidth=1
     size=18
     skewAmount=0
     typefaceFontName=86 Bold
     outlineSettings {
      bApplyOutlineToDropShadows=false
      bMiteredCorners=false
      bSeparateFillAlpha=false
      outlineMaterial=/Game/UI/Foundation/Materials/M-FontShadow
      outlineSize=2
      outlineColor=<rgba 0.0083 0 0.025 0.25>
     }
    }
    renderTransform {angle=0, scale=<1 1>, shear=<-12 0>, translation=<-5 1>}
    shadowOffset=<0 0>
   }
   slot {class=HorizontalBoxSlot, properties {verticalAlignment=VAlign_Top, padding=<ltrb 6 8 0 0>}}
  }
  XPRemaining {
   widgetClassPath=&Overlay
   bIsVariable=true
   parent=Root
   widget=XPRemaining
   slot {class=OverlaySlot, properties {horizontalAlignment=HAlign_Center, verticalAlignment=VAlign_Center, padding=<ltrb 0 100 0 0>}}
  }
  XPAmountBox {
   widgetClassPath=&HorizontalBox
   bIsVariable=true
   parent=XPRemaining
   widget=XPAmountBox
   properties {renderOpacity=0, visibility=Collapsed}
   slot {class=OverlaySlot, properties {horizontalAlignment=HAlign_Center, verticalAlignment=VAlign_Center}}
  }
  XpAmount {
   widgetClassPath=&CommonRichTextBlock
   bIsVariable=true
   parent=XPAmountBox
   widget=XpAmount
   properties {
    bOverrideDefaultStyle=true
    justification=Right
    text=+999
    defaultTextStyleOverride {
     overflowPolicy=Clip
     transformPolicy=None
     colorAndOpacity {colorUseRule=UseColor_Specified, specifiedColor=<rgba 0.0513 0.521 1 1>}
     font {
      bForceMonospaced=false
      bMaterialIsStencil=false
      fontMaterial=null
      fontObject=&HeadingNow
      letterSpacing=0
      monospacedWidth=1
      size=12
      skewAmount=0
      typefaceFontName=86 Bold
      outlineSettings {
       bApplyOutlineToDropShadows=false
       bMiteredCorners=false
       bSeparateFillAlpha=false
       outlineMaterial=null
       outlineSize=0
       outlineColor=<rgba 0 0 0 1>
      }
     }
     highlightColor {colorUseRule=UseColor_Specified, specifiedColor=<rgba 0 0 0 1>}
     highlightShape {
      bIsDynamicallyLoaded=false
      drawAs=Image
      imageType=NoImage
      mirroring=NoMirror
      resourceName=None
      resourceObject=null
      tiling=NoTile
      imageSize=<32 32>
      margin=<ltrb 0 0 0 0>
      outlineSettings {
       bUseBrushTransparency=false
       roundingType=HalfHeightRadius
       width=0
       color {colorUseRule=UseColor_Specified, specifiedColor=<rgba 0 0 0 0>}
       cornerRadii=<0 0 0 0>
      }
      tintColor {colorUseRule=UseColor_Specified, specifiedColor=<rgba 1 1 1 1>}
      uVRegion {bIsValid=false, max=<0 0>, min=<0 0>}
     }
     selectedBackgroundColor {colorUseRule=UseColor_Foreground, specifiedColor=<rgba 1 0 1 1>}
     shadowColorAndOpacity=<rgba 0 0 0 1>
     shadowOffset=<0 0>
     strikeBrush {
      bIsDynamicallyLoaded=false
      drawAs=Image
      imageType=NoImage
      mirroring=NoMirror
      resourceName=None
      resourceObject=null
      tiling=NoTile
      imageSize=<32 32>
      margin=<ltrb 0 0 0 0>
      outlineSettings {
       bUseBrushTransparency=false
       roundingType=HalfHeightRadius
       width=0
       color {colorUseRule=UseColor_Specified, specifiedColor=<rgba 0 0 0 0>}
       cornerRadii=<0 0 0 0>
      }
      tintColor {colorUseRule=UseColor_Specified, specifiedColor=<rgba 1 1 1 1>}
      uVRegion {bIsValid=false, max=<0 0>, min=<0 0>}
     }
     underlineBrush {
      bIsDynamicallyLoaded=false
      drawAs=Image
      imageType=NoImage
      mirroring=NoMirror
      resourceName=None
      resourceObject=null
      tiling=NoTile
      imageSize=<32 32>
      margin=<ltrb 0 0 0 0>
      outlineSettings {
       bUseBrushTransparency=false
       roundingType=HalfHeightRadius
       width=0
       color {colorUseRule=UseColor_Specified, specifiedColor=<rgba 0 0 0 0>}
       cornerRadii=<0 0 0 0>
      }
      tintColor {colorUseRule=UseColor_Specified, specifiedColor=<rgba 1 1 1 1>}
      uVRegion {bIsValid=false, max=<0 0>, min=<0 0>}
     }
    }
   }
   slot {class=HorizontalBoxSlot, properties {horizontalAlignment=HAlign_Left, verticalAlignment=VAlign_Top}}
  }
  XPIcon {
   widgetClassPath=&Image
   bIsVariable=true
   parent=XPAmountBox
   widget=XPIcon
   properties {
    visibility=SelfHitTestInvisible
    brush {
     bIsDynamicallyLoaded=false
     drawAs=Image
     imageType=FullColor
     mirroring=NoMirror
     resourceName=None
     resourceObject=&T_UI_UniversalXP_Icon_Standard_64
     tiling=NoTile
     imageSize=<24 24>
     margin=<ltrb 0 0 0 0>
     outlineSettings {
      bUseBrushTransparency=false
      roundingType=HalfHeightRadius
      width=0
      color {colorUseRule=UseColor_Specified, specifiedColor=<rgba 0 0 0 0>}
      cornerRadii=<0 0 0 0>
     }
     tintColor {colorUseRule=UseColor_Specified, specifiedColor=<rgba 1 1 1 1>}
     uVRegion {bIsValid=false, max=<0 0>, min=<0 0>}
    }
    renderTransform {angle=0, scale=<1.333 1.333>, shear=<0 0>, translation=<0 -1>}
   }
   slot {class=HorizontalBoxSlot, properties {verticalAlignment=VAlign_Center, padding=<ltrb 6 0 0 0>}}
  }
  Txt_XPToNextLevel {
   widgetClassPath=&CommonTextBlock
   bIsVariable=true
   parent=XPRemaining
   widget=Txt_XPToNextLevel
   properties {
    bIsScrollingEnabled=false
    text="68,067K"
    font {
     bForceMonospaced=false
     bMaterialIsStencil=false
     fontMaterial=null
     fontObject=&HeadingNow
     letterSpacing=0
     monospacedWidth=1
     size=18
     skewAmount=0
     typefaceFontName=86 Bold
     outlineSettings {
      bApplyOutlineToDropShadows=false
      bMiteredCorners=false
      bSeparateFillAlpha=false
      outlineMaterial=null
      outlineSize=0
      outlineColor=<rgba 0 0 0 1>
     }
    }
    shadowOffset=<0 0>
   }
   slot {class=OverlaySlot, properties {horizontalAlignment=HAlign_Center}}
  }
  Txt_XPToNextLevel_Label {
   widgetClassPath=&CommonTextBlock
   bIsVariable=true
   parent=XPRemaining
   widget=Txt_XPToNextLevel_Label
   properties {
    bIsScrollingEnabled=false
    text=To Lvl 27
    textTransformPolicy=ToUpper
    colorAndOpacity {colorUseRule=UseColor_Specified, specifiedColor=<rgba 0.6584 0.4179 1 1>}
    font {
     bForceMonospaced=false
     bMaterialIsStencil=false
     fontMaterial=null
     fontObject=&HeadingNow
     letterSpacing=0
     monospacedWidth=1
     size=15.75
     skewAmount=0
     typefaceFontName=75 Medium
     outlineSettings {
      bApplyOutlineToDropShadows=false
      bMiteredCorners=false
      bSeparateFillAlpha=false
      outlineMaterial=null
      outlineSize=0
      outlineColor=<rgba 0 0 0 1>
     }
    }
    shadowOffset=<0 0>
   }
   slot {class=OverlaySlot, properties {horizontalAlignment=HAlign_Center, padding=<ltrb 0 30 0 0>}}
  }
  SocialAvatarIcon {widgetClassPath=/Script/FortniteUI.FortSocialAvatarIcon, bInherited=true, widget=null}
 ]
}
